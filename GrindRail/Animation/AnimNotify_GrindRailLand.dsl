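; Anim notify for the grind rail feature that signals when a pawn's feet land during an animation.
; It gets the mesh component's owner and looks up its BP_FortPawnGrindingProvider component.
; If that component is valid, it calls BPonFeetLanded on it, and the notify returns true either way.
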(fn Received_Notify
  (params (MeshComp :type Skeletal%20Mesh%20Component%20Object%20Reference) (Animation :type Anim%20Sequence%20Base%20Object%20Reference) (EventReference :type Anim%20Notify%20Event%20Reference%20Structure%20%28by%20ref%29))
  (bind _returnvalue (Components|GetOwner MeshComp))
  (bind _returnvalue_2 (Actor|GetComponentbyClass _returnvalue "/GrindRail/Blueprints/BP_FortPawnGrindingProvider.BP_FortPawnGrindingProvider_C"))
  (Utilities|FlowControl|Sequence
    (:then_0
      (Utilities|IsValid _returnvalue_2
        (:Is%20Valid
          (Class|FortPawnComponentGrindingProvider|BPonFeetLanded _returnvalue_2))))
    (:then_1
      (return true))))
