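# Player animation layer Blueprint for grind-rail riding, filling the MovementMode layer slots on the Fortnite player skeleton.
# A FullBody state machine covers mount, idle, locomotion, backward lean, 180 turns, jump and boost, with additive leans and aim offsets.
# It is cached alongside a CombatMode machine, whose output gets grind-rail slope warping for feet and pelvis.
# The final-pose layer repurposes Leg IK for the left hand.
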AnimBlueprint /GrindRail/Animation/GrindRail_Player_LayerAnimBP
target_skeleton=&Fortnite_M_Avg_Player_Skeleton
graphs [
 AnimGraph {nodes [Output Pose() {layerGroup=None}]}
 EventGraph {links [], nodes []}
 LocomotionAdditiveGroup {nodes [
   n1 = Input Pose() {name=InPoseLocomotionAdditive}
   n2 = Get AimYawSmoothed() -> (AimYawSmoothed:double)
   n3 = Apply Additive(
    Base=AimOffset Player 'GrindRail_Pitch_AO'!(
     X:float=0
     Y:float=Get!() -> (SplineRelativeAim_Roll:float, SplineRelativeAim_Pitch:float, SplineRelativeAim_Yaw:float).SplineRelativeAim_Pitch
     BasePose:Pose Link Structure=AimOffset Player 'GrindRail_Yaw_AO'!(X:float=@n2, Y:float=0, BasePose:Pose Link Structure=@n1, Alpha:float=1) -> (Pose:Pose Link Structure) {blendSpace=/GrindRail/Animation/Male/_AO/GrindRail_Yaw_AO, alphaBoolBlend {blendOption=Linear}}
     Alpha:float=0.8
    ) -> (Pose:Pose Link Structure) {blendSpace=/GrindRail/Animation/Male/_AO/GrindRail_Pitch_AO, alphaBoolBlend {blendOption=Linear}}
    Additive=Blendspace Player 'GrindRail_Auto_Leans_BS'!(X=0) -> (Pose) {blendSpace=/GrindRail/Animation/Male/_BlendSpace/GrindRail_Auto_Leans_BS}
   ) {alphaBoolBlend {blendOption=Linear}}
   n4 = Blendspace Player 'GrindRail_FwdLeans_BS'(X=0, Y=0) {blendSpace=/GrindRail/Animation/Male/_BlendSpace/GrindRail_FwdLeans_BS}
   Output Pose(Result=Apply Additive!(
     Base=Apply Additive!(Base=@n3, Additive=@n4, Alpha:float) -> (Pose) {alphaBoolBlend {blendOption=Linear}}
     Additive=Blendspace Player 'GrindRail_MovingAdditive_BS'!(Y=0, PlayRate:float=1) -> (Pose) {blendSpace=/GrindRail/Animation/Male/_BlendSpace/GrindRail_MovingAdditive_BS}
     bAlphaBoolEnabled:bool=True
    ) -> (Pose) {alphaInputType=Bool, alphaBoolBlend {blendOption=Linear}}) {layerGroup=None}
  ]}
 MovementMode_AimOffset {nodes [
   Input Pose() {name=InAimOffsetsPose}
   Input Pose() {name=InSourcePosePreAimOffsets}
   Output Pose() {layerGroup=MovementMode}
  ]}
 MovementMode_BaseLocomotion {nodes [
   Input Pose() {name=InPoseBaseLocomotion}
   Output Pose() {layerGroup=MovementMode}
  ]}
 MovementMode_FinalPoseOverride {nodes [
   n1 = Leg IK!(ComponentPose:Component Space Pose Link Structure=Local To Component(LocalPose=Input Pose() {name=InFinalPose}), bAlphaBoolEnabled:bool=True) -> (Pose:Component Space Pose Link Structure) {
    is=title AnimGraphNode_LegIK
    alphaInputType=Bool
    alphaBoolBlend {blendOption=Linear}
    legsDefinition [{
      bEnableKneeTwistCorrection=false
      bEnableRotationLimit=false
      footBoneForwardAxis=Y
      hingeRotationAxis=Z
      minRotationAngle=15
      numBonesInLimb=2
      twistOffsetCurveName=None
      fKFootBone {boneName=hand_l}
      iKFootBone {boneName=ik_hand_l}
     }]
   }
   Output Pose(Result=Component To Local(ComponentPose=@n1)) {layerGroup=MovementMode}
  ]}
 MovementMode_FullBodyOverride {nodes [
   AnimGraphNode_StateMachine_1 = CombatMode() {is=title AnimGraphNode_StateMachine}
   n1 = Grind Rail Slope Warping!(ComponentPose:Component Space Pose Link Structure=Local To Component(LocalPose=@AnimGraphNode_StateMachine_1), Alpha:float=1) -> (Pose:Component Space Pose Link Structure) {
    is=title FortAnimGraphNode_GrindRailSlopeWarping
    bDisableInterpolationOnFootOffset=true
    bKeepMeshInsideOfCapsule=false
    bLeftCurveHighMeansFootDown=false
    bRightCurveHighMeansFootDown=false
    legIKToFKLengthOverExtensionToleranceDistance=15
    maxStepHeight=85
    alphaBoolBlend {blendOption=Linear}
    feetDefinitions [
     {numBonesInLimb=2, fKFootBone {boneName=foot_l}, iKFootBone {boneName=ik_foot_l}}
     {numBonesInLimb=2, fKFootBone {boneName=foot_r}, iKFootBone {boneName=ik_foot_r}}
    ]
    floorNormalInterpolator {stiffnessConstant=1500}
    floorOffsetInterpolator {stiffnessConstant=1500}
    iKFootRootBone {boneName=ik_foot_root}
    pelvisBone {boneName=pelvis}
    pelvisOffsetInterpolator {stiffnessConstant=1500}
   }
   Output Pose(Result=Modify Curve!(SourcePose=Component To Local(ComponentPose=@n1), CurveMap:"Map of Names to Float (single-precision)s"=(), Alpha=1) -> (Pose)) {layerGroup=MovementMode}
   Save cached pose 'CachedMainPlayerPose'(Pose=Input Pose() {name=InputPoseFullBody}) {cacheName=CachedMainPlayerPose}
   Save cached pose 'RootLeanPose'(Pose=Blendspace Player 'GrindRail_Auto_Leans_BS'(X=0) {blendSpace=&GrindRail_Auto_Leans_BS}) {cacheName=RootLeanPose}
   Save cached pose 'MovingPose'(Pose=Blendspace Player 'GrindRail_Moving_BS'(X=0, Y=0) {blendSpace=/GrindRail/Animation/Male/_BlendSpace/GrindRail_Moving_BS}) {cacheName=MovingPose}
   Save cached pose 'BoostLeanPose'(Pose=Blendspace Player 'GrindRail_Boost_Leans_BS'!(X=0) -> (Pose) {blendSpace=/GrindRail/Animation/Male/_BlendSpace/GrindRail_Boost_Leans_BS}) {cacheName=BoostLeanPose}
   AnimGraphNode_StateMachine_0 = FullBody() {is=title AnimGraphNode_StateMachine, bAllowConduitEntryStates=true}
   Save cached pose 'FullBodyPose'(Pose=@AnimGraphNode_StateMachine_0) {cacheName=FullBodyPose}
  ]}
 FullBody {
  entry=InitialSelector
  path=MovementMode_FullBodyOverride.AnimGraphNode_StateMachine_0.FullBody
  ids=id
  nodes [
   FromEntryToLoco() {is=title AnimStateAliasNode, also {aliases [Locomotion]}}
   ToBoost() {is=title AnimStateAliasNode, also {aliases [BoostOutro, Entry, Idle, Lean_Bwd, Lean_Bwd_Outro, Locomotion, Turn]}}
   ToLocoSelector() {is=title AnimStateAliasNode, also {aliases [Idle]}}
   ToEntry() {is=title AnimStateAliasNode, also {aliases [Entry]}}
   FromBoostToLoco() {is=title AnimStateAliasNode, also {aliases [Locomotion]}}
   ToJump() {
    is=title AnimStateAliasNode
    also {aliases [Boost, BoostOutro, BoostTurn, Entry, Idle, JumpOutro, Lean_Bwd, Lean_Bwd_Outro, Locomotion, Turn]}
   }
   ToLocomotion() {is=title AnimStateAliasNode, also {aliases [Locomotion]}}
   AnimStateConduitNode_0 = LocoSelector() {is=title AnimStateConduitNode}
   AnimStateConduitNode_1 = InitialSelector() {is=title AnimStateConduitNode}
   AnimStateNode_0 = Turn() {
    is=title AnimStateNode
    bAlwaysResetOnEntry=true
    stateEntered {guid=#1}
    stateFullyBlended {guid=#2}
    stateLeft {guid=#3}
   }
   AnimStateNode_1 = Lean_Bwd() {is=title AnimStateNode}
   AnimStateNode_10 = Lean_Bwd_Outro() {is=title AnimStateNode}
   AnimStateNode_2 = JumpOutro() {is=title AnimStateNode}
   AnimStateNode_3 = Locomotion() {is=title AnimStateNode}
   AnimStateNode_4 = Idle() {stateEntered {guid=#4, notifyName=IdleEnter}}
   AnimStateNode_5 = Jump() {is=title AnimStateNode}
   AnimStateNode_6 = Boost() {is=title AnimStateNode}
   AnimStateNode_7 = Entry() {
    is=title AnimStateNode
    stateEntered {guid=#5, notifyName=EntryEnter}
    stateLeft {guid=#6, notifyName=EntryExit}
   }
   AnimStateNode_8 = BoostOutro() {is=title AnimStateNode}
   AnimStateNode_9 = BoostTurn() {is=title AnimStateNode}
   AnimStateTransitionNode_0 = AnimStateTransitionNode()
   AnimStateTransitionNode_1 = AnimStateTransitionNode() {crossfadeDuration=0.033, logicType=TLT_Inertialization, priorityOrder=2}
   AnimStateTransitionNode_10 = AnimStateTransitionNode() {bAutomaticRuleBasedOnSequencePlayerInState=true, crossfadeDuration=0.6}
   AnimStateTransitionNode_11 = AnimStateTransitionNode() {bAutomaticRuleBasedOnSequencePlayerInState=true, crossfadeDuration=0.4}
   AnimStateTransitionNode_12 = AnimStateTransitionNode() {logicType=TLT_Inertialization}
   AnimStateTransitionNode_13 = AnimStateTransitionNode() {bAutomaticRuleBasedOnSequencePlayerInState=true, priorityOrder=2}
   AnimStateTransitionNode_14 = AnimStateTransitionNode() {crossfadeDuration=0.067, logicType=TLT_Inertialization}
   AnimStateTransitionNode_15 = AnimStateTransitionNode() {priorityOrder=0}
   AnimStateTransitionNode_16 = AnimStateTransitionNode()
   AnimStateTransitionNode_17 = AnimStateTransitionNode() {crossfadeDuration=0.4, logicType=TLT_Inertialization, priorityOrder=100}
   AnimStateTransitionNode_18 = AnimStateTransitionNode() {logicType=TLT_Inertialization}
   AnimStateTransitionNode_2 = AnimStateTransitionNode() {crossfadeDuration=0.35, logicType=TLT_Inertialization}
   AnimStateTransitionNode_22 = AnimStateTransitionNode()
   AnimStateTransitionNode_23 = AnimStateTransitionNode() {logicType=TLT_Inertialization, priorityOrder=100}
   AnimStateTransitionNode_24 = AnimStateTransitionNode()
   AnimStateTransitionNode_25 = AnimStateTransitionNode() {bAutomaticRuleBasedOnSequencePlayerInState=true, priorityOrder=2}
   AnimStateTransitionNode_26 = AnimStateTransitionNode() {bAutomaticRuleBasedOnSequencePlayerInState=true}
   AnimStateTransitionNode_3 = AnimStateTransitionNode() {bAutomaticRuleBasedOnSequencePlayerInState=true, crossfadeDuration=0.35, logicType=TLT_Inertialization, priorityOrder=2}
   AnimStateTransitionNode_36 = AnimStateTransitionNode()
   AnimStateTransitionNode_4 = AnimStateTransitionNode() {crossfadeDuration=0.35, priorityOrder=2}
   AnimStateTransitionNode_5 = AnimStateTransitionNode()
   AnimStateTransitionNode_6 = AnimStateTransitionNode() {crossfadeDuration=0.033, logicType=TLT_Inertialization}
   AnimStateTransitionNode_7 = AnimStateTransitionNode() {logicType=TLT_Inertialization}
   AnimStateTransitionNode_8 = AnimStateTransitionNode() {crossfadeDuration=0.55, logicType=TLT_Inertialization}
   AnimStateTransitionNode_9 = AnimStateTransitionNode() {logicType=TLT_Inertialization}
  ]
  links [
   {
    dst=BoostOutro
    rule=MovementMode_FullBodyOverride.AnimGraphNode_StateMachine_0.FullBody.AnimStateTransitionNode_9.Transition
    src=Boost
    via=AnimStateTransitionNode_9
   }
   {
    dst=BoostTurn
    rule=MovementMode_FullBodyOverride.AnimGraphNode_StateMachine_0.FullBody.AnimStateTransitionNode_5.Transition
    src=Boost
    via=AnimStateTransitionNode_5
   }
   {
    dst=FromBoostToLoco
    rule=MovementMode_FullBodyOverride.AnimGraphNode_StateMachine_0.FullBody.AnimStateTransitionNode_10.Transition
    src=BoostOutro
    via=AnimStateTransitionNode_10
   }
   {
    dst=Boost
    rule=MovementMode_FullBodyOverride.AnimGraphNode_StateMachine_0.FullBody.AnimStateTransitionNode_13.Transition
    src=BoostTurn
    via=AnimStateTransitionNode_13
   }
   {
    dst=Boost
    rule=MovementMode_FullBodyOverride.AnimGraphNode_StateMachine_0.FullBody.AnimStateTransitionNode_18.Transition
    src=BoostTurn
    via=AnimStateTransitionNode_18
   }
   {
    dst=FromEntryToLoco
    rule=MovementMode_FullBodyOverride.AnimGraphNode_StateMachine_0.FullBody.AnimStateTransitionNode_15.Transition
    src=Entry
    via=AnimStateTransitionNode_15
   }
   {
    dst=FromEntryToLoco
    rule=MovementMode_FullBodyOverride.AnimGraphNode_StateMachine_0.FullBody.AnimStateTransitionNode_8.Transition
    src=Entry
    via=AnimStateTransitionNode_8
   }
   {
    dst=Idle
    rule=MovementMode_FullBodyOverride.AnimGraphNode_StateMachine_0.FullBody.AnimStateTransitionNode_25.Transition
    src=Entry
    via=AnimStateTransitionNode_25
   }
   {
    dst=Entry
    rule=MovementMode_FullBodyOverride.AnimGraphNode_StateMachine_0.FullBody.AnimStateTransitionNode_24.Transition
    src=InitialSelector
    via=AnimStateTransitionNode_24
   }
   {
    dst=Idle
    rule=MovementMode_FullBodyOverride.AnimGraphNode_StateMachine_0.FullBody.AnimStateTransitionNode_23.Transition
    src=InitialSelector
    via=AnimStateTransitionNode_23
   }
   {
    dst=JumpOutro
    rule=MovementMode_FullBodyOverride.AnimGraphNode_StateMachine_0.FullBody.AnimStateTransitionNode_1.Transition
    src=Jump
    via=AnimStateTransitionNode_1
   }
   {
    dst=ToEntry
    rule=MovementMode_FullBodyOverride.AnimGraphNode_StateMachine_0.FullBody.AnimStateTransitionNode_12.Transition
    src=Jump
    via=AnimStateTransitionNode_12
   }
   {
    dst=ToLocomotion
    rule=MovementMode_FullBodyOverride.AnimGraphNode_StateMachine_0.FullBody.AnimStateTransitionNode_11.Transition
    src=JumpOutro
    via=AnimStateTransitionNode_11
   }
   {
    dst=Lean_Bwd_Outro
    rule=MovementMode_FullBodyOverride.AnimGraphNode_StateMachine_0.FullBody.AnimStateTransitionNode_2.Transition
    src=Lean_Bwd
    via=AnimStateTransitionNode_2
   }
   {
    dst=LocoSelector
    rule=MovementMode_FullBodyOverride.AnimGraphNode_StateMachine_0.FullBody.AnimStateTransitionNode_22.Transition
    src=Lean_Bwd_Outro
    via=AnimStateTransitionNode_22
   }
   {
    dst=Locomotion
    rule=MovementMode_FullBodyOverride.AnimGraphNode_StateMachine_0.FullBody.AnimStateTransitionNode_3.Transition
    src=Lean_Bwd_Outro
    via=AnimStateTransitionNode_3
   }
   {
    dst=Lean_Bwd
    rule=MovementMode_FullBodyOverride.AnimGraphNode_StateMachine_0.FullBody.AnimStateTransitionNode_4.Transition
    src=LocoSelector
    via=AnimStateTransitionNode_4
   }
   {
    dst=Locomotion
    rule=MovementMode_FullBodyOverride.AnimGraphNode_StateMachine_0.FullBody.AnimStateTransitionNode_17.Transition
    src=LocoSelector
    via=AnimStateTransitionNode_17
   }
   {
    dst=Turn
    rule=MovementMode_FullBodyOverride.AnimGraphNode_StateMachine_0.FullBody.AnimStateTransitionNode_14.Transition
    src=LocoSelector
    via=AnimStateTransitionNode_14
   }
   {
    dst=LocoSelector
    rule=MovementMode_FullBodyOverride.AnimGraphNode_StateMachine_0.FullBody.AnimStateTransitionNode_0.Transition
    src=Locomotion
    via=AnimStateTransitionNode_0
   }
   {
    dst=Boost
    rule=MovementMode_FullBodyOverride.AnimGraphNode_StateMachine_0.FullBody.AnimStateTransitionNode_7.Transition
    src=ToBoost
    via=AnimStateTransitionNode_7
   }
   {
    dst=Jump
    rule=MovementMode_FullBodyOverride.AnimGraphNode_StateMachine_0.FullBody.AnimStateTransitionNode_6.Transition
    src=ToJump
    via=AnimStateTransitionNode_6
   }
   {
    dst=LocoSelector
    rule=MovementMode_FullBodyOverride.AnimGraphNode_StateMachine_0.FullBody.AnimStateTransitionNode_16.Transition
    src=ToLocoSelector
    via=AnimStateTransitionNode_16
   }
   {
    dst=LocoSelector
    rule=MovementMode_FullBodyOverride.AnimGraphNode_StateMachine_0.FullBody.AnimStateTransitionNode_26.Transition
    src=Turn
    via=AnimStateTransitionNode_26
   }
   {
    dst=LocoSelector
    rule=MovementMode_FullBodyOverride.AnimGraphNode_StateMachine_0.FullBody.AnimStateTransitionNode_36.Transition
    src=Turn
    via=AnimStateTransitionNode_36
   }
  ]
 }
 LocoSelector {
  path=MovementMode_FullBodyOverride.AnimGraphNode_StateMachine_0.FullBody.AnimStateConduitNode_0.LocoSelector
  nodes [Result(bCanEnterTransition=true)]
 }
 InitialSelector {
  path=MovementMode_FullBodyOverride.AnimGraphNode_StateMachine_0.FullBody.AnimStateConduitNode_1.InitialSelector
  nodes [Result(bCanEnterTransition=true)]
 }
 Turn {
  path=MovementMode_FullBodyOverride.AnimGraphNode_StateMachine_0.FullBody.AnimStateNode_0.Turn
  nodes [
   AnimGraphNode_StateMachine_0 = TurnDirection() {is=title AnimGraphNode_StateMachine, bAllowConduitEntryStates=true}
   Output Animation Pose(Result=@AnimGraphNode_StateMachine_0)
  ]
 }
 TurnDirection {
  entry=TurnSelector
  path=MovementMode_FullBodyOverride.AnimGraphNode_StateMachine_0.FullBody.AnimStateNode_0.Turn.AnimGraphNode_StateMachine_0.TurnDirection
  ids=id
  nodes [
   AnimStateConduitNode_0 = TurnSelector() {is=title AnimStateConduitNode}
   AnimStateNode_0 = 180TurnLeft() {is=title AnimStateNode}
   AnimStateNode_1 = 180TurnRight() {is=title AnimStateNode}
   AnimStateTransitionNode_3 = AnimStateTransitionNode() {logicType=TLT_Inertialization}
   AnimStateTransitionNode_4 = AnimStateTransitionNode() {logicType=TLT_Inertialization, priorityOrder=100}
  ]
  links [
   {
    dst=180TurnLeft
    rule=MovementMode_FullBodyOverride.AnimGraphNode_StateMachine_0.FullBody.AnimStateNode_0.Turn.AnimGraphNode_StateMachine_0.TurnDirection.AnimStateTransitionNode_3.Transition
    src=TurnSelector
    via=AnimStateTransitionNode_3
   }
   {
    dst=180TurnRight
    rule=MovementMode_FullBodyOverride.AnimGraphNode_StateMachine_0.FullBody.AnimStateNode_0.Turn.AnimGraphNode_StateMachine_0.TurnDirection.AnimStateTransitionNode_4.Transition
    src=TurnSelector
    via=AnimStateTransitionNode_4
   }
  ]
 }
 TurnSelector {
  path=MovementMode_FullBodyOverride.AnimGraphNode_StateMachine_0.FullBody.AnimStateNode_0.Turn.AnimGraphNode_StateMachine_0.TurnDirection.AnimStateConduitNode_0.TurnSelector
  nodes [Result(bCanEnterTransition=true)]
 }
 180TurnLeft {
  path=MovementMode_FullBodyOverride.AnimGraphNode_StateMachine_0.FullBody.AnimStateNode_0.Turn.AnimGraphNode_StateMachine_0.TurnDirection.AnimStateNode_0.180TurnLeft
  nodes [Output Animation Pose(Result=Sequence Player 'GrindRail_Base_180Turn_Left'() {bLoopAnimation=false, sequence=/GrindRail/Animation/Male/Base/GrindRail_Base_180Turn_Left})]
 }
 180TurnRight {
  path=MovementMode_FullBodyOverride.AnimGraphNode_StateMachine_0.FullBody.AnimStateNode_0.Turn.AnimGraphNode_StateMachine_0.TurnDirection.AnimStateNode_1.180TurnRight
  nodes [Output Animation Pose(Result=Sequence Player 'GrindRail_Base_180Turn_Right'() {bLoopAnimation=false, sequence=/GrindRail/Animation/Male/Base/GrindRail_Base_180Turn_Right})]
 }
 Transition {
  path=MovementMode_FullBodyOverride.AnimGraphNode_StateMachine_0.FullBody.AnimStateNode_0.Turn.AnimGraphNode_StateMachine_0.TurnDirection.AnimStateTransitionNode_3.Transition
  nodes [Result(bCanEnterTransition=Get bIsLeft180Turn() -> (bIsLeft180Turn:bool))]
 }
 Transition {
  path=MovementMode_FullBodyOverride.AnimGraphNode_StateMachine_0.FullBody.AnimStateNode_0.Turn.AnimGraphNode_StateMachine_0.TurnDirection.AnimStateTransitionNode_4.Transition
  nodes [Result(bCanEnterTransition=true)]
 }
 Lean_Bwd {
  path=MovementMode_FullBodyOverride.AnimGraphNode_StateMachine_0.FullBody.AnimStateNode_1.Lean_Bwd
  nodes [
   n1 = Blendspace Player 'GrindRail_Leans_Bwd_BS'!(X=0) -> (Pose) {blendSpace=/GrindRail/Animation/Male/_BlendSpace/GrindRail_Leans_Bwd_BS}
   AnimGraphNode_StateMachine_0 = LeanBwdTransition() {is=title AnimGraphNode_StateMachine}
   Output Animation Pose(Result=Apply Additive(
     Base=Apply Additive(Base=@AnimGraphNode_StateMachine_0, Additive=@n1, Alpha=0.5) {alphaBoolBlend {blendOption=Linear}}
     Additive=Use cached pose 'RootLeanPose'()
    ) {alphaBoolBlend {blendOption=Linear}})
  ]
 }
 LeanBwdTransition {
  entry=Lean_Bwd_Start
  path=MovementMode_FullBodyOverride.AnimGraphNode_StateMachine_0.FullBody.AnimStateNode_1.Lean_Bwd.AnimGraphNode_StateMachine_0.LeanBwdTransition
  ids=id
  nodes [
   AnimStateNode_0 = Lean_Bwd_Start() {is=title AnimStateNode}
   AnimStateNode_1 = Lean_Bwd_Loop() {is=title AnimStateNode}
   AnimStateTransitionNode_0 = AnimStateTransitionNode() {bAutomaticRuleBasedOnSequencePlayerInState=true, logicType=TLT_Inertialization}
  ]
  links [{
    dst=Lean_Bwd_Loop
    rule=MovementMode_FullBodyOverride.AnimGraphNode_StateMachine_0.FullBody.AnimStateNode_1.Lean_Bwd.AnimGraphNode_StateMachine_0.LeanBwdTransition.AnimStateTransitionNode_0.Transition
    src=Lean_Bwd_Start
    via=AnimStateTransitionNode_0
   }]
 }
 Lean_Bwd_Start {
  path=MovementMode_FullBodyOverride.AnimGraphNode_StateMachine_0.FullBody.AnimStateNode_1.Lean_Bwd.AnimGraphNode_StateMachine_0.LeanBwdTransition.AnimStateNode_0.Lean_Bwd_Start
  nodes [Output Animation Pose(Result=Sequence Player 'GrindRail_PoseB_to_Bckwd_Transition'() {bLoopAnimation=false, sequence=/GrindRail/Animation/Male/Base/GrindRail_PoseB_to_Bckwd_Transition})]
 }
 Lean_Bwd_Loop {
  path=MovementMode_FullBodyOverride.AnimGraphNode_StateMachine_0.FullBody.AnimStateNode_1.Lean_Bwd.AnimGraphNode_StateMachine_0.LeanBwdTransition.AnimStateNode_1.Lean_Bwd_Loop
  nodes [Output Animation Pose(Result=Apply Additive(
     Base=Sequence Player 'GrindRail_Base_Lean_Backward'() {sequence=/GrindRail/Animation/Male/Base/GrindRail_Base_Lean_Backward}
     Additive="Sequence Player (additive) 'GrindRail_Bwd_Idle_Noise'"() {sequence=/GrindRail/Animation/Male/Base/GrindRail_Bwd_Idle_Noise}
    ) {alphaBoolBlend {blendOption=Linear}})]
 }
 Transition {
  path=MovementMode_FullBodyOverride.AnimGraphNode_StateMachine_0.FullBody.AnimStateNode_1.Lean_Bwd.AnimGraphNode_StateMachine_0.LeanBwdTransition.AnimStateTransitionNode_0.Transition
  nodes [Result()]
 }
 Lean_Bwd_Outro {
  path=MovementMode_FullBodyOverride.AnimGraphNode_StateMachine_0.FullBody.AnimStateNode_10.Lean_Bwd_Outro
  nodes [Output Animation Pose(Result=Apply Additive(
     Base=Sequence Player 'GrindRail_Bckwd_to_PoseB_Transition'!(StartPosition:float=0) -> (Pose) {bLoopAnimation=false, sequence=/GrindRail/Animation/Male/Base/GrindRail_Bckwd_to_PoseB_Transition}
     Additive=Use cached pose 'RootLeanPose'()
    ) {alphaBoolBlend {blendOption=Linear}})]
 }
 JumpOutro {
  path=MovementMode_FullBodyOverride.AnimGraphNode_StateMachine_0.FullBody.AnimStateNode_2.JumpOutro
  nodes [Output Animation Pose(Result=Apply Additive(
     Base=Sequence Player 'GrindRail_Base_Jump_Land'() {bLoopAnimation=false, sequence=/GrindRail/Animation/Male/Base/GrindRail_Base_Jump_Land}
     Additive=Use cached pose 'RootLeanPose'()
    ) {alphaBoolBlend {blendOption=Linear}})]
 }
 Locomotion {
  path=MovementMode_FullBodyOverride.AnimGraphNode_StateMachine_0.FullBody.AnimStateNode_3.Locomotion
  nodes [Output Animation Pose(Result=LocomotionAdditiveGroup!(InPoseLocomotionAdditive:Pose Link Structure=Use cached pose 'MovingPose'()) -> (Pose:Pose Link Structure) {is=title AnimGraphNode_LinkedAnimLayer, layer=LocomotionAdditiveGroup})]
 }
 Idle {
  path=MovementMode_FullBodyOverride.AnimGraphNode_StateMachine_0.FullBody.AnimStateNode_4.Idle
  nodes [Output Animation Pose(Result=Use cached pose 'MovingPose'())]
 }
 Jump {
  path=MovementMode_FullBodyOverride.AnimGraphNode_StateMachine_0.FullBody.AnimStateNode_5.Jump
  nodes [
   AnimGraphNode_StateMachine_0 = JumpTransition() {is=title AnimGraphNode_StateMachine}
   Output Animation Pose(Result=Apply Additive(Base=@AnimGraphNode_StateMachine_0, Additive=Use cached pose 'RootLeanPose'()) {alphaBoolBlend {blendOption=Linear}})
  ]
 }
 JumpTransition {
  entry=Jump_Intro
  path=MovementMode_FullBodyOverride.AnimGraphNode_StateMachine_0.FullBody.AnimStateNode_5.Jump.AnimGraphNode_StateMachine_0.JumpTransition
  ids=id
  nodes [
   AnimStateNode_0 = Jump_Intro() {is=title AnimStateNode}
   AnimStateNode_1 = Jump_loop() {is=title AnimStateNode}
   AnimStateTransitionNode_2 = AnimStateTransitionNode() {bAutomaticRuleBasedOnSequencePlayerInState=true}
  ]
  links [{
    dst=Jump_loop
    rule=MovementMode_FullBodyOverride.AnimGraphNode_StateMachine_0.FullBody.AnimStateNode_5.Jump.AnimGraphNode_StateMachine_0.JumpTransition.AnimStateTransitionNode_2.Transition
    src=Jump_Intro
    via=AnimStateTransitionNode_2
   }]
 }
 Jump_Intro {
  path=MovementMode_FullBodyOverride.AnimGraphNode_StateMachine_0.FullBody.AnimStateNode_5.Jump.AnimGraphNode_StateMachine_0.JumpTransition.AnimStateNode_0.Jump_Intro
  nodes [Output Animation Pose(Result=Sequence Player 'GrindRail_Base_Jump_Start'() {bLoopAnimation=false, sequence=/GrindRail/Animation/Male/Base/GrindRail_Base_Jump_Start})]
 }
 Jump_loop {
  path=MovementMode_FullBodyOverride.AnimGraphNode_StateMachine_0.FullBody.AnimStateNode_5.Jump.AnimGraphNode_StateMachine_0.JumpTransition.AnimStateNode_1.Jump_loop
  nodes [Output Animation Pose(Result=Sequence Player 'GrindRail_Base_Jump_Loop'() {sequence=/GrindRail/Animation/Male/Base/GrindRail_Base_Jump_Loop})]
 }
 Transition {
  path=MovementMode_FullBodyOverride.AnimGraphNode_StateMachine_0.FullBody.AnimStateNode_5.Jump.AnimGraphNode_StateMachine_0.JumpTransition.AnimStateTransitionNode_2.Transition
  nodes [Result()]
 }
 Boost {
  path=MovementMode_FullBodyOverride.AnimGraphNode_StateMachine_0.FullBody.AnimStateNode_6.Boost
  nodes [
   AnimGraphNode_StateMachine_1 = BoostTransition() {is=title AnimGraphNode_StateMachine, bAllowConduitEntryStates=true}
   Output Animation Pose(Result=Apply Additive(Base=@AnimGraphNode_StateMachine_1, Additive=Use cached pose 'BoostLeanPose'()) {alphaBoolBlend {blendOption=Linear}})
  ]
 }
 BoostTransition {
  entry=Conduit
  path=MovementMode_FullBodyOverride.AnimGraphNode_StateMachine_0.FullBody.AnimStateNode_6.Boost.AnimGraphNode_StateMachine_1.BoostTransition
  ids=id
  nodes [
   AnimStateConduitNode_0 = Conduit()
   AnimStateNode_0 = Boost_Intro() {is=title AnimStateNode}
   AnimStateNode_1 = Boost_Loop() {is=title AnimStateNode}
   AnimStateTransitionNode_0 = AnimStateTransitionNode() {bAutomaticRuleBasedOnSequencePlayerInState=true}
   AnimStateTransitionNode_2 = AnimStateTransitionNode() {priorityOrder=100}
   AnimStateTransitionNode_3 = AnimStateTransitionNode()
  ]
  links [
   {
    dst=Boost_Loop
    rule=MovementMode_FullBodyOverride.AnimGraphNode_StateMachine_0.FullBody.AnimStateNode_6.Boost.AnimGraphNode_StateMachine_1.BoostTransition.AnimStateTransitionNode_0.AnimationTransitionGraph_0
    src=Boost_Intro
    via=AnimStateTransitionNode_0
   }
   {
    dst=Boost_Intro
    rule=MovementMode_FullBodyOverride.AnimGraphNode_StateMachine_0.FullBody.AnimStateNode_6.Boost.AnimGraphNode_StateMachine_1.BoostTransition.AnimStateTransitionNode_2.Transition
    src=Conduit
    via=AnimStateTransitionNode_2
   }
   {
    dst=Boost_Loop
    rule=MovementMode_FullBodyOverride.AnimGraphNode_StateMachine_0.FullBody.AnimStateNode_6.Boost.AnimGraphNode_StateMachine_1.BoostTransition.AnimStateTransitionNode_3.Transition
    src=Conduit
    via=AnimStateTransitionNode_3
   }
  ]
 }
 Conduit {
  path=MovementMode_FullBodyOverride.AnimGraphNode_StateMachine_0.FullBody.AnimStateNode_6.Boost.AnimGraphNode_StateMachine_1.BoostTransition.AnimStateConduitNode_0.Conduit
  nodes [Result(bCanEnterTransition=true)]
 }
 Boost_Intro {
  path=MovementMode_FullBodyOverride.AnimGraphNode_StateMachine_0.FullBody.AnimStateNode_6.Boost.AnimGraphNode_StateMachine_1.BoostTransition.AnimStateNode_0.Boost_Intro
  nodes [Output Animation Pose(Result=Sequence Player 'GrindRail_Boost_Intro'() {bLoopAnimation=false, sequence=/GrindRail/Animation/Male/Boost/GrindRail_Boost_Intro})]
 }
 Boost_Loop {
  path=MovementMode_FullBodyOverride.AnimGraphNode_StateMachine_0.FullBody.AnimStateNode_6.Boost.AnimGraphNode_StateMachine_1.BoostTransition.AnimStateNode_1.Boost_Loop
  nodes [Output Animation Pose(Result=Sequence Player 'GrindRail_Boost_Loop'() {sequence=/GrindRail/Animation/Male/Boost/GrindRail_Boost_Loop})]
 }
 AnimationTransitionGraph_0 {
  path=MovementMode_FullBodyOverride.AnimGraphNode_StateMachine_0.FullBody.AnimStateNode_6.Boost.AnimGraphNode_StateMachine_1.BoostTransition.AnimStateTransitionNode_0.AnimationTransitionGraph_0
  nodes [Result()]
 }
 Transition {
  path=MovementMode_FullBodyOverride.AnimGraphNode_StateMachine_0.FullBody.AnimStateNode_6.Boost.AnimGraphNode_StateMachine_1.BoostTransition.AnimStateTransitionNode_2.Transition
  nodes [Result(bCanEnterTransition=true)]
 }
 Transition {
  path=MovementMode_FullBodyOverride.AnimGraphNode_StateMachine_0.FullBody.AnimStateNode_6.Boost.AnimGraphNode_StateMachine_1.BoostTransition.AnimStateTransitionNode_3.Transition
  nodes [Result(bCanEnterTransition=Get bStateRuleToBoostLoop() -> (bStateRuleToBoostLoop:bool))]
 }
 Entry {
  path=MovementMode_FullBodyOverride.AnimGraphNode_StateMachine_0.FullBody.AnimStateNode_7.Entry
  nodes [
   AnimGraphNode_StateMachine_0 = DirectionalEntry() {is=title AnimGraphNode_StateMachine, bAllowConduitEntryStates=true}
   Output Animation Pose(Result=Apply Additive(Base=@AnimGraphNode_StateMachine_0, Additive=Use cached pose 'RootLeanPose'()) {alphaBoolBlend {blendOption=Linear}})
  ]
 }
 DirectionalEntry {
  entry=EntrySelector
  path=MovementMode_FullBodyOverride.AnimGraphNode_StateMachine_0.FullBody.AnimStateNode_7.Entry.AnimGraphNode_StateMachine_0.DirectionalEntry
  ids=id
  nodes [
   AnimStateConduitNode_0 = EntrySelector() {is=title AnimStateConduitNode}
   AnimStateNode_0 = GroundEntryLeft() {is=title AnimStateNode}
   AnimStateNode_1 = GroundEntryRight() {is=title AnimStateNode}
   AnimStateNode_2 = JumpEntry() {is=title AnimStateNode}
   AnimStateTransitionNode_0 = AnimStateTransitionNode()
   AnimStateTransitionNode_1 = AnimStateTransitionNode() {priorityOrder=2}
   AnimStateTransitionNode_2 = AnimStateTransitionNode() {priorityOrder=100}
  ]
  links [
   {
    dst=GroundEntryLeft
    rule=MovementMode_FullBodyOverride.AnimGraphNode_StateMachine_0.FullBody.AnimStateNode_7.Entry.AnimGraphNode_StateMachine_0.DirectionalEntry.AnimStateTransitionNode_0.Transition
    src=EntrySelector
    via=AnimStateTransitionNode_0
   }
   {
    dst=GroundEntryRight
    rule=MovementMode_FullBodyOverride.AnimGraphNode_StateMachine_0.FullBody.AnimStateNode_7.Entry.AnimGraphNode_StateMachine_0.DirectionalEntry.AnimStateTransitionNode_1.Transition
    src=EntrySelector
    via=AnimStateTransitionNode_1
   }
   {
    dst=JumpEntry
    rule=MovementMode_FullBodyOverride.AnimGraphNode_StateMachine_0.FullBody.AnimStateNode_7.Entry.AnimGraphNode_StateMachine_0.DirectionalEntry.AnimStateTransitionNode_2.Transition
    src=EntrySelector
    via=AnimStateTransitionNode_2
   }
  ]
 }
 EntrySelector {
  path=MovementMode_FullBodyOverride.AnimGraphNode_StateMachine_0.FullBody.AnimStateNode_7.Entry.AnimGraphNode_StateMachine_0.DirectionalEntry.AnimStateConduitNode_0.EntrySelector
  nodes [Result(bCanEnterTransition=true)]
 }
 GroundEntryLeft {
  path=MovementMode_FullBodyOverride.AnimGraphNode_StateMachine_0.FullBody.AnimStateNode_7.Entry.AnimGraphNode_StateMachine_0.DirectionalEntry.AnimStateNode_0.GroundEntryLeft
  nodes [Output Animation Pose(Result=Sequence Player 'GrindRail_Ground_Mount_L'() {bLoopAnimation=false, sequence=/GrindRail/Animation/Male/Base/GrindRail_Ground_Mount_L})]
 }
 GroundEntryRight {
  path=MovementMode_FullBodyOverride.AnimGraphNode_StateMachine_0.FullBody.AnimStateNode_7.Entry.AnimGraphNode_StateMachine_0.DirectionalEntry.AnimStateNode_1.GroundEntryRight
  nodes [Output Animation Pose(Result=Sequence Player 'GrindRail_Ground_Mount_R'() {bLoopAnimation=false, sequence=/GrindRail/Animation/Male/Base/GrindRail_Ground_Mount_R})]
 }
 JumpEntry {
  path=MovementMode_FullBodyOverride.AnimGraphNode_StateMachine_0.FullBody.AnimStateNode_7.Entry.AnimGraphNode_StateMachine_0.DirectionalEntry.AnimStateNode_2.JumpEntry
  nodes [Output Animation Pose(Result=Sequence Player 'GrindRail_Mount_Land'() {bLoopAnimation=false, sequence=/GrindRail/Animation/Male/Base/GrindRail_Mount_Land})]
 }
 Transition {
  path=MovementMode_FullBodyOverride.AnimGraphNode_StateMachine_0.FullBody.AnimStateNode_7.Entry.AnimGraphNode_StateMachine_0.DirectionalEntry.AnimStateTransitionNode_0.Transition
  nodes [Result(bCanEnterTransition=Get bIsEntryLeft() -> (bIsEntryLeft:bool))]
 }
 Transition {
  path=MovementMode_FullBodyOverride.AnimGraphNode_StateMachine_0.FullBody.AnimStateNode_7.Entry.AnimGraphNode_StateMachine_0.DirectionalEntry.AnimStateTransitionNode_1.Transition
  nodes [Result(bCanEnterTransition=Get bIsEntryRight() -> (bIsEntryRight:bool))]
 }
 Transition {
  path=MovementMode_FullBodyOverride.AnimGraphNode_StateMachine_0.FullBody.AnimStateNode_7.Entry.AnimGraphNode_StateMachine_0.DirectionalEntry.AnimStateTransitionNode_2.Transition
  nodes [
   Result(bCanEnterTransition=true)
   Get bIsEntryFromAir() -> (bIsEntryFromAir:bool)
  ]
 }
 BoostOutro {
  path=MovementMode_FullBodyOverride.AnimGraphNode_StateMachine_0.FullBody.AnimStateNode_8.BoostOutro
  nodes [Output Animation Pose(Result=Apply Additive(
     Base=Sequence Player 'GrindRail_Boost_Outro'() {bLoopAnimation=false, sequence=/GrindRail/Animation/Male/Boost/GrindRail_Boost_Outro}
     Additive=Use cached pose 'BoostLeanPose'()
    ) {alphaBoolBlend {blendOption=Linear}})]
 }
 BoostTurn {
  path=MovementMode_FullBodyOverride.AnimGraphNode_StateMachine_0.FullBody.AnimStateNode_9.BoostTurn
  nodes [
   AnimGraphNode_StateMachine_0 = BoostTurnDirection() {is=title AnimGraphNode_StateMachine, bAllowConduitEntryStates=true}
   Output Animation Pose(Result=@AnimGraphNode_StateMachine_0)
  ]
 }
 BoostTurnDirection {
  entry=TurnSelector
  path=MovementMode_FullBodyOverride.AnimGraphNode_StateMachine_0.FullBody.AnimStateNode_9.BoostTurn.AnimGraphNode_StateMachine_0.BoostTurnDirection
  ids=id
  nodes [
   AnimStateConduitNode_0 = TurnSelector()
   AnimStateNode_0 = Boost180TurnLeft() {is=title AnimStateNode}
   AnimStateNode_1 = Boost180TurnRight() {is=title AnimStateNode}
   AnimStateTransitionNode_3 = AnimStateTransitionNode() {logicType=TLT_Inertialization}
   AnimStateTransitionNode_4 = AnimStateTransitionNode() {logicType=TLT_Inertialization, priorityOrder=100}
  ]
  links [
   {
    dst=Boost180TurnLeft
    rule=MovementMode_FullBodyOverride.AnimGraphNode_StateMachine_0.FullBody.AnimStateNode_9.BoostTurn.AnimGraphNode_StateMachine_0.BoostTurnDirection.AnimStateTransitionNode_3.Transition
    src=TurnSelector
    via=AnimStateTransitionNode_3
   }
   {
    dst=Boost180TurnRight
    rule=MovementMode_FullBodyOverride.AnimGraphNode_StateMachine_0.FullBody.AnimStateNode_9.BoostTurn.AnimGraphNode_StateMachine_0.BoostTurnDirection.AnimStateTransitionNode_4.Transition
    src=TurnSelector
    via=AnimStateTransitionNode_4
   }
  ]
 }
 TurnSelector {
  path=MovementMode_FullBodyOverride.AnimGraphNode_StateMachine_0.FullBody.AnimStateNode_9.BoostTurn.AnimGraphNode_StateMachine_0.BoostTurnDirection.AnimStateConduitNode_0.TurnSelector
  nodes [Result(bCanEnterTransition=true)]
 }
 Boost180TurnLeft {
  path=MovementMode_FullBodyOverride.AnimGraphNode_StateMachine_0.FullBody.AnimStateNode_9.BoostTurn.AnimGraphNode_StateMachine_0.BoostTurnDirection.AnimStateNode_0.Boost180TurnLeft
  nodes [Output Animation Pose(Result=Sequence Player 'GrindRail_Boost_180Turn_Left'() {bLoopAnimation=false, sequence=/GrindRail/Animation/Male/Boost/GrindRail_Boost_180Turn_Left})]
 }
 Boost180TurnRight {
  path=MovementMode_FullBodyOverride.AnimGraphNode_StateMachine_0.FullBody.AnimStateNode_9.BoostTurn.AnimGraphNode_StateMachine_0.BoostTurnDirection.AnimStateNode_1.Boost180TurnRight
  nodes [Output Animation Pose(Result=Sequence Player 'GrindRail_Boost_180Turn_Right'() {bLoopAnimation=false, sequence=/GrindRail/Animation/Male/Boost/GrindRail_Boost_180Turn_Right})]
 }
 Transition {
  path=MovementMode_FullBodyOverride.AnimGraphNode_StateMachine_0.FullBody.AnimStateNode_9.BoostTurn.AnimGraphNode_StateMachine_0.BoostTurnDirection.AnimStateTransitionNode_3.Transition
  nodes [Result(bCanEnterTransition=Get bIsLeft180Turn())]
 }
 Transition {
  path=MovementMode_FullBodyOverride.AnimGraphNode_StateMachine_0.FullBody.AnimStateNode_9.BoostTurn.AnimGraphNode_StateMachine_0.BoostTurnDirection.AnimStateTransitionNode_4.Transition
  nodes [Result(bCanEnterTransition=true)]
 }
 Transition {
  path=MovementMode_FullBodyOverride.AnimGraphNode_StateMachine_0.FullBody.AnimStateTransitionNode_0.Transition
  nodes [Result(bCanEnterTransition=Get bShouldExitLocomotion() -> (bShouldExitLocomotion:bool))]
 }
 Transition {
  path=MovementMode_FullBodyOverride.AnimGraphNode_StateMachine_0.FullBody.AnimStateTransitionNode_1.Transition
  nodes [Result(bCanEnterTransition=NOT Boolean(A=Get bIsFalling()))]
 }
 Transition {path=MovementMode_FullBodyOverride.AnimGraphNode_StateMachine_0.FullBody.AnimStateTransitionNode_10.Transition, nodes [Result()]}
 Transition {path=MovementMode_FullBodyOverride.AnimGraphNode_StateMachine_0.FullBody.AnimStateTransitionNode_11.Transition, nodes [Result()]}
 Transition {
  path=MovementMode_FullBodyOverride.AnimGraphNode_StateMachine_0.FullBody.AnimStateTransitionNode_12.Transition
  nodes [Result(bCanEnterTransition=Get bEnteredFromInteraction() -> (bEnteredFromInteraction:bool))]
 }
 Transition {path=MovementMode_FullBodyOverride.AnimGraphNode_StateMachine_0.FullBody.AnimStateTransitionNode_13.Transition, nodes [Result()]}
 Transition {
  path=MovementMode_FullBodyOverride.AnimGraphNode_StateMachine_0.FullBody.AnimStateTransitionNode_14.Transition
  nodes [Result(bCanEnterTransition=Get bIsTurning() -> (bIsTurning:bool))]
 }
 Transition {
  path=MovementMode_FullBodyOverride.AnimGraphNode_StateMachine_0.FullBody.AnimStateTransitionNode_15.Transition
  nodes [Result(bCanEnterTransition=Get bIsWeaponActive() -> (bIsWeaponActive:bool))]
 }
 Transition {
  path=MovementMode_FullBodyOverride.AnimGraphNode_StateMachine_0.FullBody.AnimStateTransitionNode_16.Transition
  nodes [
   Result(bCanEnterTransition=Get bIsMovingOrIsTurning() -> (bIsMovingOrIsTurning:bool))
   OR Boolean(A=Get bIsMoving() -> (bIsMoving:bool), B=Get bIsTurning())
  ]
 }
 Transition {
  path=MovementMode_FullBodyOverride.AnimGraphNode_StateMachine_0.FullBody.AnimStateTransitionNode_17.Transition
  nodes [Result(bCanEnterTransition=true)]
 }
 Transition {
  path=MovementMode_FullBodyOverride.AnimGraphNode_StateMachine_0.FullBody.AnimStateTransitionNode_18.Transition
  nodes [Result(bCanEnterTransition=NOT Boolean(A=Get bIsBoosting() -> (bIsBoosting:bool)))]
 }
 Transition {
  path=MovementMode_FullBodyOverride.AnimGraphNode_StateMachine_0.FullBody.AnimStateTransitionNode_2.Transition
  nodes [
   Result(bCanEnterTransition=Get bIsTurningOrNotIsLeanBwd() -> (bIsTurningOrNotIsLeanBwd:bool))
   OR Boolean(A=NOT Boolean(A=Get bIsLeanBwd() -> (bIsLeanBwd:bool)), B=Get bIsTurning())
  ]
 }
 Transition {
  path=MovementMode_FullBodyOverride.AnimGraphNode_StateMachine_0.FullBody.AnimStateTransitionNode_22.Transition
  nodes [Result(bCanEnterTransition=Get bIsTurning())]
 }
 Transition {
  path=MovementMode_FullBodyOverride.AnimGraphNode_StateMachine_0.FullBody.AnimStateTransitionNode_23.Transition
  nodes [Result(bCanEnterTransition=true)]
 }
 Transition {
  path=MovementMode_FullBodyOverride.AnimGraphNode_StateMachine_0.FullBody.AnimStateTransitionNode_24.Transition
  nodes [Result(bCanEnterTransition=Get bShouldPlayEntry() -> (bShouldPlayEntry:bool))]
 }
 Transition {path=MovementMode_FullBodyOverride.AnimGraphNode_StateMachine_0.FullBody.AnimStateTransitionNode_25.Transition, nodes [Result()]}
 Transition {path=MovementMode_FullBodyOverride.AnimGraphNode_StateMachine_0.FullBody.AnimStateTransitionNode_26.Transition, nodes [Result()]}
 Transition {path=MovementMode_FullBodyOverride.AnimGraphNode_StateMachine_0.FullBody.AnimStateTransitionNode_3.Transition, nodes [Result()]}
 Transition {
  path=MovementMode_FullBodyOverride.AnimGraphNode_StateMachine_0.FullBody.AnimStateTransitionNode_36.Transition
  nodes [Result(bCanEnterTransition=Get bShould180TurnAgain() -> (bShould180TurnAgain:bool))]
 }
 Transition {
  path=MovementMode_FullBodyOverride.AnimGraphNode_StateMachine_0.FullBody.AnimStateTransitionNode_4.Transition
  nodes [Result(bCanEnterTransition=Get bIsLeanBwd())]
 }
 Transition {
  path=MovementMode_FullBodyOverride.AnimGraphNode_StateMachine_0.FullBody.AnimStateTransitionNode_5.Transition
  nodes [Result(bCanEnterTransition=Get bIsTurning())]
 }
 Transition {
  path=MovementMode_FullBodyOverride.AnimGraphNode_StateMachine_0.FullBody.AnimStateTransitionNode_6.Transition
  nodes [
   Result(bCanEnterTransition=Get bIsFallingAndNotEnteredFromInteraction() -> (bIsFallingAndNotEnteredFromInteraction:bool))
   AND Boolean(A=Get bIsFalling(), B=NOT Boolean(A=Get bEnteredFromInteraction()))
  ]
 }
 Transition {
  path=MovementMode_FullBodyOverride.AnimGraphNode_StateMachine_0.FullBody.AnimStateTransitionNode_7.Transition
  nodes [Result(bCanEnterTransition=Get bIsBoosting())]
 }
 Transition {
  path=MovementMode_FullBodyOverride.AnimGraphNode_StateMachine_0.FullBody.AnimStateTransitionNode_8.Transition
  nodes [Result(bCanEnterTransition=Get bEarlyExitFromEntry() -> (bEarlyExitFromEntry:bool))]
 }
 Transition {
  path=MovementMode_FullBodyOverride.AnimGraphNode_StateMachine_0.FullBody.AnimStateTransitionNode_9.Transition
  nodes [Result(bCanEnterTransition=NOT Boolean(A=Get bIsBoosting()))]
 }
 CombatMode {
  entry=CachedFullBody
  path=MovementMode_FullBodyOverride.AnimGraphNode_StateMachine_1.CombatMode
  ids=id
  nodes [
   AnimStateNode_0 = CachedFullBody() {is=title AnimStateNode}
   AnimStateNode_1 = Combat() {is=title AnimStateNode}
   AnimStateNode_2 = TransitionToRiding() {is=title AnimStateNode}
   AnimStateTransitionNode_0 = AnimStateTransitionNode() {logicType=TLT_Inertialization}
   AnimStateTransitionNode_1 = AnimStateTransitionNode() {crossfadeDuration=0.1, logicType=TLT_Inertialization, priorityOrder=2}
   AnimStateTransitionNode_2 = AnimStateTransitionNode() {bAutomaticRuleBasedOnSequencePlayerInState=true, crossfadeDuration=0.4, priorityOrder=3}
   AnimStateTransitionNode_3 = AnimStateTransitionNode()
   AnimStateTransitionNode_33 = AnimStateTransitionNode() {priorityOrder=2}
   AnimStateTransitionNode_4 = AnimStateTransitionNode() {priorityOrder=2}
   AnimStateTransitionNode_5 = AnimStateTransitionNode()
   AnimStateTransitionNode_6 = AnimStateTransitionNode() {crossfadeDuration=0.1, logicType=TLT_Inertialization}
   AnimStateTransitionNode_7 = AnimStateTransitionNode() {crossfadeDuration=0.1, logicType=TLT_Inertialization}
  ]
  links [
   {
    dst=Combat
    rule=MovementMode_FullBodyOverride.AnimGraphNode_StateMachine_1.CombatMode.AnimStateTransitionNode_0.Transition
    src=CachedFullBody
    via=AnimStateTransitionNode_0
   }
   {
    dst=CachedFullBody
    rule=MovementMode_FullBodyOverride.AnimGraphNode_StateMachine_1.CombatMode.AnimStateTransitionNode_3.Transition
    src=Combat
    via=AnimStateTransitionNode_3
   }
   {
    dst=CachedFullBody
    rule=MovementMode_FullBodyOverride.AnimGraphNode_StateMachine_1.CombatMode.AnimStateTransitionNode_33.Transition
    src=Combat
    via=AnimStateTransitionNode_33
   }
   {
    dst=CachedFullBody
    rule=MovementMode_FullBodyOverride.AnimGraphNode_StateMachine_1.CombatMode.AnimStateTransitionNode_7.Transition
    src=Combat
    via=AnimStateTransitionNode_7
   }
   {
    dst=TransitionToRiding
    rule=MovementMode_FullBodyOverride.AnimGraphNode_StateMachine_1.CombatMode.AnimStateTransitionNode_1.Transition
    src=Combat
    via=AnimStateTransitionNode_1
   }
   {
    dst=CachedFullBody
    rule=MovementMode_FullBodyOverride.AnimGraphNode_StateMachine_1.CombatMode.AnimStateTransitionNode_2.Transition
    src=TransitionToRiding
    via=AnimStateTransitionNode_2
   }
   {
    dst=CachedFullBody
    rule=MovementMode_FullBodyOverride.AnimGraphNode_StateMachine_1.CombatMode.AnimStateTransitionNode_4.Transition
    src=TransitionToRiding
    via=AnimStateTransitionNode_4
   }
   {
    dst=CachedFullBody
    rule=MovementMode_FullBodyOverride.AnimGraphNode_StateMachine_1.CombatMode.AnimStateTransitionNode_5.Transition
    src=TransitionToRiding
    via=AnimStateTransitionNode_5
   }
   {
    dst=CachedFullBody
    rule=MovementMode_FullBodyOverride.AnimGraphNode_StateMachine_1.CombatMode.AnimStateTransitionNode_6.Transition
    src=TransitionToRiding
    via=AnimStateTransitionNode_6
   }
  ]
 }
 CachedFullBody {
  path=MovementMode_FullBodyOverride.AnimGraphNode_StateMachine_1.CombatMode.AnimStateNode_0.CachedFullBody
  nodes [Output Animation Pose(Result=Modify Curve!(SourcePose=Use cached pose 'FullBodyPose'(), CurveValues_0:float=1, Alpha=1) -> (Pose) {curveNames [BothHandFK], curveValues [0]})]
 }
 Combat {
  path=MovementMode_FullBodyOverride.AnimGraphNode_StateMachine_1.CombatMode.AnimStateNode_1.Combat
  nodes [
   n1 = "Sequence Player (additive) 'GrindRail_Combat_MeleeNoWeapOverrideBody_Pose'"() {sequence=/GrindRail/Animation/Male/Combat/GrindRail_Combat_MeleeNoWeapOverrideBody_Pose}
   n2 = "Sequence Player (additive) 'GrindRail_Combat_MeleeNoWeapOverried_Pose'"() {sequence=/GrindRail/Animation/Male/Combat/GrindRail_Combat_MeleeNoWeapOverried_Pose}
   n3 = Get bIsInAction() -> (bIsInAction:bool)
   AnimGraphNode_StateMachine_0 = CombatTransition() {is=title AnimGraphNode_StateMachine}
   n4 = Layered blend per bone(
    BasePose=Apply Mesh Space Additive!(
     Base:Pose Link Structure=@AnimGraphNode_StateMachine_0
     Additive:Pose Link Structure=Blend Poses by bool(BlendPose_0=@n1, BlendPose_1=@n2, BlendTime_0=0.15, BlendTime_1=0.5, bActiveValue=@n3) {blendType=Linear, transitionType=Inertialization}
     bAlphaBoolEnabled:bool=Get bShouldCorrectUpperBody() -> (bShouldCorrectUpperBody:bool)
    ) -> (Pose:Pose Link Structure) {is=title AnimGraphNode_ApplyMeshSpaceAdditive, alphaInputType=Bool, alphaBoolBlend {blendInTime=0.1, blendOption=Linear, blendOutTime=0.1}}
    BlendPoses_0=Local Space Ref Pose()
    BlendWeights_0=Get MuteUpperBodyAlpha() -> (MuteUpperBodyAlpha:double)
   ) {
    bMeshSpaceRotationBlend=true
    bMeshSpaceScaleBlend=true
    layerSetup [{branchFilters [{boneName=Spine_01}, {blendDepth=-1, boneName=clavicle_l}, {blendDepth=-1, boneName=clavicle_r}]}]
   }
   n5 = "Transform (Modify) Bone - Bone: pelvis"!(
    ComponentPose=Reset Root Transform!(ComponentPose:Component Space Pose Link Structure=Local To Component(LocalPose=@n4), Alpha:float=1) -> (Pose:Component Space Pose Link Structure) {is=title AnimGraphNode_ResetRoot, alphaBoolBlend {blendOption=Linear}}
    Alpha=0.25
    Rotation="0,0,0"
   ) -> (Pose) {
    rotationMode=BMM_Additive
    rotationSpace=BCS_WorldSpace
    translationSpace=BCS_WorldSpace
    alphaBoolBlend {blendOption=Linear}
    boneToModify {boneName=pelvis}
   }
   Output Animation Pose(Result=Apply Additive(
     Base=Component To Local(ComponentPose=@n5)
     Additive=Blendspace Player 'GrindRail_Combat_Leans_BS'(X=0, Y=0) {blendSpace=/GrindRail/Animation/Male/Combat/_BlendSpace/GrindRail_Combat_Leans_BS}
    ) {alphaBoolBlend {blendOption=Linear}})
  ]
 }
 CombatTransition {
  entry=Combat_Loop
  path=MovementMode_FullBodyOverride.AnimGraphNode_StateMachine_1.CombatMode.AnimStateNode_1.Combat.AnimGraphNode_StateMachine_0.CombatTransition
  ids=id
  nodes [AnimStateNode_1 = Combat_Loop() {is=title AnimStateNode}]
 }
 Combat_Loop {
  path=MovementMode_FullBodyOverride.AnimGraphNode_StateMachine_1.CombatMode.AnimStateNode_1.Combat.AnimGraphNode_StateMachine_0.CombatTransition.AnimStateNode_1.Combat_Loop
  nodes [
   AnimGraphNode_StateMachine_2 = CombatUpperBodyAdditive() {is=title AnimGraphNode_StateMachine}
   Output Animation Pose(Result=Apply Additive(
     Base=Use cached pose 'CachedMainPlayerPose'()
     Additive=@AnimGraphNode_StateMachine_2
     Alpha=Get CombatNoiseAlpha() -> (CombatNoiseAlpha:double)
    ) {alphaBoolBlend {blendOption=Linear}})
  ]
 }
 CombatUpperBodyAdditive {
  entry=IdleAdd
  path=MovementMode_FullBodyOverride.AnimGraphNode_StateMachine_1.CombatMode.AnimStateNode_1.Combat.AnimGraphNode_StateMachine_0.CombatTransition.AnimStateNode_1.Combat_Loop.AnimGraphNode_StateMachine_2.CombatUpperBodyAdditive
  ids=id
  nodes [
   AnimStateNode_0 = IdleAdd() {is=title AnimStateNode}
   AnimStateNode_1 = RidingAdd() {is=title AnimStateNode}
   AnimStateTransitionNode_0 = AnimStateTransitionNode()
   AnimStateTransitionNode_1 = AnimStateTransitionNode()
  ]
  links [
   {
    dst=RidingAdd
    rule=MovementMode_FullBodyOverride.AnimGraphNode_StateMachine_1.CombatMode.AnimStateNode_1.Combat.AnimGraphNode_StateMachine_0.CombatTransition.AnimStateNode_1.Combat_Loop.AnimGraphNode_StateMachine_2.CombatUpperBodyAdditive.AnimStateTransitionNode_0.Transition
    src=IdleAdd
    via=AnimStateTransitionNode_0
   }
   {
    dst=IdleAdd
    rule=MovementMode_FullBodyOverride.AnimGraphNode_StateMachine_1.CombatMode.AnimStateNode_1.Combat.AnimGraphNode_StateMachine_0.CombatTransition.AnimStateNode_1.Combat_Loop.AnimGraphNode_StateMachine_2.CombatUpperBodyAdditive.AnimStateTransitionNode_1.Transition
    src=RidingAdd
    via=AnimStateTransitionNode_1
   }
  ]
 }
 IdleAdd {
  path=MovementMode_FullBodyOverride.AnimGraphNode_StateMachine_1.CombatMode.AnimStateNode_1.Combat.AnimGraphNode_StateMachine_0.CombatTransition.AnimStateNode_1.Combat_Loop.AnimGraphNode_StateMachine_2.CombatUpperBodyAdditive.AnimStateNode_0.IdleAdd
  nodes [Output Animation Pose(Result=Additive Identity Pose())]
 }
 RidingAdd {
  path=MovementMode_FullBodyOverride.AnimGraphNode_StateMachine_1.CombatMode.AnimStateNode_1.Combat.AnimGraphNode_StateMachine_0.CombatTransition.AnimStateNode_1.Combat_Loop.AnimGraphNode_StateMachine_2.CombatUpperBodyAdditive.AnimStateNode_1.RidingAdd
  nodes [Output Animation Pose(Result="Sequence Player (additive) 'GrindRail_Combat_Idle_Noise'"!(PlayRate:float=Get CombatNoisePlayRate() -> (CombatNoisePlayRate:double)) -> (Pose) {sequence=/GrindRail/Animation/Male/Combat/GrindRail_Combat_Idle_Noise})]
 }
 Transition {
  path=MovementMode_FullBodyOverride.AnimGraphNode_StateMachine_1.CombatMode.AnimStateNode_1.Combat.AnimGraphNode_StateMachine_0.CombatTransition.AnimStateNode_1.Combat_Loop.AnimGraphNode_StateMachine_2.CombatUpperBodyAdditive.AnimStateTransitionNode_0.Transition
  nodes [Result(bCanEnterTransition=Get bIsMoving())]
 }
 Transition {
  path=MovementMode_FullBodyOverride.AnimGraphNode_StateMachine_1.CombatMode.AnimStateNode_1.Combat.AnimGraphNode_StateMachine_0.CombatTransition.AnimStateNode_1.Combat_Loop.AnimGraphNode_StateMachine_2.CombatUpperBodyAdditive.AnimStateTransitionNode_1.Transition
  nodes [Result(bCanEnterTransition=NOT Boolean(A=Get bIsMoving()))]
 }
 TransitionToRiding {
  path=MovementMode_FullBodyOverride.AnimGraphNode_StateMachine_1.CombatMode.AnimStateNode_2.TransitionToRiding
  nodes [Output Animation Pose(Result=Apply Additive(
     Base=Sequence Player 'GrindRail_Combat_to_Riding_Transistion'() {bLoopAnimation=false, sequence=/GrindRail/Animation/Male/Combat/GrindRail_Combat_to_Riding_Transistion}
     Additive=Use cached pose 'RootLeanPose'()
    ) {alphaBoolBlend {blendOption=Linear}})]
 }
 Transition {
  path=MovementMode_FullBodyOverride.AnimGraphNode_StateMachine_1.CombatMode.AnimStateTransitionNode_0.Transition
  nodes [Result(bCanEnterTransition=Get bIsWeaponActive())]
 }
 Transition {
  path=MovementMode_FullBodyOverride.AnimGraphNode_StateMachine_1.CombatMode.AnimStateTransitionNode_1.Transition
  nodes [
   Result(bCanEnterTransition=Get bNotIsWeaponActiveAndNotIsFalling() -> (bNotIsWeaponActiveAndNotIsFalling:bool))
   AND Boolean(A=NOT Boolean(A=Get bIsWeaponActive()), B=NOT Boolean(A=Get bIsFalling()))
  ]
 }
 Transition {path=MovementMode_FullBodyOverride.AnimGraphNode_StateMachine_1.CombatMode.AnimStateTransitionNode_2.Transition, nodes [Result()]}
 Transition {
  path=MovementMode_FullBodyOverride.AnimGraphNode_StateMachine_1.CombatMode.AnimStateTransitionNode_3.Transition
  nodes [Result(bCanEnterTransition=Get bEnteredFromInteraction())]
 }
 Transition {
  path=MovementMode_FullBodyOverride.AnimGraphNode_StateMachine_1.CombatMode.AnimStateTransitionNode_33.Transition
  nodes [
   Result(bCanEnterTransition=Get bIsFallingAndNotIsWeaponActive() -> (bIsFallingAndNotIsWeaponActive:bool))
   AND Boolean(A=NOT Boolean(A=Get bIsWeaponActive()), B=Get bIsFalling())
  ]
 }
 Transition {
  path=MovementMode_FullBodyOverride.AnimGraphNode_StateMachine_1.CombatMode.AnimStateTransitionNode_4.Transition
  nodes [Result(bCanEnterTransition=Get bIsFalling())]
 }
 Transition {
  path=MovementMode_FullBodyOverride.AnimGraphNode_StateMachine_1.CombatMode.AnimStateTransitionNode_5.Transition
  nodes [Result(bCanEnterTransition=Get bEnteredFromInteraction())]
 }
 Transition {
  path=MovementMode_FullBodyOverride.AnimGraphNode_StateMachine_1.CombatMode.AnimStateTransitionNode_6.Transition
  nodes [
   Result(bCanEnterTransition=Get bIsWeaponActiveOrIsBoosting() -> (bIsWeaponActiveOrIsBoosting:bool))
   OR Boolean(A=Get bIsWeaponActive(), B=Get bIsBoosting())
  ]
 }
 Transition {
  path=MovementMode_FullBodyOverride.AnimGraphNode_StateMachine_1.CombatMode.AnimStateTransitionNode_7.Transition
  nodes [Result(bCanEnterTransition=Get bIsBoosting())]
 }
 MovementMode_InAirOverride {nodes [
   Input Pose() {name=InSource_Pose}
   Input Pose() {name=InAirPose}
   Output Pose() {layerGroup=MovementMode}
  ]}
 MovementMode_LocomotionAdditives {nodes [
   Input Pose() {name=InLocomotionAdditivesPose}
   Output Pose() {layerGroup=MovementMode}
  ]}
 MovementMode_LowerBodyOverride {nodes [
   Input Pose() {name=InSourcePose}
   AnimGraphNode_StateMachine_0 = LowerBodyTurn() {is=title AnimGraphNode_StateMachine}
   Output Pose(Result=Blend Poses by bool(
     BlendPose_0=Rotate Root Bone!(BasePose:Pose Link Structure=@AnimGraphNode_StateMachine_0, Yaw:float=0, MeshToComponent:Rotator="0,0,0") -> (Pose:Pose Link Structure) {is=title AnimGraphNode_RotateRootBone}
     BlendPose_1=Input Pose() {name=InLowerBodyPose}
    ) {blendType=Linear, transitionType=Inertialization}) {layerGroup=MovementMode}
  ]}
 LowerBodyTurn {
  entry=LowerBodyPose_N
  path=MovementMode_LowerBodyOverride.AnimGraphNode_StateMachine_0.LowerBodyTurn
  ids=id
  nodes [
   AnimStateNode_10 = Transition_EtoS() {is=title AnimStateNode}
   AnimStateNode_11 = LowerBodyPose_S() {is=title AnimStateNode}
   AnimStateNode_12 = Transition_StoW() {is=title AnimStateNode}
   AnimStateNode_13 = LowerBodyPose_W() {is=title AnimStateNode}
   AnimStateNode_15 = Transition_WtoN() {is=title AnimStateNode}
   AnimStateNode_16 = LowerBodyPose_N() {is=title AnimStateNode}
   AnimStateNode_17 = Transition_EtoN() {is=title AnimStateNode}
   AnimStateNode_18 = Transition_StoE() {is=title AnimStateNode}
   AnimStateNode_19 = Transition_WtoS() {is=title AnimStateNode}
   AnimStateNode_20 = Transition_NtoW() {is=title AnimStateNode}
   AnimStateNode_8 = Transition_NtoE() {is=title AnimStateNode}
   AnimStateNode_9 = LowerBodyPose_E() {is=title AnimStateNode}
   AnimStateTransitionNode_0 = AnimStateTransitionNode() {logicType=TLT_Inertialization}
   AnimStateTransitionNode_1 = AnimStateTransitionNode() {bAutomaticRuleBasedOnSequencePlayerInState=true, logicType=TLT_Inertialization}
   AnimStateTransitionNode_10 = AnimStateTransitionNode() {logicType=TLT_Inertialization, priorityOrder=2}
   AnimStateTransitionNode_11 = AnimStateTransitionNode() {logicType=TLT_Inertialization, priorityOrder=2}
   AnimStateTransitionNode_2 = AnimStateTransitionNode() {bAutomaticRuleBasedOnSequencePlayerInState=true, logicType=TLT_Inertialization}
   AnimStateTransitionNode_25 = AnimStateTransitionNode() {logicType=TLT_Inertialization}
   AnimStateTransitionNode_26 = AnimStateTransitionNode() {bAutomaticRuleBasedOnSequencePlayerInState=true, logicType=TLT_Inertialization}
   AnimStateTransitionNode_27 = AnimStateTransitionNode() {logicType=TLT_Inertialization}
   AnimStateTransitionNode_28 = AnimStateTransitionNode() {logicType=TLT_Inertialization}
   AnimStateTransitionNode_29 = AnimStateTransitionNode() {logicType=TLT_Inertialization}
   AnimStateTransitionNode_3 = AnimStateTransitionNode() {bAutomaticRuleBasedOnSequencePlayerInState=true, logicType=TLT_Inertialization}
   AnimStateTransitionNode_30 = AnimStateTransitionNode() {logicType=TLT_Inertialization}
   AnimStateTransitionNode_31 = AnimStateTransitionNode() {logicType=TLT_Inertialization}
   AnimStateTransitionNode_32 = AnimStateTransitionNode() {logicType=TLT_Inertialization}
   AnimStateTransitionNode_35 = AnimStateTransitionNode() {logicType=TLT_Inertialization}
   AnimStateTransitionNode_36 = AnimStateTransitionNode() {logicType=TLT_Inertialization}
   AnimStateTransitionNode_37 = AnimStateTransitionNode() {logicType=TLT_Inertialization}
   AnimStateTransitionNode_38 = AnimStateTransitionNode() {bAutomaticRuleBasedOnSequencePlayerInState=true, logicType=TLT_Inertialization}
   AnimStateTransitionNode_39 = AnimStateTransitionNode() {logicType=TLT_Inertialization}
   AnimStateTransitionNode_4 = AnimStateTransitionNode() {logicType=TLT_Inertialization}
   AnimStateTransitionNode_40 = AnimStateTransitionNode() {bAutomaticRuleBasedOnSequencePlayerInState=true, logicType=TLT_Inertialization}
   AnimStateTransitionNode_41 = AnimStateTransitionNode() {logicType=TLT_Inertialization}
   AnimStateTransitionNode_42 = AnimStateTransitionNode() {bAutomaticRuleBasedOnSequencePlayerInState=true, logicType=TLT_Inertialization}
   AnimStateTransitionNode_5 = AnimStateTransitionNode() {logicType=TLT_Inertialization}
   AnimStateTransitionNode_6 = AnimStateTransitionNode() {logicType=TLT_Inertialization}
   AnimStateTransitionNode_7 = AnimStateTransitionNode() {bAutomaticRuleBasedOnSequencePlayerInState=true, logicType=TLT_Inertialization}
   AnimStateTransitionNode_8 = AnimStateTransitionNode() {logicType=TLT_Inertialization, priorityOrder=2}
   AnimStateTransitionNode_9 = AnimStateTransitionNode() {logicType=TLT_Inertialization, priorityOrder=2}
  ]
  links [
   {
    dst=LowerBodyPose_W
    rule=MovementMode_LowerBodyOverride.AnimGraphNode_StateMachine_0.LowerBodyTurn.AnimStateTransitionNode_11.AnimationTransitionGraph_0
    src=LowerBodyPose_E
    via=AnimStateTransitionNode_11
   }
   {
    dst=Transition_EtoN
    rule=MovementMode_LowerBodyOverride.AnimGraphNode_StateMachine_0.LowerBodyTurn.AnimStateTransitionNode_35.AnimationTransitionGraph_0
    src=LowerBodyPose_E
    via=AnimStateTransitionNode_35
   }
   {
    dst=Transition_EtoS
    rule=MovementMode_LowerBodyOverride.AnimGraphNode_StateMachine_0.LowerBodyTurn.AnimStateTransitionNode_27.AnimationTransitionGraph_1
    src=LowerBodyPose_E
    via=AnimStateTransitionNode_27
   }
   {
    dst=LowerBodyPose_S
    rule=MovementMode_LowerBodyOverride.AnimGraphNode_StateMachine_0.LowerBodyTurn.AnimStateTransitionNode_8.AnimationTransitionGraph_1
    src=LowerBodyPose_N
    via=AnimStateTransitionNode_8
   }
   {
    dst=Transition_NtoE
    rule=MovementMode_LowerBodyOverride.AnimGraphNode_StateMachine_0.LowerBodyTurn.AnimStateTransitionNode_25.AnimationTransitionGraph_0
    src=LowerBodyPose_N
    via=AnimStateTransitionNode_25
   }
   {
    dst=Transition_NtoW
    rule=MovementMode_LowerBodyOverride.AnimGraphNode_StateMachine_0.LowerBodyTurn.AnimStateTransitionNode_41.AnimationTransitionGraph_0
    src=LowerBodyPose_N
    via=AnimStateTransitionNode_41
   }
   {
    dst=LowerBodyPose_N
    rule=MovementMode_LowerBodyOverride.AnimGraphNode_StateMachine_0.LowerBodyTurn.AnimStateTransitionNode_9.AnimationTransitionGraph_1
    src=LowerBodyPose_S
    via=AnimStateTransitionNode_9
   }
   {
    dst=Transition_StoE
    rule=MovementMode_LowerBodyOverride.AnimGraphNode_StateMachine_0.LowerBodyTurn.AnimStateTransitionNode_37.AnimationTransitionGraph_0
    src=LowerBodyPose_S
    via=AnimStateTransitionNode_37
   }
   {
    dst=Transition_StoW
    rule=MovementMode_LowerBodyOverride.AnimGraphNode_StateMachine_0.LowerBodyTurn.AnimStateTransitionNode_29.AnimationTransitionGraph_1
    src=LowerBodyPose_S
    via=AnimStateTransitionNode_29
   }
   {
    dst=LowerBodyPose_E
    rule=MovementMode_LowerBodyOverride.AnimGraphNode_StateMachine_0.LowerBodyTurn.AnimStateTransitionNode_10.AnimationTransitionGraph_0
    src=LowerBodyPose_W
    via=AnimStateTransitionNode_10
   }
   {
    dst=Transition_WtoN
    rule=MovementMode_LowerBodyOverride.AnimGraphNode_StateMachine_0.LowerBodyTurn.AnimStateTransitionNode_31.AnimationTransitionGraph_0
    src=LowerBodyPose_W
    via=AnimStateTransitionNode_31
   }
   {
    dst=Transition_WtoS
    rule=MovementMode_LowerBodyOverride.AnimGraphNode_StateMachine_0.LowerBodyTurn.AnimStateTransitionNode_39.AnimationTransitionGraph_1
    src=LowerBodyPose_W
    via=AnimStateTransitionNode_39
   }
   {
    dst=LowerBodyPose_N
    rule=MovementMode_LowerBodyOverride.AnimGraphNode_StateMachine_0.LowerBodyTurn.AnimStateTransitionNode_36.AnimationTransitionGraph_0
    src=Transition_EtoN
    via=AnimStateTransitionNode_36
   }
   {
    dst=LowerBodyPose_N
    rule=MovementMode_LowerBodyOverride.AnimGraphNode_StateMachine_0.LowerBodyTurn.AnimStateTransitionNode_7.AnimationTransitionGraph_0
    src=Transition_EtoN
    via=AnimStateTransitionNode_7
   }
   {
    dst=LowerBodyPose_S
    rule=MovementMode_LowerBodyOverride.AnimGraphNode_StateMachine_0.LowerBodyTurn.AnimStateTransitionNode_2.AnimationTransitionGraph_1
    src=Transition_EtoS
    via=AnimStateTransitionNode_2
   }
   {
    dst=LowerBodyPose_S
    rule=MovementMode_LowerBodyOverride.AnimGraphNode_StateMachine_0.LowerBodyTurn.AnimStateTransitionNode_28.AnimationTransitionGraph_1
    src=Transition_EtoS
    via=AnimStateTransitionNode_28
   }
   {
    dst=LowerBodyPose_E
    rule=MovementMode_LowerBodyOverride.AnimGraphNode_StateMachine_0.LowerBodyTurn.AnimStateTransitionNode_0.AnimationTransitionGraph_0
    src=Transition_NtoE
    via=AnimStateTransitionNode_0
   }
   {
    dst=LowerBodyPose_E
    rule=MovementMode_LowerBodyOverride.AnimGraphNode_StateMachine_0.LowerBodyTurn.AnimStateTransitionNode_26.AnimationTransitionGraph_0
    src=Transition_NtoE
    via=AnimStateTransitionNode_26
   }
   {
    dst=LowerBodyPose_W
    rule=MovementMode_LowerBodyOverride.AnimGraphNode_StateMachine_0.LowerBodyTurn.AnimStateTransitionNode_4.AnimationTransitionGraph_0
    src=Transition_NtoW
    via=AnimStateTransitionNode_4
   }
   {
    dst=LowerBodyPose_W
    rule=MovementMode_LowerBodyOverride.AnimGraphNode_StateMachine_0.LowerBodyTurn.AnimStateTransitionNode_42.AnimationTransitionGraph_0
    src=Transition_NtoW
    via=AnimStateTransitionNode_42
   }
   {
    dst=LowerBodyPose_E
    rule=MovementMode_LowerBodyOverride.AnimGraphNode_StateMachine_0.LowerBodyTurn.AnimStateTransitionNode_38.AnimationTransitionGraph_1
    src=Transition_StoE
    via=AnimStateTransitionNode_38
   }
   {
    dst=LowerBodyPose_E
    rule=MovementMode_LowerBodyOverride.AnimGraphNode_StateMachine_0.LowerBodyTurn.AnimStateTransitionNode_6.AnimationTransitionGraph_1
    src=Transition_StoE
    via=AnimStateTransitionNode_6
   }
   {
    dst=LowerBodyPose_W
    rule=MovementMode_LowerBodyOverride.AnimGraphNode_StateMachine_0.LowerBodyTurn.AnimStateTransitionNode_1.AnimationTransitionGraph_1
    src=Transition_StoW
    via=AnimStateTransitionNode_1
   }
   {
    dst=LowerBodyPose_W
    rule=MovementMode_LowerBodyOverride.AnimGraphNode_StateMachine_0.LowerBodyTurn.AnimStateTransitionNode_30.AnimationTransitionGraph_1
    src=Transition_StoW
    via=AnimStateTransitionNode_30
   }
   {
    dst=LowerBodyPose_N
    rule=MovementMode_LowerBodyOverride.AnimGraphNode_StateMachine_0.LowerBodyTurn.AnimStateTransitionNode_3.AnimationTransitionGraph_0
    src=Transition_WtoN
    via=AnimStateTransitionNode_3
   }
   {
    dst=LowerBodyPose_N
    rule=MovementMode_LowerBodyOverride.AnimGraphNode_StateMachine_0.LowerBodyTurn.AnimStateTransitionNode_32.AnimationTransitionGraph_0
    src=Transition_WtoN
    via=AnimStateTransitionNode_32
   }
   {
    dst=LowerBodyPose_S
    rule=MovementMode_LowerBodyOverride.AnimGraphNode_StateMachine_0.LowerBodyTurn.AnimStateTransitionNode_40.AnimationTransitionGraph_0
    src=Transition_WtoS
    via=AnimStateTransitionNode_40
   }
   {
    dst=LowerBodyPose_S
    rule=MovementMode_LowerBodyOverride.AnimGraphNode_StateMachine_0.LowerBodyTurn.AnimStateTransitionNode_5.AnimationTransitionGraph_0
    src=Transition_WtoS
    via=AnimStateTransitionNode_5
   }
  ]
 }
 Transition_EtoS {
  path=MovementMode_LowerBodyOverride.AnimGraphNode_StateMachine_0.LowerBodyTurn.AnimStateNode_10.Transition_EtoS
  nodes [Output Animation Pose(Result=Sequence Player 'GrindRail_Combat_Turn_EtoS'() {bLoopAnimation=false, sequence=/GrindRail/Animation/Male/Combat/GrindRail_Combat_Turn_EtoS})]
 }
 LowerBodyPose_S {
  path=MovementMode_LowerBodyOverride.AnimGraphNode_StateMachine_0.LowerBodyTurn.AnimStateNode_11.LowerBodyPose_S
  nodes [
   n1 = AimOffset Player 'ShoppingCart_Passenger_LowerBodyAO'!(
    X:float=0
    BasePose:Pose Link Structure=Rotate Root Bone(
     BasePose=Sequence Player 'GrindRail_Combat_S_Pose'() {sequence=/GrindRail/Animation/Male/Combat/GrindRail_Combat_S_Pose}
     Yaw=180
     MeshToComponent="0,0,0"
    )
    Alpha:float=1
   ) -> (Pose:Pose Link Structure) {blendSpace=/Game/Animation/Game/MainPlayer/Vehicles/ShoppingCart/AO/ShoppingCart_Passenger_LowerBodyAO, alphaBoolBlend {blendOption=Linear}}
   Output Animation Pose(Result=Rotate Root Bone(BasePose=@n1, Yaw=-180, MeshToComponent="0,0,0"))
  ]
 }
 Transition_StoW {
  path=MovementMode_LowerBodyOverride.AnimGraphNode_StateMachine_0.LowerBodyTurn.AnimStateNode_12.Transition_StoW
  nodes [Output Animation Pose(Result=Sequence Player 'GrindRail_Combat_Turn_StoW'() {bLoopAnimation=false, sequence=/GrindRail/Animation/Male/Combat/GrindRail_Combat_Turn_StoW})]
 }
 LowerBodyPose_W {
  path=MovementMode_LowerBodyOverride.AnimGraphNode_StateMachine_0.LowerBodyTurn.AnimStateNode_13.LowerBodyPose_W
  nodes [
   n1 = AimOffset Player 'ShoppingCart_Passenger_LowerBodyAO'(
    X=0
    BasePose=Rotate Root Bone(
     BasePose=Sequence Player 'GrindRail_Combat_W_Pose'() {sequence=/GrindRail/Animation/Male/Combat/GrindRail_Combat_W_Pose}
     Yaw=90
     MeshToComponent="0,0,0"
    )
    Alpha=1
   ) {blendSpace=&ShoppingCart_Passenger_LowerBodyAO, alphaBoolBlend {blendOption=Linear}}
   Output Animation Pose(Result=Rotate Root Bone(BasePose=@n1, Yaw=-90, MeshToComponent="0,0,0"))
  ]
 }
 Transition_WtoN {
  path=MovementMode_LowerBodyOverride.AnimGraphNode_StateMachine_0.LowerBodyTurn.AnimStateNode_15.Transition_WtoN
  nodes [Output Animation Pose(Result=Sequence Player 'GrindRail_Combat_Turn_WtoN'() {bLoopAnimation=false, sequence=/GrindRail/Animation/Male/Combat/GrindRail_Combat_Turn_WtoN})]
 }
 LowerBodyPose_N {
  path=MovementMode_LowerBodyOverride.AnimGraphNode_StateMachine_0.LowerBodyTurn.AnimStateNode_16.LowerBodyPose_N
  nodes [
   n1 = AimOffset Player 'ShoppingCart_Passenger_LowerBodyAO'(X=0, BasePose=Sequence Player 'GrindRail_Combat_N_Pose'() {sequence=/GrindRail/Animation/Male/Combat/GrindRail_Combat_N_Pose}, Alpha=1) {blendSpace=&ShoppingCart_Passenger_LowerBodyAO, alphaBoolBlend {blendOption=Linear}}
   Output Animation Pose(Result=@n1)
  ]
 }
 Transition_EtoN {
  path=MovementMode_LowerBodyOverride.AnimGraphNode_StateMachine_0.LowerBodyTurn.AnimStateNode_17.Transition_EtoN
  nodes [Output Animation Pose(Result=Sequence Player 'GrindRail_Combat_Turn_EtoN'() {bLoopAnimation=false, sequence=/GrindRail/Animation/Male/Combat/GrindRail_Combat_Turn_EtoN})]
 }
 Transition_StoE {
  path=MovementMode_LowerBodyOverride.AnimGraphNode_StateMachine_0.LowerBodyTurn.AnimStateNode_18.Transition_StoE
  nodes [Output Animation Pose(Result=Sequence Player 'GrindRail_Combat_Turn_StoE'() {bLoopAnimation=false, sequence=/GrindRail/Animation/Male/Combat/GrindRail_Combat_Turn_StoE})]
 }
 Transition_WtoS {
  path=MovementMode_LowerBodyOverride.AnimGraphNode_StateMachine_0.LowerBodyTurn.AnimStateNode_19.Transition_WtoS
  nodes [Output Animation Pose(Result=Sequence Player 'GrindRail_Combat_Turn_WtoS'() {bLoopAnimation=false, sequence=/GrindRail/Animation/Male/Combat/GrindRail_Combat_Turn_WtoS})]
 }
 Transition_NtoW {
  path=MovementMode_LowerBodyOverride.AnimGraphNode_StateMachine_0.LowerBodyTurn.AnimStateNode_20.Transition_NtoW
  nodes [Output Animation Pose(Result=Sequence Player 'GrindRail_Combat_Turn_NtoW'() {bLoopAnimation=false, sequence=/GrindRail/Animation/Male/Combat/GrindRail_Combat_Turn_NtoW})]
 }
 Transition_NtoE {
  path=MovementMode_LowerBodyOverride.AnimGraphNode_StateMachine_0.LowerBodyTurn.AnimStateNode_8.Transition_NtoE
  nodes [Output Animation Pose(Result=Sequence Player 'GrindRail_Combat_Turn_NtoE'() {bLoopAnimation=false, sequence=/GrindRail/Animation/Male/Combat/GrindRail_Combat_Turn_NtoE})]
 }
 LowerBodyPose_E {
  path=MovementMode_LowerBodyOverride.AnimGraphNode_StateMachine_0.LowerBodyTurn.AnimStateNode_9.LowerBodyPose_E
  nodes [
   n1 = AimOffset Player 'ShoppingCart_Passenger_LowerBodyAO'(
    X=0
    BasePose=Rotate Root Bone(
     BasePose=Sequence Player 'GrindRail_Combat_E_Pose'() {sequence=/GrindRail/Animation/Male/Combat/GrindRail_Combat_E_Pose}
     Yaw=-90
     MeshToComponent="0,0,0"
    )
    Alpha=1
   ) {blendSpace=&ShoppingCart_Passenger_LowerBodyAO, alphaBoolBlend {blendOption=Linear}}
   Output Animation Pose(Result=Rotate Root Bone(BasePose=@n1, Yaw=90, MeshToComponent="0,0,0"))
  ]
 }
 AnimationTransitionGraph_0 {
  path=MovementMode_LowerBodyOverride.AnimGraphNode_StateMachine_0.LowerBodyTurn.AnimStateTransitionNode_0.AnimationTransitionGraph_0
  nodes [Result(bCanEnterTransition=NOT Boolean(A=Get bAimRGT() -> (bAimRGT:bool)))]
 }
 AnimationTransitionGraph_1 {
  path=MovementMode_LowerBodyOverride.AnimGraphNode_StateMachine_0.LowerBodyTurn.AnimStateTransitionNode_1.AnimationTransitionGraph_1
  nodes [Result()]
 }
 AnimationTransitionGraph_0 {
  path=MovementMode_LowerBodyOverride.AnimGraphNode_StateMachine_0.LowerBodyTurn.AnimStateTransitionNode_10.AnimationTransitionGraph_0
  nodes [Result(bCanEnterTransition=Get bAimRGT())]
 }
 AnimationTransitionGraph_0 {
  path=MovementMode_LowerBodyOverride.AnimGraphNode_StateMachine_0.LowerBodyTurn.AnimStateTransitionNode_11.AnimationTransitionGraph_0
  nodes [Result(bCanEnterTransition=Get bAimLFT() -> (bAimLFT:bool))]
 }
 AnimationTransitionGraph_1 {
  path=MovementMode_LowerBodyOverride.AnimGraphNode_StateMachine_0.LowerBodyTurn.AnimStateTransitionNode_2.AnimationTransitionGraph_1
  nodes [Result()]
 }
 AnimationTransitionGraph_0 {
  path=MovementMode_LowerBodyOverride.AnimGraphNode_StateMachine_0.LowerBodyTurn.AnimStateTransitionNode_25.AnimationTransitionGraph_0
  nodes [Result(bCanEnterTransition=Get bAimRGT())]
 }
 AnimationTransitionGraph_0 {
  path=MovementMode_LowerBodyOverride.AnimGraphNode_StateMachine_0.LowerBodyTurn.AnimStateTransitionNode_26.AnimationTransitionGraph_0
  nodes [Result()]
 }
 AnimationTransitionGraph_1 {
  path=MovementMode_LowerBodyOverride.AnimGraphNode_StateMachine_0.LowerBodyTurn.AnimStateTransitionNode_27.AnimationTransitionGraph_1
  nodes [Result(bCanEnterTransition=Get bAimBWD() -> (bAimBWD:bool))]
 }
 AnimationTransitionGraph_1 {
  path=MovementMode_LowerBodyOverride.AnimGraphNode_StateMachine_0.LowerBodyTurn.AnimStateTransitionNode_28.AnimationTransitionGraph_1
  nodes [Result(bCanEnterTransition=NOT Boolean(A=Get bAimBWD()))]
 }
 AnimationTransitionGraph_1 {
  path=MovementMode_LowerBodyOverride.AnimGraphNode_StateMachine_0.LowerBodyTurn.AnimStateTransitionNode_29.AnimationTransitionGraph_1
  nodes [Result(bCanEnterTransition=Get bAimLFT())]
 }
 AnimationTransitionGraph_0 {
  path=MovementMode_LowerBodyOverride.AnimGraphNode_StateMachine_0.LowerBodyTurn.AnimStateTransitionNode_3.AnimationTransitionGraph_0
  nodes [Result()]
 }
 AnimationTransitionGraph_1 {
  path=MovementMode_LowerBodyOverride.AnimGraphNode_StateMachine_0.LowerBodyTurn.AnimStateTransitionNode_30.AnimationTransitionGraph_1
  nodes [Result(bCanEnterTransition=NOT Boolean(A=Get bAimLFT()))]
 }
 AnimationTransitionGraph_0 {
  path=MovementMode_LowerBodyOverride.AnimGraphNode_StateMachine_0.LowerBodyTurn.AnimStateTransitionNode_31.AnimationTransitionGraph_0
  nodes [Result(bCanEnterTransition=Get bAimFWD() -> (bAimFWD:bool))]
 }
 AnimationTransitionGraph_0 {
  path=MovementMode_LowerBodyOverride.AnimGraphNode_StateMachine_0.LowerBodyTurn.AnimStateTransitionNode_32.AnimationTransitionGraph_0
  nodes [Result(bCanEnterTransition=NOT Boolean(A=Get bAimFWD()))]
 }
 AnimationTransitionGraph_0 {
  path=MovementMode_LowerBodyOverride.AnimGraphNode_StateMachine_0.LowerBodyTurn.AnimStateTransitionNode_35.AnimationTransitionGraph_0
  nodes [Result(bCanEnterTransition=Get bAimFWD())]
 }
 AnimationTransitionGraph_0 {
  path=MovementMode_LowerBodyOverride.AnimGraphNode_StateMachine_0.LowerBodyTurn.AnimStateTransitionNode_36.AnimationTransitionGraph_0
  nodes [Result(bCanEnterTransition=NOT Boolean(A=Get bAimFWD()))]
 }
 AnimationTransitionGraph_0 {
  path=MovementMode_LowerBodyOverride.AnimGraphNode_StateMachine_0.LowerBodyTurn.AnimStateTransitionNode_37.AnimationTransitionGraph_0
  nodes [Result(bCanEnterTransition=Get bAimRGT())]
 }
 AnimationTransitionGraph_1 {
  path=MovementMode_LowerBodyOverride.AnimGraphNode_StateMachine_0.LowerBodyTurn.AnimStateTransitionNode_38.AnimationTransitionGraph_1
  nodes [Result()]
 }
 AnimationTransitionGraph_1 {
  path=MovementMode_LowerBodyOverride.AnimGraphNode_StateMachine_0.LowerBodyTurn.AnimStateTransitionNode_39.AnimationTransitionGraph_1
  nodes [Result(bCanEnterTransition=Get bAimBWD())]
 }
 AnimationTransitionGraph_0 {
  path=MovementMode_LowerBodyOverride.AnimGraphNode_StateMachine_0.LowerBodyTurn.AnimStateTransitionNode_4.AnimationTransitionGraph_0
  nodes [Result(bCanEnterTransition=NOT Boolean(A=Get bAimLFT()))]
 }
 AnimationTransitionGraph_0 {
  path=MovementMode_LowerBodyOverride.AnimGraphNode_StateMachine_0.LowerBodyTurn.AnimStateTransitionNode_40.AnimationTransitionGraph_0
  nodes [Result()]
 }
 AnimationTransitionGraph_0 {
  path=MovementMode_LowerBodyOverride.AnimGraphNode_StateMachine_0.LowerBodyTurn.AnimStateTransitionNode_41.AnimationTransitionGraph_0
  nodes [Result(bCanEnterTransition=Get bAimLFT())]
 }
 AnimationTransitionGraph_0 {
  path=MovementMode_LowerBodyOverride.AnimGraphNode_StateMachine_0.LowerBodyTurn.AnimStateTransitionNode_42.AnimationTransitionGraph_0
  nodes [Result()]
 }
 AnimationTransitionGraph_0 {
  path=MovementMode_LowerBodyOverride.AnimGraphNode_StateMachine_0.LowerBodyTurn.AnimStateTransitionNode_5.AnimationTransitionGraph_0
  nodes [Result(bCanEnterTransition=NOT Boolean(A=Get bAimBWD()))]
 }
 AnimationTransitionGraph_1 {
  path=MovementMode_LowerBodyOverride.AnimGraphNode_StateMachine_0.LowerBodyTurn.AnimStateTransitionNode_6.AnimationTransitionGraph_1
  nodes [Result(bCanEnterTransition=NOT Boolean(A=Get bAimRGT()))]
 }
 AnimationTransitionGraph_0 {
  path=MovementMode_LowerBodyOverride.AnimGraphNode_StateMachine_0.LowerBodyTurn.AnimStateTransitionNode_7.AnimationTransitionGraph_0
  nodes [Result()]
 }
 AnimationTransitionGraph_1 {
  path=MovementMode_LowerBodyOverride.AnimGraphNode_StateMachine_0.LowerBodyTurn.AnimStateTransitionNode_8.AnimationTransitionGraph_1
  nodes [Result(bCanEnterTransition=Get bAimBWD())]
 }
 AnimationTransitionGraph_1 {
  path=MovementMode_LowerBodyOverride.AnimGraphNode_StateMachine_0.LowerBodyTurn.AnimStateTransitionNode_9.AnimationTransitionGraph_1
  nodes [Result(bCanEnterTransition=Get bAimFWD())]
 }
 MovementMode_OffScreen {nodes [
   Input Pose() {name=InPoseOffscreen}
   Output Pose() {layerGroup=MovementMode}
  ]}
 MovementMode_SourcePose {nodes [
   Input Pose() {name=InSourcePoseBase}
   Output Pose() {layerGroup=MovementMode}
  ]}
]
variables []
