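# GrindRail boost Sound Cue that plays a layered boost sound, with separate first- and third-person versions.
# A branch on IsLocal selects 1P attenuation when true, otherwise 3P attenuation with a quieter mix.
# Both versions layer a quickly faded-in Meatball boost start with a delayed Hamster impact and pitched-up copy of the Meatball wave.
# GrindRail_Boost_Conc and the Fort_Gadgets sound class govern playback.
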SoundCue /GrindRail/Audio/GrindRail_Boost_Cue
root=Branch_0
nodes [
 {id=Attenuation_0, children [Modulator_0], properties {attenuationSettings=&/Grinding_1P_Att}}
 {id=Attenuation_1, children [Modulator_2], properties {attenuationSettings=&/Grinding_3P_Loops_Att}}
 {id=Branch_0, children [Attenuation_0, Attenuation_1, Attenuation_1], properties {boolParameterName=IsLocal}}
 {id=Delay_0, children [Mixer_1], properties {delayMax=0.65, delayMin=0.65}}
 {id=Delay_1, children [Mixer_5], properties {delayMax=0.65, delayMin=0.65}}
 {
  id=Enveloper_0
  children [WavePlayer_2]
  properties {pitchCurve {editorCurveData {keys [{time=0}]}}, volumeCurve {editorCurveData {keys [<at 0 0.01>,<at 0.1 1>]}}}
 }
 {
  id=Enveloper_1
  children [WavePlayer_8]
  properties {pitchCurve {editorCurveData {keys [{time=0}]}}, volumeCurve {editorCurveData {keys [<at 0 0.01>,<at 0.1 1>]}}}
 }
 {id=Mixer_0, children [Delay_0,Enveloper_0], properties {inputVolume [0.8,0.8]}}
 {id=Mixer_1, children [WavePlayer_0,Modulator_1], properties {inputVolume [1,1]}}
 {id=Mixer_4, children [Delay_1,Enveloper_1], properties {inputVolume [0.5,0.5]}}
 {id=Mixer_5, children [WavePlayer_7,Modulator_3], properties {inputVolume [1,1]}}
 {id=Modulator_0, children [Mixer_0], properties {pitchMax=1, pitchMin=1}}
 {id=Modulator_1, children [WavePlayer_3], properties {pitchMax=1.5, pitchMin=1.5, volumeMax=1, volumeMin=1}}
 {id=Modulator_2, children [Mixer_4], properties {pitchMax=1, pitchMin=1}}
 {id=Modulator_3, children [WavePlayer_9], properties {pitchMax=1.5, pitchMin=1.5, volumeMax=1, volumeMin=1}}
 {id=WavePlayer_0, properties {soundWaveAssetPtr=/Game/Athena/Sounds/Vehicles/Hamster/Hamster_Impact_Boost_01}}
 {id=WavePlayer_2, properties {soundWaveAssetPtr=/Game/Athena/Sounds/Vehicles/Meatball/SoundWaves/Veh_Meatball_Boost_Start_01}}
 {id=WavePlayer_3, properties {soundWaveAssetPtr=&Veh_Meatball_Boost_Start_01}}
 {id=WavePlayer_7, properties {soundWaveAssetPtr=&Hamster_Impact_Boost_01}}
 {id=WavePlayer_8, properties {soundWaveAssetPtr=&Veh_Meatball_Boost_Start_01}}
 {id=WavePlayer_9, properties {soundWaveAssetPtr=&Veh_Meatball_Boost_Start_01}}
]
properties {
 maxDistance=8500
 soundClassObject=&Fort_Gadgets
 volumeMultiplier=1.133
 attenuationOverrides {
  bEnableOcclusion=true
  lPFFrequencyAtMin=10000
  lPFRadiusMin=2000
  nonFocusVolumeAttenuation=0.7
  occlusionLowPassFilterFrequency=4000
  occlusionTraceChannel=ECC_GameTraceChannel11
  occlusionVolumeAttenuation=4
 }
 concurrencyOverrides {bLimitToOwner=true}
 concurrencySet [&/GrindRail_Boost_Conc]
}
