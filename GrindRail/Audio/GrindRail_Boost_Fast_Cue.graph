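# Sound cue for the fast boost effect during rail grinding, routed to the Fort_Gadgets sound class.
# An IsLocal branch selects first-person attenuation when true, and third-person loop attenuation when false or unset.
# Both paths mix the same two waves, a borrowed Hamster vehicle boost impact and a fast energy layer.
# A boost concurrency set is assigned, alongside occlusion and distance low-pass settings in the cue's overrides.
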SoundCue /GrindRail/Audio/GrindRail_Boost_Fast_Cue
root=Branch_0
nodes [
 {id=Attenuation_0, children [Mixer_2], properties {attenuationSettings=&/Grinding_1P_Att}}
 {id=Attenuation_1, children [Mixer_3], properties {attenuationSettings=&/Grinding_3P_Loops_Att}}
 {id=Branch_0, children [Attenuation_0, Attenuation_1, Attenuation_1], properties {boolParameterName=IsLocal}}
 {id=Mixer_2, children [WavePlayer_5,WavePlayer_4], properties {inputVolume [1,1]}}
 {id=Mixer_3, children [WavePlayer_9,WavePlayer_8], properties {inputVolume [1,1]}}
 {id=WavePlayer_4, properties {soundWaveAssetPtr=&/Grind_Rail_Boost_Fast_Energy_01}}
 {id=WavePlayer_5, properties {soundWaveAssetPtr=/Game/Athena/Sounds/Vehicles/Hamster/Hamster_Impact_Boost_01}}
 {id=WavePlayer_8, properties {soundWaveAssetPtr=&Grind_Rail_Boost_Fast_Energy_01}}
 {id=WavePlayer_9, properties {soundWaveAssetPtr=&Hamster_Impact_Boost_01}}
]
properties {
 maxDistance=8500
 priority=0.9
 soundClassObject=&Fort_Gadgets
 volumeMultiplier=0.881
 attenuationOverrides {
  bEnableOcclusion=true
  lPFFrequencyAtMin=10000
  lPFRadiusMin=2000
  nonFocusVolumeAttenuation=0.7
  occlusionLowPassFilterFrequency=4000
  occlusionTraceChannel=ECC_GameTraceChannel11
  occlusionVolumeAttenuation=4
 }
 concurrencyOverrides {bLimitToOwner=true}
 concurrencySet [&/GrindRail_Boost_Conc]
}
