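# Sound cue for energy footstep sounds used by grind rail boost steps, routed to the gadgets sound class.
# A branch on the IsLocal parameter picks first-person attenuation when true and third-person attenuation when false or unset.
# Each path randomly plays one of six equally weighted, identically named footstep waves, with both random nodes excluded from branch culling.
# Only the local path adds slight pitch modulation.
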SoundCue /GrindRail/Audio/GrindRail_Boost_Steps_Cue
root=Branch_0
nodes [
 {id=Attenuation_0, children [Modulator_0], properties {attenuationSettings=&/Grinding_1P_Att}}
 {id=Attenuation_1, children [Random_7], properties {attenuationSettings=&/Grinding_3P_Att}}
 {id=Branch_0, children [Attenuation_0, Attenuation_1, Attenuation_1], properties {boolParameterName=IsLocal}}
 {id=Modulator_0, children [Random_3], properties {pitchMax=1.02, pitchMin=0.98}}
 {
  id=Random_3
  children [WavePlayer_17, WavePlayer_18, WavePlayer_21, WavePlayer_22, WavePlayer_23, WavePlayer_24]
  properties {bShouldExcludeFromBranchCulling=true, weights [1,1,1,1,1,1]}
 }
 {
  id=Random_7
  children [WavePlayer_34, WavePlayer_35, WavePlayer_36, WavePlayer_37, WavePlayer_38, WavePlayer_39]
  properties {bShouldExcludeFromBranchCulling=true, weights [1,1,1,1,1,1]}
 }
 {id=WavePlayer_17, properties {soundWaveAssetPtr=&/Grind_Rail_Footstep_Energy_01}}
 {id=WavePlayer_18, properties {soundWaveAssetPtr=&/Grind_Rail_Footstep_Energy_02}}
 {id=WavePlayer_21, properties {soundWaveAssetPtr=&/Grind_Rail_Footstep_Energy_03}}
 {id=WavePlayer_22, properties {soundWaveAssetPtr=&/Grind_Rail_Footstep_Energy_04}}
 {id=WavePlayer_23, properties {soundWaveAssetPtr=&/Grind_Rail_Footstep_Energy_05}}
 {id=WavePlayer_24, properties {soundWaveAssetPtr=&/Grind_Rail_Footstep_Energy_06}}
 {id=WavePlayer_34, properties {soundWaveAssetPtr=&Grind_Rail_Footstep_Energy_01}}
 {id=WavePlayer_35, properties {soundWaveAssetPtr=&Grind_Rail_Footstep_Energy_02}}
 {id=WavePlayer_36, properties {soundWaveAssetPtr=&Grind_Rail_Footstep_Energy_03}}
 {id=WavePlayer_37, properties {soundWaveAssetPtr=&Grind_Rail_Footstep_Energy_04}}
 {id=WavePlayer_38, properties {soundWaveAssetPtr=&Grind_Rail_Footstep_Energy_05}}
 {id=WavePlayer_39, properties {soundWaveAssetPtr=&Grind_Rail_Footstep_Energy_06}}
]
properties {maxDistance=6500, soundClassObject=&Fort_Gadgets, volumeMultiplier=1.07}
