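# Sound cue for the metal impact when a player lands on a grind rail, routed to the Athena player foley sound class.
# A branch on the IsLocal parameter picks first-person attenuation for the local player, or third-person attenuation for others and when the parameter is unset.
# Each path applies slight modulation to a random pick between two equally weighted landing waves, with the random nodes excluded from branch culling.
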SoundCue /GrindRail/Audio/GrindRail_Enter_Land_Cue
root=Branch_3
nodes [
 {id=Attenuation_6, children [Modulator_0], properties {attenuationSettings=&/Grinding_1P_Att}}
 {id=Attenuation_7, children [Modulator_1], properties {attenuationSettings=&/Grinding_3P_Att}}
 {id=Branch_3, children [Attenuation_6, Attenuation_7, Attenuation_7], properties {boolParameterName=IsLocal}}
 {id=Modulator_0, children [Random_0], properties {pitchMin=0.98}}
 {id=Modulator_1, children [Random_1], properties {pitchMin=0.98}}
 {id=Random_0, children [WavePlayer_4,WavePlayer_5], properties {bShouldExcludeFromBranchCulling=true, weights [1,1]}}
 {id=Random_1, children [WavePlayer_0,WavePlayer_1], properties {bShouldExcludeFromBranchCulling=true, weights [1,1]}}
 {id=WavePlayer_0, properties {soundWaveAssetPtr=&/Grind_Land_On_Rail_Metal_01}}
 {id=WavePlayer_1, properties {soundWaveAssetPtr=&/Grind_Land_On_Rail_Metal_02}}
 {id=WavePlayer_4, properties {soundWaveAssetPtr=&Grind_Land_On_Rail_Metal_01}}
 {id=WavePlayer_5, properties {soundWaveAssetPtr=&Grind_Land_On_Rail_Metal_02}}
]
properties {maxDistance=6500, soundClassObject=&Fort_Player_Foley_Athena, volumeMultiplier=1.511}
