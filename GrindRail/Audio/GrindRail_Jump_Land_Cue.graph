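# Sound cue for landing a jump on a grind rail, played as player foley.
# A branch on IsLocal picks a first-person attenuation path, or a third-person one when false or unset.
# Each path mixes a random energy-connect layer with a louder-mixed random landing thud, both through modulators, using the same wave variations.
# A concurrency set governs simultaneous plays.
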SoundCue /GrindRail/Audio/GrindRail_Jump_Land_Cue
root=Branch_0
nodes [
 {id=Attenuation_0, children [Mixer_1], properties {attenuationSettings=&/Grinding_1P_Att}}
 {id=Attenuation_1, children [Mixer_2], properties {attenuationSettings=&/Grinding_3P_Att}}
 {id=Branch_0, children [Attenuation_0, Attenuation_1, Attenuation_1], properties {boolParameterName=IsLocal}}
 {id=Mixer_1, children [Modulator_1,Modulator_3], properties {inputVolume [1,1.5]}}
 {id=Mixer_2, children [Modulator_4,Modulator_5], properties {inputVolume [1,1.5]}}
 {id=Modulator_1, children [Random_1], properties {volumeMax=1.2}}
 {id=Modulator_3, children [Random_4]}
 {id=Modulator_4, children [Random_5], properties {volumeMax=1.2}}
 {id=Modulator_5, children [Random_6]}
 {id=Random_1, children [WavePlayer_4,WavePlayer_5], properties {weights [1,1]}}
 {id=Random_4, children [WavePlayer_6,WavePlayer_7,WavePlayer_8], properties {bShouldExcludeFromBranchCulling=true, weights [1,1,1]}}
 {id=Random_5, children [WavePlayer_9, WavePlayer_10], properties {weights [1,1]}}
 {id=Random_6, children [WavePlayer_11, WavePlayer_12, WavePlayer_13], properties {bShouldExcludeFromBranchCulling=true, weights [1,1,1]}}
 {id=WavePlayer_10, properties {soundWaveAssetPtr=&/Grind_Rail_Jump_Land_Energy_Connect_02}}
 {id=WavePlayer_11, properties {soundWaveAssetPtr=&/Grind_Rail_Jump_On_Land_Thud_01}}
 {id=WavePlayer_12, properties {soundWaveAssetPtr=&/Grind_Rail_Jump_On_Land_Thud_02}}
 {id=WavePlayer_13, properties {soundWaveAssetPtr=&/Grind_Rail_Jump_On_Land_Thud_03}}
 {id=WavePlayer_4, properties {soundWaveAssetPtr=&/Grind_Rail_Jump_Land_Energy_Connect_01}}
 {id=WavePlayer_5, properties {soundWaveAssetPtr=&Grind_Rail_Jump_Land_Energy_Connect_02}}
 {id=WavePlayer_6, properties {soundWaveAssetPtr=&Grind_Rail_Jump_On_Land_Thud_01}}
 {id=WavePlayer_7, properties {soundWaveAssetPtr=&Grind_Rail_Jump_On_Land_Thud_02}}
 {id=WavePlayer_8, properties {soundWaveAssetPtr=&Grind_Rail_Jump_On_Land_Thud_03}}
 {id=WavePlayer_9, properties {soundWaveAssetPtr=&Grind_Rail_Jump_Land_Energy_Connect_01}}
]
properties {
 maxDistance=6500
 soundClassObject=&Fort_Player_Foley_Athena
 volumeMultiplier=1.8
 concurrencySet [&/SC_GrindRail_Jump_SC]
}
