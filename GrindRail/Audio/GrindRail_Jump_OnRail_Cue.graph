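# Sound cue for the grind rail jump-on-rail whoosh, routed to the player foley sound class.
# An IsLocal branch selects first-person attenuation when true and third-person attenuation when false or unset.
# Both paths mix the same three randomized, modulated layers: airy whooshes, louder main whooshes, and quieter dead-air whooshes after a short fixed delay.
# The two paths duplicate their nodes, differing only in attenuation.
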SoundCue /GrindRail/Audio/GrindRail_Jump_OnRail_Cue
root=Branch_0
nodes [
 {id=Attenuation_0, children [Mixer_1], properties {attenuationSettings=&/Grinding_1P_Att}}
 {id=Attenuation_1, children [Mixer_0], properties {attenuationSettings=&/Grinding_3P_Att}}
 {id=Branch_0, children [Attenuation_0, Attenuation_1, Attenuation_1], properties {boolParameterName=IsLocal}}
 {id=Delay_0, children [Random_3], properties {delayMax=0.1, delayMin=0.1}}
 {id=Delay_1, children [Random_5], properties {delayMax=0.1, delayMin=0.1}}
 {id=Mixer_0, children [Modulator_5,Modulator_6,Modulator_7], properties {inputVolume [0.5,1,0.3]}}
 {id=Mixer_1, children [Modulator_0,Modulator_1,Modulator_2], properties {inputVolume [0.5,1,0.3]}}
 {id=Modulator_0, children [Random_2], properties {pitchMax=1.2, pitchMin=0.8, volumeMax=1.2, volumeMin=0.5}}
 {id=Modulator_1, children [Random_1]}
 {id=Modulator_2, children [Delay_0], properties {pitchMax=1.4, volumeMax=1.25}}
 {id=Modulator_5, children [Random_4], properties {pitchMax=1.2, pitchMin=0.8, volumeMax=1.2, volumeMin=0.5}}
 {id=Modulator_6, children [Random_0]}
 {id=Modulator_7, children [Delay_1], properties {pitchMax=1.4, volumeMax=1.25}}
 {
  id=Random_0
  children [WavePlayer_5, WavePlayer_6, WavePlayer_9, WavePlayer_10, WavePlayer_11, WavePlayer_12, WavePlayer_19, WavePlayer_20, WavePlayer_21]
  properties {bShouldExcludeFromBranchCulling=true, weights [1,1,1,1,1,1,1,1,1]}
 }
 {
  id=Random_1
  children [WavePlayer_3, WavePlayer_7, WavePlayer_8, WavePlayer_13, WavePlayer_14, WavePlayer_15, WavePlayer_16, WavePlayer_17, WavePlayer_18]
  properties {bShouldExcludeFromBranchCulling=true, weights [1,1,1,1,1,1,1,1,1]}
 }
 {
  id=Random_2
  children [WavePlayer_23, WavePlayer_24, WavePlayer_25, WavePlayer_26]
  properties {bShouldExcludeFromBranchCulling=true, weights [1,1,1,1]}
 }
 {id=Random_3, children [WavePlayer_27, WavePlayer_28, WavePlayer_29], properties {bShouldExcludeFromBranchCulling=true, weights [1,1,1]}}
 {
  id=Random_4
  children [WavePlayer_22, WavePlayer_30, WavePlayer_31, WavePlayer_32]
  properties {bShouldExcludeFromBranchCulling=true, weights [1,1,1,1]}
 }
 {id=Random_5, children [WavePlayer_33, WavePlayer_34, WavePlayer_35], properties {bShouldExcludeFromBranchCulling=true, weights [1,1,1]}}
 {id=WavePlayer_10, properties {soundWaveAssetPtr=&/GrindRail_Jump_Whoosh_04}}
 {id=WavePlayer_11, properties {soundWaveAssetPtr=&/GrindRail_Jump_Whoosh_05}}
 {id=WavePlayer_12, properties {soundWaveAssetPtr=&/GrindRail_Jump_Whoosh_06}}
 {id=WavePlayer_13, properties {soundWaveAssetPtr=&GrindRail_Jump_Whoosh_04}}
 {id=WavePlayer_14, properties {soundWaveAssetPtr=&GrindRail_Jump_Whoosh_05}}
 {id=WavePlayer_15, properties {soundWaveAssetPtr=&GrindRail_Jump_Whoosh_06}}
 {id=WavePlayer_16, properties {soundWaveAssetPtr=&/GrindRail_Jump_Whoosh_07}}
 {id=WavePlayer_17, properties {soundWaveAssetPtr=&/GrindRail_Jump_Whoosh_08}}
 {id=WavePlayer_18, properties {soundWaveAssetPtr=&/GrindRail_Jump_Whoosh_09}}
 {id=WavePlayer_19, properties {soundWaveAssetPtr=&GrindRail_Jump_Whoosh_07}}
 {id=WavePlayer_20, properties {soundWaveAssetPtr=&GrindRail_Jump_Whoosh_08}}
 {id=WavePlayer_21, properties {soundWaveAssetPtr=&GrindRail_Jump_Whoosh_09}}
 {id=WavePlayer_22, properties {soundWaveAssetPtr=&/GrindRail_Jump_Whoosh_Air_01}}
 {id=WavePlayer_23, properties {soundWaveAssetPtr=&GrindRail_Jump_Whoosh_Air_01}}
 {id=WavePlayer_24, properties {soundWaveAssetPtr=&/GrindRail_Jump_Whoosh_Air_02}}
 {id=WavePlayer_25, properties {soundWaveAssetPtr=&/GrindRail_Jump_Whoosh_Air_03}}
 {id=WavePlayer_26, properties {soundWaveAssetPtr=&/GrindRail_Jump_Whoosh_Air_04}}
 {id=WavePlayer_27, properties {soundWaveAssetPtr=&/GrindRail_Jump_Whoosh_Dead_Air_01}}
 {id=WavePlayer_28, properties {soundWaveAssetPtr=&/GrindRail_Jump_Whoosh_Dead_Air_02}}
 {id=WavePlayer_29, properties {soundWaveAssetPtr=&/GrindRail_Jump_Whoosh_Dead_Air_03}}
 {id=WavePlayer_3, properties {soundWaveAssetPtr=&/GrindRail_Jump_Whoosh_01}}
 {id=WavePlayer_30, properties {soundWaveAssetPtr=&GrindRail_Jump_Whoosh_Air_02}}
 {id=WavePlayer_31, properties {soundWaveAssetPtr=&GrindRail_Jump_Whoosh_Air_03}}
 {id=WavePlayer_32, properties {soundWaveAssetPtr=&GrindRail_Jump_Whoosh_Air_04}}
 {id=WavePlayer_33, properties {soundWaveAssetPtr=&GrindRail_Jump_Whoosh_Dead_Air_01}}
 {id=WavePlayer_34, properties {soundWaveAssetPtr=&GrindRail_Jump_Whoosh_Dead_Air_02}}
 {id=WavePlayer_35, properties {soundWaveAssetPtr=&GrindRail_Jump_Whoosh_Dead_Air_03}}
 {id=WavePlayer_5, properties {soundWaveAssetPtr=&GrindRail_Jump_Whoosh_01}}
 {id=WavePlayer_6, properties {soundWaveAssetPtr=&/GrindRail_Jump_Whoosh_02}}
 {id=WavePlayer_7, properties {soundWaveAssetPtr=&GrindRail_Jump_Whoosh_02}}
 {id=WavePlayer_8, properties {soundWaveAssetPtr=&/GrindRail_Jump_Whoosh_03}}
 {id=WavePlayer_9, properties {soundWaveAssetPtr=&GrindRail_Jump_Whoosh_03}}
]
properties {maxDistance=6500, soundClassObject=&Fort_Player_Foley_Athena, volumeMultiplier=0.881}
