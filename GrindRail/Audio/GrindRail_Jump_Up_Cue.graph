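# Sound cue for jumping off a grind rail, classed as player foley.
# An IsLocal branch selects first-person attenuation when true, and third-person attenuation when false or unset.
# Both paths mix a random jump-off variant with a louder random thud, then apply slight pitch modulation.
# The random nodes are excluded from branch culling, and a grind-rail jump concurrency set governs playback.
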SoundCue /GrindRail/Audio/GrindRail_Jump_Up_Cue
root=Branch_0
nodes [
 {id=Attenuation_0, children [Modulator_4], properties {attenuationSettings=&/Grinding_1P_Att}}
 {id=Attenuation_1, children [Modulator_3], properties {attenuationSettings=&/Grinding_3P_Att}}
 {id=Branch_0, children [Attenuation_0, Attenuation_1, Attenuation_1], properties {boolParameterName=IsLocal}}
 {id=Mixer_2, children [Random_6,Random_8], properties {inputVolume [1,1.5]}}
 {id=Mixer_3, children [Random_7,Random_9], properties {inputVolume [1,1.5]}}
 {id=Modulator_3, children [Mixer_3], properties {pitchMax=1.1}}
 {id=Modulator_4, children [Mixer_2], properties {pitchMax=1.1}}
 {
  id=Random_6
  children [WavePlayer_0,WavePlayer_1,WavePlayer_2,WavePlayer_4]
  properties {bShouldExcludeFromBranchCulling=true, weights [1,1,1,1]}
 }
 {
  id=Random_7
  children [WavePlayer_36, WavePlayer_37, WavePlayer_38, WavePlayer_39]
  properties {bShouldExcludeFromBranchCulling=true, weights [1,1,1,1]}
 }
 {id=Random_8, children [WavePlayer_40, WavePlayer_41, WavePlayer_42], properties {bShouldExcludeFromBranchCulling=true, weights [1,1,1]}}
 {id=Random_9, children [WavePlayer_43, WavePlayer_44, WavePlayer_45], properties {bShouldExcludeFromBranchCulling=true, weights [1,1,1]}}
 {id=WavePlayer_0, properties {soundWaveAssetPtr=&/Grind_Rail_Jump_Off_Rail_01}}
 {id=WavePlayer_1, properties {soundWaveAssetPtr=&/Grind_Rail_Jump_Off_Rail_02}}
 {id=WavePlayer_2, properties {soundWaveAssetPtr=&/Grind_Rail_Jump_Off_Rail_03}}
 {id=WavePlayer_36, properties {soundWaveAssetPtr=&Grind_Rail_Jump_Off_Rail_01}}
 {id=WavePlayer_37, properties {soundWaveAssetPtr=&Grind_Rail_Jump_Off_Rail_02}}
 {id=WavePlayer_38, properties {soundWaveAssetPtr=&Grind_Rail_Jump_Off_Rail_03}}
 {id=WavePlayer_39, properties {soundWaveAssetPtr=&/Grind_Rail_Jump_Off_Rail_04}}
 {id=WavePlayer_4, properties {soundWaveAssetPtr=&Grind_Rail_Jump_Off_Rail_04}}
 {id=WavePlayer_40, properties {soundWaveAssetPtr=&/Grind_Rail_Jump_Off_Rail_Thud_01}}
 {id=WavePlayer_41, properties {soundWaveAssetPtr=&/Grind_Rail_Jump_Off_Rail_Thud_02}}
 {id=WavePlayer_42, properties {soundWaveAssetPtr=&/Grind_Rail_Jump_Off_Rail_Thud_03}}
 {id=WavePlayer_43, properties {soundWaveAssetPtr=&Grind_Rail_Jump_Off_Rail_Thud_01}}
 {id=WavePlayer_44, properties {soundWaveAssetPtr=&Grind_Rail_Jump_Off_Rail_Thud_02}}
 {id=WavePlayer_45, properties {soundWaveAssetPtr=&Grind_Rail_Jump_Off_Rail_Thud_03}}
]
properties {
 maxDistance=6500
 soundClassObject=&Fort_Player_Foley_Athena
 volumeMultiplier=1.2
 concurrencyOverrides {bLimitToOwner=true, maxCount=1}
 concurrencySet [&/SC_GrindRail_Jump_SC]
}
