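# Sound cue for the lean-forward sound during rail grinding, in the gadgets sound class.
# A Branch on the IsLocal parameter selects first-person grinding attenuation when true, and third-person attenuation when false or unset.
# Both paths apply slight volume modulation to a random pick between two lean-forward waves, with the random nodes excluded from branch culling.
# Overridden concurrency allows two instances per owner.
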SoundCue /GrindRail/Audio/GrindRail_Lean_Forward_Cue
root=Branch_0
nodes [
 {id=Attenuation_0, children [Modulator_0], properties {attenuationSettings=&/Grinding_1P_Att}}
 {id=Attenuation_1, children [Modulator_1], properties {attenuationSettings=&/Grinding_3P_Att}}
 {id=Branch_0, children [Attenuation_0, Attenuation_1, Attenuation_1], properties {boolParameterName=IsLocal}}
 {id=Modulator_0, children [Random_0], properties {volumeMax=1, volumeMin=0.9}}
 {id=Modulator_1, children [Random_1], properties {volumeMax=1, volumeMin=0.9}}
 {id=Random_0, children [WavePlayer_0,WavePlayer_1], properties {bShouldExcludeFromBranchCulling=true, weights [1,1]}}
 {id=Random_1, children [WavePlayer_2,WavePlayer_3], properties {bShouldExcludeFromBranchCulling=true, weights [1,1]}}
 {id=WavePlayer_0, properties {soundWaveAssetPtr=&/GrindRail_LeanForward_01}}
 {id=WavePlayer_1, properties {soundWaveAssetPtr=&/GrindRail_LeanForward_02}}
 {id=WavePlayer_2, properties {soundWaveAssetPtr=&GrindRail_LeanForward_01}}
 {id=WavePlayer_3, properties {soundWaveAssetPtr=&GrindRail_LeanForward_02}}
]
properties {
 bOverrideConcurrency=true
 maxDistance=6500
 soundClassObject=&Fort_Gadgets
 volumeMultiplier=0.504
 attenuationOverrides {
  bEnableOcclusion=true
  lPFFrequencyAtMin=10000
  lPFRadiusMin=2000
  nonFocusVolumeAttenuation=0.7
  occlusionLowPassFilterFrequency=4000
  occlusionTraceChannel=ECC_GameTraceChannel11
  occlusionVolumeAttenuation=4
 }
 concurrencyOverrides {bLimitToOwner=true, maxCount=2}
}
