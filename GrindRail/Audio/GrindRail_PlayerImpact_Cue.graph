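# Sound cue for the player impact sound on grind rails in the GrindRail feature.
# A Random node picks one of five equally weighted impact waves, and a Modulator adds slight pitch variation on each play.
# The cue uses the Grinding_Impact_Att attenuation asset and the Fort_Gadgets sound class.
# Its concurrency override caps simultaneous instances per owner and sets a short retrigger time.
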SoundCue /GrindRail/Audio/GrindRail_PlayerImpact_Cue
root=Modulator_4
nodes [
 {id=Modulator_4, children [Random_1], properties {pitchMax=1.1, pitchMin=0.9}}
 {id=Random_1, children [WavePlayer_8, WavePlayer_9, WavePlayer_11, WavePlayer_12, WavePlayer_13], properties {weights [1,1,1,1,1]}}
 {id=WavePlayer_11, properties {soundWaveAssetPtr=&/GrindRail_PlayerImpact_02}}
 {id=WavePlayer_12, properties {soundWaveAssetPtr=&/GrindRail_PlayerImpact_03}}
 {id=WavePlayer_13, properties {soundWaveAssetPtr=&/GrindRail_PlayerImpact_04}}
 {id=WavePlayer_8, properties {soundWaveAssetPtr=&/GrindRail_PlayerImpact_05}}
 {id=WavePlayer_9, properties {soundWaveAssetPtr=&/GrindRail_PlayerImpact_01}}
]
properties {
 attenuationSettings=&/Grinding_Impact_Att
 bOverrideConcurrency=true
 maxDistance=4000
 soundClassObject=&Fort_Gadgets
 volumeMultiplier=0.881
 attenuationOverrides {
  bEnableOcclusion=true
  lPFFrequencyAtMin=10000
  lPFRadiusMin=2000
  nonFocusVolumeAttenuation=0.7
  occlusionLowPassFilterFrequency=4000
  occlusionTraceChannel=ECC_GameTraceChannel11
  occlusionVolumeAttenuation=4
 }
 concurrencyOverrides {bLimitToOwner=true, maxCount=3, retriggerTime=0.25}
}
