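# Sound cue for grind rail start-loop audio in the Gadgets sound class, playing a pitched-up goop landing whoosh after a short random delay.
# An IsLocal branch selects first-person attenuation when true and third-person attenuation when false or unset.
# Both paths mirror each other: random delay, pitch modulation, then a random pick between the same two whoosh waves.
# No looping node appears in the cue itself.
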SoundCue /GrindRail/Audio/GrindRail_StartLoop_Cue
root=Branch_0
nodes [
 {id=Attenuation_0, children [Delay_0], properties {attenuationSettings=&/Grinding_1P_Att}}
 {id=Attenuation_1, children [Delay_1], properties {attenuationSettings=&/Grinding_3P_Att}}
 {id=Branch_0, children [Attenuation_0, Attenuation_1, Attenuation_1], properties {boolParameterName=IsLocal}}
 {id=Delay_0, children [Modulator_0], properties {delayMax=0.7, delayMin=0.5}}
 {id=Delay_1, children [Modulator_1], properties {delayMax=0.7, delayMin=0.5}}
 {id=Modulator_0, children [Random_0], properties {pitchMax=1.3, pitchMin=1.2}}
 {id=Modulator_1, children [Random_1], properties {pitchMax=1.3, pitchMin=1.2}}
 {id=Random_0, children [WavePlayer_0,WavePlayer_1], properties {bShouldExcludeFromBranchCulling=true, weights [1,1]}}
 {id=Random_1, children [WavePlayer_2,WavePlayer_3], properties {bShouldExcludeFromBranchCulling=true, weights [1,1]}}
 {id=WavePlayer_0, properties {soundWaveAssetPtr=&/Goop_Land_Whoosh_02}}
 {id=WavePlayer_1, properties {soundWaveAssetPtr=&/Goop_Land_Whoosh_04}}
 {id=WavePlayer_2, properties {soundWaveAssetPtr=&Goop_Land_Whoosh_02}}
 {id=WavePlayer_3, properties {soundWaveAssetPtr=&Goop_Land_Whoosh_04}}
]
properties {
 maxDistance=6500
 soundClassObject=&Fort_Gadgets
 volumeMultiplier=0.692
 attenuationOverrides {
  bEnableOcclusion=true
  lPFFrequencyAtMin=10000
  lPFRadiusMin=2000
  nonFocusVolumeAttenuation=0.7
  occlusionLowPassFilterFrequency=4000
  occlusionTraceChannel=ECC_GameTraceChannel11
  occlusionVolumeAttenuation=4
 }
 concurrencyOverrides {bLimitToOwner=true}
}
