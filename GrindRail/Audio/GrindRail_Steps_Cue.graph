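# Sound cue for grind rail footsteps, playing one of six energy footstep waves with slight pitch variation.
# A Branch on the IsLocal bool picks a first-person attenuation for the local player or a third-person attenuation for others and when the parameter is unset.
# Each side has its own Modulator and equally weighted Random over the same six waves, with both Randoms excluded from branch culling.
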SoundCue /GrindRail/Audio/GrindRail_Steps_Cue
root=Branch_0
nodes [
 {id=Attenuation_0, children [Modulator_0], properties {attenuationSettings=&/Grinding_1P_Att}}
 {id=Attenuation_1, children [Modulator_1], properties {attenuationSettings=&/Grinding_3P_Att}}
 {id=Branch_0, children [Attenuation_0, Attenuation_1, Attenuation_1], properties {boolParameterName=IsLocal}}
 {id=Modulator_0, children [Random_2], properties {pitchMax=1.02, pitchMin=0.98}}
 {id=Modulator_1, children [Random_3], properties {pitchMax=1.02, pitchMin=0.98}}
 {
  id=Random_2
  children [WavePlayer_8, WavePlayer_9, WavePlayer_10, WavePlayer_11, WavePlayer_12, WavePlayer_13]
  properties {bShouldExcludeFromBranchCulling=true, weights [1,1,1,1,1,1]}
 }
 {
  id=Random_3
  children [WavePlayer_14, WavePlayer_15, WavePlayer_16, WavePlayer_17, WavePlayer_18, WavePlayer_19]
  properties {bShouldExcludeFromBranchCulling=true, weights [1,1,1,1,1,1]}
 }
 {id=WavePlayer_10, properties {soundWaveAssetPtr=&/Grind_Rail_Footstep_Energy_03}}
 {id=WavePlayer_11, properties {soundWaveAssetPtr=&/Grind_Rail_Footstep_Energy_04}}
 {id=WavePlayer_12, properties {soundWaveAssetPtr=&/Grind_Rail_Footstep_Energy_05}}
 {id=WavePlayer_13, properties {soundWaveAssetPtr=&/Grind_Rail_Footstep_Energy_06}}
 {id=WavePlayer_14, properties {soundWaveAssetPtr=&/Grind_Rail_Footstep_Energy_01}}
 {id=WavePlayer_15, properties {soundWaveAssetPtr=&/Grind_Rail_Footstep_Energy_02}}
 {id=WavePlayer_16, properties {soundWaveAssetPtr=&Grind_Rail_Footstep_Energy_03}}
 {id=WavePlayer_17, properties {soundWaveAssetPtr=&Grind_Rail_Footstep_Energy_04}}
 {id=WavePlayer_18, properties {soundWaveAssetPtr=&Grind_Rail_Footstep_Energy_05}}
 {id=WavePlayer_19, properties {soundWaveAssetPtr=&Grind_Rail_Footstep_Energy_06}}
 {id=WavePlayer_8, properties {soundWaveAssetPtr=&Grind_Rail_Footstep_Energy_01}}
 {id=WavePlayer_9, properties {soundWaveAssetPtr=&Grind_Rail_Footstep_Energy_02}}
]
properties {
 attenuationSettings=&Grinding_3P_Att
 maxDistance=6500
 soundClassObject=&Fort_Gadgets
 volumeMultiplier=0.755
}
