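; Function library helper ValidateMeshCanRun checks, in editor only, whether an actor's static mesh has cheap enough collision for grind rail use.
; It reads asset registry tags for sections with collision, collision complexity and primitive count, flagging complex collision not set to simple-as-complex or too many simple primitives.
; A Sequence's second pin acts as the failure fallthrough, printing and returning the reason unless an earlier branch returned true.
; Outside the editor it always passes.
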(fn ValidateMeshCanRun
  (params (Mesh :type Static%20Mesh%20Object%20Reference) (CallingActor :type Actor%20Object%20Reference))
  (Utilities|FlowControl|Sequence
    (:then_0
      (if (Editor|InEditor)
        (Utilities|IsValid (Variables|GetMesh)
          (:Is%20Valid
            (bind _output_get (LocalVariables|Setl_assetData (AssetData|CreateAssetData (Variables|GetMesh))))
            (bind (_outtagvalue _returnvalue) (AssetData|GetTagValue _output_get (Utilities|Name|MakeLiteralName "SectionsWithCollision")))
            (if _returnvalue
              (if (not (Utilities|String|EqualExactly(String) _outtagvalue "0"))
                (bind (_outtagvalue_2 _returnvalue_2) (AssetData|GetTagValue (LocalVariables|Getl_assetData) (Utilities|Name|MakeLiteralName "CollisionComplexity")))
                (if _returnvalue_2
                  (if (not (Utilities|String|EqualExactly(String) _outtagvalue_2 "CTF_UseSimpleAsComplex"))
                    (LocalVariables|Setl_failureReason (Utilities|String|ToString(Text) (Utilities|Text|FormatText "NSLOCTEXT(\"[68383B1590364E4CA1F83F0E16203DBA]\", \"3A8A429C40E3CF77B5641D861AE20DBE\", \"{actor_label} is using a mesh ({mesh_name}) with complex collision enabled which is a major memory hit and should never happen.  Use very simplified Simple collision only and use CollisionComplexity=SimpleAsComplex\")" (Utilities|GetDisplayName (Variables|GetCallingActor)) (Utilities|GetDisplayName (Variables|GetMesh))))))))
              (bind (_outtagvalue_3 _returnvalue_3) (AssetData|GetTagValue (LocalVariables|Getl_assetData) (Utilities|Name|MakeLiteralName "CollisionPrims")))
              (if _returnvalue_3
                (if (<= (Utilities|String|StringToInteger _outtagvalue_3) (LocalVariables|Getl_maxSimple))
                  (return true "")
                  (else
                    (LocalVariables|Setl_failureReason (Utilities|String|ToString(Text) (Utilities|Text|FormatText "NSLOCTEXT(\"[68383B1590364E4CA1F83F0E16203DBA]\", \"CA7FDD2D474C998640DDAFBD633203CB\", \"{actor_label} is using a mesh ({mesh_name}) with too many Simple Collision primitives.  We want to use less than {count}.\")" (Utilities|GetDisplayName (Variables|GetCallingActor)) (Utilities|GetDisplayName (Variables|GetMesh)) (LocalVariables|Getl_maxSimple))))))))))
        (else
          (return true ""))))
    (:then_1
      (Development|PrintString 0 (LocalVariables|Getl_failureReason))
      (return false (LocalVariables|Getl_failureReason)))))
