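; Grind rail actor that players ride along a spline, starting by collision or interaction through the pawn's grinding provider component.
; It builds solid and interaction spline meshes per segment plus end tips, with editor-time error visuals for invalid meshes.
; Boosters become grinding range data and instanced arrow meshes, with a Chrono variant driven by per-instance custom data.
; Grinding delegates bound for the local player drive rail material effects and a music sound mix.
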(event Custom|PushGrindingSoundMix
  (params (bWasAlreadyGrinding :type Boolean) (bWasJumpingFromRail :type Boolean) (bEnteredFromInteraction :type Boolean) (PreviousPlayerLocation :type Vector))
  ;[0.1135 0.1135 0.1135] Push Music Up In Level If Grinding
  (Audio|PushSoundMixModifier (Variables|Sound|GetGrindingSoundMixMusic)))

(event Custom|PopGrindingSoundMix
  ;[0.1135 0.1135 0.1135] Push Music Up In Level If Grinding
  (Audio|PopSoundMixModifier (Variables|Sound|GetGrindingSoundMixMusic)))

(event EventBeginPlay
  (CallFunction|FillinBoosterInfo))

(event Custom|HandlePlayerGrindStart
  (params (AttachedPlayer :type Player%20Pawn%20Athena%20Object%20Reference))
  (bind _returnvalue (Actor|GetComponentbyClass AttachedPlayer "/GrindRail/Blueprints/BP_FortPawnGrindingProvider.BP_FortPawnGrindingProvider_C"))
  (bind _grindingprovider (Variables|Default|GetGrindingProvider))
  ;[1 0.1264 0.697] Apply Material Effect when Grinding, Sprinting while grinding and jumping off the grind rail
  (if (Pawn|IsLocallyControlled AttachedPlayer)
    (Variables|Default|SetPlayer%20Mesh (Variables|/Script/Engine.Character|GetMesh AttachedPlayer))
    (Default|CallPlayerStartedGrinding AttachedPlayer _returnvalue)
    (bind _output_get (Variables|Default|SetGrindingProvider _returnvalue))
    (EventDispatchers|BindEventtoBegunGrindingDelegate _output_get (EventDispatchers|CreateEvent :!event Custom|StartedGrinding))
    (EventDispatchers|BindEventtoSprintingStateChangedDelegate _grindingprovider (EventDispatchers|CreateEvent :!event Custom|Sprinting%20state))
    (EventDispatchers|AssignGrindRailChangedDelegate _grindingprovider (EventDispatchers|CreateEvent :!event Custom|GrindRailChanged))
    (EventDispatchers|AssignBegunGrindingDelegate _grindingprovider (EventDispatchers|CreateEvent :!event Custom|PushGrindingSoundMix))
    (EventDispatchers|BindEventtoEndedGrindingDelegate _grindingprovider (EventDispatchers|CreateEvent :!event Custom|PopGrindingSoundMix))))

(event Custom|PutPlayerOnRail_ClientAndServer
  (params (InputPin :type Player%20Pawn%20Athena%20Object%20Reference))
  ;[1 1 1] Starting Rail Grinding from hits
  (CallFunction|PutPlayeronRail InputPin))

(event Custom|PutPlayerOnRail_Server
  (params (InputPin :type Player%20Pawn%20Athena%20Object%20Reference))
  ;[1 1 1] Starting Rail Grinding from hits
  (CallFunction|PutPlayeronRail InputPin))

(event Custom|StartedGrinding
  (params (bWasAlreadyGrinding :type Boolean) (bWasJumpingFromRail :type Boolean) (bEnteredFromInteraction :type Boolean) (PreviousPlayerLocation :type Vector))
  ;[1 0.1264 0.697] Apply Material Effect when Grinding, Sprinting while grinding and jumping off the grind rail
  (CallFunction|StartedGrindingMaterial))

(event Custom|PutPlayerOnRail
  (params (Player :type Player%20Pawn%20Athena%20Object%20Reference))
  (bind _returnvalue (Actor|GetComponentbyClass Player "/GrindRail/Blueprints/BP_FortPawnGrindingProvider.BP_FortPawnGrindingProvider_C"))
  ;[1 1 1] Starting Rail Grinding from hits
  (Utilities|IsValid _returnvalue
    (:Is%20Valid
      (if (not (Class|FortPawnComponentGrindingProvider|IsGrinding _returnvalue))
        (Class|FortPawnComponentGrindingProvider|RequestGrindingonRail _returnvalue self -1.0 false)))))

(event Custom|Sprinting%20state
  (params (bNewIsSprinting :type Boolean))
  ;[1 0.1264 0.697] Apply Material Effect when Grinding, Sprinting while grinding and jumping off the grind rail
  (CallFunction|SprintingStateMaterial bNewIsSprinting))

(event Collision|EventHit
  (params (MyComp :type Primitive%20Component%20Object%20Reference) (Other :type Actor%20Object%20Reference) (OtherComp :type Primitive%20Component%20Object%20Reference) (bSelfMoved :type Boolean) (HitLocation :type Vector :default "0, 0, 0") (HitNormal :type Vector :default "0, 0, 0") (NormalImpulse :type Vector :default "0, 0, 0") (Hit :type Hit%20Result%20Structure%20%28by%20ref%29))
  ;[1 1 1] Starting Rail Grinding from hits
  (bind _asplayer_pawn_athena (Utilities|Casting|CastToPlayerPawn_Athena Other))
  (bind _bstillexiting (CallFunction|IsPlayerExitingThisRail _asplayer_pawn_athena))
  (if (and (not _bstillexiting) (CallFunction|IsStartingGrindingfromWalkingEnabled))
    (CallFunction|PutPlayeronRailClientandServer _asplayer_pawn_athena)))

(event Custom|GrindRailChanged
  (params (OldRail :type Fort%20Grind%20Rail%20Object%20Reference) (NewRail :type Fort%20Grind%20Rail%20Object%20Reference))
  (bind _grindingprovider (Variables|Default|GetGrindingProvider))
  ;[1 0.1264 0.697] Apply Material Effect when Grinding, Sprinting while grinding and jumping off the grind rail
  (CallFunction|JumpOffMaterial)
  (EventDispatchers|UnbindEventfromGrindRailChangedDelegate _grindingprovider (EventDispatchers|CreateEvent :!event Custom|GrindRailChanged))
  (EventDispatchers|UnbindEventfromSprintingStateChangedDelegate _grindingprovider (EventDispatchers|CreateEvent :!event Custom|Sprinting%20state))
  (EventDispatchers|UnbindEventfromBegunGrindingDelegate _grindingprovider (EventDispatchers|CreateEvent :!event Custom|StartedGrinding))
  (EventDispatchers|UnbindEventfromBegunGrindingDelegate _grindingprovider (EventDispatchers|CreateEvent :!event Custom|PushGrindingSoundMix))
  (EventDispatchers|UnbindEventfromEndedGrindingDelegate _grindingprovider (EventDispatchers|CreateEvent :!event Custom|PopGrindingSoundMix)))

(event Custom|GenerateRailComponents
  ;[1 1 1] Building the grindrail - from meshes along the spline to tips
  (if (not (Variables|Default|GetAllowUsingInvalidPossiblyHorrificallyExpensiveMeshes))
    (bind (_valid _failurereason) (Default|ValidateMeshCanRun (Variables|Default|GetMesh_Rail_Default) self))
    (if (not _valid)
      (CallFunction|ShowErrorVisuals _failurereason)))
  (Utilities|FlowControl|Sequence
    (:then_0
      (CallFunction|TryCreateEndTips))
    (:then_1
      (CallFunction|GenerateMeshesAlongSpline))))

(event Custom|RedoSplineMesh
  ;[1 1 1] Building the grindrail - from meshes along the spline to tips
  (for _array_element (Actor|GetComponentsByClass "/Script/Engine.SplineMeshComponent")
    (Components|DestroyComponent _array_element))
  (CallFunction|GenerateMeshesAlongSpline))

(event Custom|Tips
  ;[1 1 1] Building the grindrail - from meshes along the spline to tips
  (CallFunction|DestroyAllTipMeshComponents)
  (CallFunction|TryCreateEndTips))

(event Custom|DestroyTip
  (params (Target :type Actor%20Component%20Object%20Reference))
  ;[1 1 1] Building the grindrail - from meshes along the spline to tips
  (if (== Target (Variables|Default|GetStartTip))
    (Components|DestroyComponent (Variables|Default|GetStartTip))))

(event Custom|DestroyStartTip
  ;[1 1 1] Building the grindrail - from meshes along the spline to tips
  (Components|DestroyComponent (Variables|Default|GetStartTip)))

(event Custom|DestroyEndTip
  ;[1 1 1] Building the grindrail - from meshes along the spline to tips
  (Components|DestroyComponent (Variables|Default|GetEndTip)))

(event EventForceClearBoosters
  ;[1 1 1] Building the grindrail - from meshes along the spline to tips
  (for _array_element (Variables|Booster|GetBoosterMeshes)
    (Components|DestroyComponent _array_element))
  (Utilities|Array|Clear (Variables|Booster|GetBoosterMeshes))
  (Utilities|Array|Clear (Variables|Booster|GetBoosters)))

(event EventGenerateMeshesAlongSpline
  ;[1 1 1] Building the grindrail - from meshes along the spline to tips
  (CallFunction|GenerateRailComponents))

(event Custom|UpdateAllBoosterMeshesInLevel
  ;[1 1 1] Editor-Only Tools for Boosters
  (bind _outactors (Actor|GetAllActorsOfClass (Utilities|GetClass self)))
  (for _array_element _outactors
    (CallFunction|UpdateBoosterMeshes _array_element)))

(event Custom|UpdateBoosterMeshes
  (bind _boosters (Variables|Booster|GetBoosters))
  ;[1 1 1] Editor-Only Tools for Boosters
  (Components|InstancedStaticMesh|ClearInstances (Variables|Default|Getbooster_InstancedStaticMesh))
  (Utilities|Array|Clear (Variables|Booster|GetBoosterMeshes))
  (for _array_element _boosters
    (CallFunction|AddBoosterModel _array_element (Utilities|Array|Length _boosters))))

(fn AddBoosterModel
  (params (Booster :type Grind%20Rail%20Booster%20Directional%20Structure%20%28by%20ref%29) (array%20Index :type Integer))
  (bind (_locationonrailintiles_21_4c298d7e43d6d531e9294d9ebf6d35af _lengthintiles_22_72b92e3b4647946bc77e16b7c38e3157) (Utilities|Struct|BreakGrindRailBoosterDirectional Booster))
  (bind _booster (Variables|GetBooster))
  (bind (_locationonrailintiles_21_4c298d7e43d6d531e9294d9ebf6d35af_2 _lengthintiles_22_72b92e3b4647946bc77e16b7c38e3157_2 _fliplocation_24_a14d038b4ce3b99f148ce080581b9b64 _positivemode_18_f5de4e09428a0ee407a054bbedc68176 _negativemode_20_6d6ecc5845390875bec167b11090b704) (Utilities|Struct|BreakGrindRailBoosterDirectional _booster))
  (bind _returnvalue (Math|Vector|MakeVector 0.0 0.0 -1.0))
  (bind _returnvalue_2 (Math|Vector|MakeVector 0.0 0.0 -1.0))
  (bind _returnvalue_3 (Math|Random|RandomFloatinRange -2e+01 2e+01))
  (bind _returnvalue_4 (Math|Random|RandomFloatinRange 0.85 1.5))
  (bind _returnvalue_5 (Math|Random|RandomFloatinRange 1e+01 -1e+01))
  (bind _distanceperboosterarrow (Variables|Booster|GetDistancePerBoosterArrow))
  (bind _returnvalue_6 (* _lengthintiles_22_72b92e3b4647946bc77e16b7c38e3157 512.0))
  (bind _returnvalue_7 (/ _returnvalue_6 _distanceperboosterarrow))
  (bind _returnvalue_8 (Math|Float|Round _returnvalue_7))
  (bind _boosterlocalpostiony_offset (Variables|Booster|GetBoosterLocalPostionY_Offset))
  (bind _returnvalue_9 (* _boosterlocalpostiony_offset -1.0))
  (bind _returnvalue_10 (Math|Random|RandomFloatinRange _returnvalue_9 _boosterlocalpostiony_offset))
  (bind _boosterscale (Variables|Booster|GetBoosterScale))
  (bind _returnvalue_11 (* _boosterscale _returnvalue_4))
  (bind _spline (Variables|Default|GetSpline))
  (bind _spline_2 (Variables|Default|GetSpline))
  (bind _spline_3 (Variables|Default|GetSpline))
  (bind _distanceperboosterarrow_2 (Variables|Booster|GetDistancePerBoosterArrow))
  (bind _boosterlocalpostionz (Variables|Booster|GetBoosterLocalPostionZ))
  (bind _boosterlocalpostionx (Variables|Booster|GetBoosterLocalPostionX))
  (bind _distanceperboosterarrow_3 (Variables|Booster|GetDistancePerBoosterArrow))
  (Variables|Booster|SetArray%20Index array%20Index)
  (if (Variables|Default|GetIsChronoGrindRail)
    ;[0.002605 0.01042 0.001146] Chrono Booster Setup
    (bind _output_get (Variables|Booster|SetBoosterMeshes _returnvalue_8))
    (for _index (range (+ (- _output_get 5) 1))
      (bind _output_get_2 (Variables|Material|SetInstance%20Index _index))
      (Utilities|Array|Add (LocalVariables|GetTransforms) (Math|Transform|MakeTransform (+ (+ (+ (* (* (Math|Vector|CrossProduct (Spline|GetDirectionatDistanceAlongSpline _spline (CallFunction|BoosterLengthtoSplineLength Booster (* _output_get_2 _distanceperboosterarrow_3)) "World") _returnvalue) _boosterlocalpostionx) (* (- (Math|Float|Clamp(Float) (- (Math|Float|Division(WholeandRemainder) (Math|Conversions|ToFloat(Integer) _index) 4.0) 1.0)) 0.5) 2.0)) (* (Spline|GetDirectionatDistanceAlongSpline _spline (CallFunction|BoosterLengthtoSplineLength Booster (* _output_get_2 _distanceperboosterarrow_3)) "World") _returnvalue_10)) (Spline|GetLocationatDistanceAlongSpline _spline (CallFunction|BoosterLengthtoSplineLength Booster (* _output_get_2 _distanceperboosterarrow_3)) "World")) (* (Math|Vector|Normalize (Math|Vector|RotateVectorAroundAxis (Math|Vector|CrossProduct (Spline|GetDirectionatDistanceAlongSpline _spline (CallFunction|BoosterLengthtoSplineLength Booster (* _output_get_2 _distanceperboosterarrow_3)) "World") (Math|Vector|CrossProduct (Spline|GetDirectionatDistanceAlongSpline _spline (CallFunction|BoosterLengthtoSplineLength Booster (* _output_get_2 _distanceperboosterarrow_3)) "World") _returnvalue)) 0.0 (Spline|GetDirectionatDistanceAlongSpline _spline (CallFunction|BoosterLengthtoSplineLength Booster (* _output_get_2 _distanceperboosterarrow_3)) "World"))) _boosterlocalpostionz)) (Math|Rotator|CombineRotators (Math|Rotator|MakeRotationfromAxes (Spline|GetDirectionatDistanceAlongSpline _spline (CallFunction|BoosterLengthtoSplineLength Booster (* _output_get_2 _distanceperboosterarrow_3)) "World") (Math|Vector|CrossProduct (Spline|GetDirectionatDistanceAlongSpline _spline (CallFunction|BoosterLengthtoSplineLength Booster (* _output_get_2 _distanceperboosterarrow_3)) "World") _returnvalue) (Math|Vector|Normalize (Math|Vector|RotateVectorAroundAxis (Math|Vector|CrossProduct (Spline|GetDirectionatDistanceAlongSpline _spline (CallFunction|BoosterLengthtoSplineLength Booster (* _output_get_2 _distanceperboosterarrow_3)) "World") (Math|Vector|CrossProduct (Spline|GetDirectionatDistanceAlongSpline _spline (CallFunction|BoosterLengthtoSplineLength Booster (* _output_get_2 _distanceperboosterarrow_3)) "World") _returnvalue)) 0.0 (Spline|GetDirectionatDistanceAlongSpline _spline (CallFunction|BoosterLengthtoSplineLength Booster (* _output_get_2 _distanceperboosterarrow_3)) "World")))) (Math|Rotator|MakeRotator _returnvalue_5 _returnvalue_3 (* (Math|Float|Clamp(Float) (- (Math|Float|Division(WholeandRemainder) (Math|Conversions|ToFloat(Integer) _index) 4.0) 1.0)) 1.8e+02))) _returnvalue_11))
      (Utilities|Array|Add (Variables|Booster|GetBoosterMeshes) (Variables|Default|Getbooster_InstancedStaticMesh))
      (Development|PrintString "Hello")
      (bind _returnvalue_12 (Components|InstancedStaticMesh|AddInstance (Variables|Default|Getbooster_InstancedStaticMesh) (Math|Transform|MakeTransform (+ (+ (+ (* (* (Math|Vector|CrossProduct (Spline|GetDirectionatDistanceAlongSpline _spline (CallFunction|BoosterLengthtoSplineLength Booster (* _output_get_2 _distanceperboosterarrow_3)) "World") _returnvalue) _boosterlocalpostionx) (* (- (Math|Float|Clamp(Float) (- (Math|Float|Division(WholeandRemainder) (Math|Conversions|ToFloat(Integer) _index) 4.0) 1.0)) 0.5) 2.0)) (* (Spline|GetDirectionatDistanceAlongSpline _spline (CallFunction|BoosterLengthtoSplineLength Booster (* _output_get_2 _distanceperboosterarrow_3)) "World") _returnvalue_10)) (Spline|GetLocationatDistanceAlongSpline _spline (CallFunction|BoosterLengthtoSplineLength Booster (* _output_get_2 _distanceperboosterarrow_3)) "World")) (* (Math|Vector|Normalize (Math|Vector|RotateVectorAroundAxis (Math|Vector|CrossProduct (Spline|GetDirectionatDistanceAlongSpline _spline (CallFunction|BoosterLengthtoSplineLength Booster (* _output_get_2 _distanceperboosterarrow_3)) "World") (Math|Vector|CrossProduct (Spline|GetDirectionatDistanceAlongSpline _spline (CallFunction|BoosterLengthtoSplineLength Booster (* _output_get_2 _distanceperboosterarrow_3)) "World") _returnvalue)) 0.0 (Spline|GetDirectionatDistanceAlongSpline _spline (CallFunction|BoosterLengthtoSplineLength Booster (* _output_get_2 _distanceperboosterarrow_3)) "World"))) _boosterlocalpostionz)) (Math|Rotator|CombineRotators (Math|Rotator|MakeRotationfromAxes (Spline|GetDirectionatDistanceAlongSpline _spline (CallFunction|BoosterLengthtoSplineLength Booster (* _output_get_2 _distanceperboosterarrow_3)) "World") (Math|Vector|CrossProduct (Spline|GetDirectionatDistanceAlongSpline _spline (CallFunction|BoosterLengthtoSplineLength Booster (* _output_get_2 _distanceperboosterarrow_3)) "World") _returnvalue) (Math|Vector|Normalize (Math|Vector|RotateVectorAroundAxis (Math|Vector|CrossProduct (Spline|GetDirectionatDistanceAlongSpline _spline (CallFunction|BoosterLengthtoSplineLength Booster (* _output_get_2 _distanceperboosterarrow_3)) "World") (Math|Vector|CrossProduct (Spline|GetDirectionatDistanceAlongSpline _spline (CallFunction|BoosterLengthtoSplineLength Booster (* _output_get_2 _distanceperboosterarrow_3)) "World") _returnvalue)) 0.0 (Spline|GetDirectionatDistanceAlongSpline _spline (CallFunction|BoosterLengthtoSplineLength Booster (* _output_get_2 _distanceperboosterarrow_3)) "World")))) (Math|Rotator|MakeRotator _returnvalue_5 _returnvalue_3 (* (Math|Float|Clamp(Float) (- (Math|Float|Division(WholeandRemainder) (Math|Conversions|ToFloat(Integer) _index) 4.0) 1.0)) 1.8e+02))) _returnvalue_11) true))
      (Variables|Booster|Setinstance%20meshIndex _returnvalue_12)
      ;[1 0 0.9726] update  instance staic mesh
      (Components|InstancedStaticMesh|SetCustomDataValue (Variables|Default|Getbooster_InstancedStaticMesh) (Variables|Booster|Getinstance%20meshIndex) 0 (* (- (Variables|Booster|GetBoosterMeshes) -1) (Variables|Booster|GetArray%20Index)) false)
      (Components|InstancedStaticMesh|SetCustomDataValue (Variables|Default|Getbooster_InstancedStaticMesh) (Variables|Booster|Getinstance%20meshIndex) 1 (Math|Conversions|ToFloat(Integer) (Variables|Booster|Getinstance%20meshIndex)) false)
      (Components|InstancedStaticMesh|SetCustomDataValue (Variables|Default|Getbooster_InstancedStaticMesh) (Variables|Booster|Getinstance%20meshIndex) 2 (Utilities|Select :Index _positivemode_18_f5de4e09428a0ee407a054bbedc68176 :NewEnumerator0 1.0 :NewEnumerator1 -1.0 :NewEnumerator4 0.0) false)
      (Components|InstancedStaticMesh|SetCustomDataValue (Variables|Default|Getbooster_InstancedStaticMesh) (Variables|Booster|Getinstance%20meshIndex) 3 (Utilities|Select :Index _negativemode_20_6d6ecc5845390875bec167b11090b704 :NewEnumerator0 1.0 :NewEnumerator1 -1.0 :NewEnumerator4 0.0) false)
      (Components|InstancedStaticMesh|SetCustomDataValue (Variables|Default|Getbooster_InstancedStaticMesh) (Variables|Booster|Getinstance%20meshIndex) 4 (Variables|BoosterMaterial|GetAnimation%20period) false)
      (Components|InstancedStaticMesh|SetCustomDataValue (Variables|Default|Getbooster_InstancedStaticMesh) (Variables|Booster|Getinstance%20meshIndex) 5 (Math|Conversions|ToFloat(Boolean) (Variables|BoosterMaterial|GetSideA_Reverse%20animation%20direction)) false)
      (Components|InstancedStaticMesh|SetCustomDataValue (Variables|Default|Getbooster_InstancedStaticMesh) (Variables|Booster|Getinstance%20meshIndex) 6 (Math|Conversions|ToFloat(Boolean) (Variables|BoosterMaterial|GetSideB_Reverse%20animation%20direction)) false)
      (Components|InstancedStaticMesh|SetCustomDataValue (Variables|Default|Getbooster_InstancedStaticMesh) (Variables|Booster|Getinstance%20meshIndex) 7 (Math|Conversions|ToFloat(Integer) (Variables|BoosterMaterial|GetAnimation%20trailing%20length)) false)
      (LOD|SetMaxDrawDistance (Variables|Default|Getbooster_InstancedStaticMesh) 6e+03))
    (else
      (bind _output_get_3 (Variables|Booster|SetBoosterMeshes _returnvalue_8))
      (for _index_2 (range _output_get_3)
        (Variables|Material|SetInstance%20Index _index_2)
        (Utilities|Array|Add (LocalVariables|GetTransforms) (Math|Transform|MakeTransform (Utilities|Operators|Add (Spline|GetLocationatDistanceAlongSpline _spline_2 (CallFunction|BoosterLengthtoSplineLength Booster (* _index_2 _distanceperboosterarrow_2)) "World") (* (Math|Vector|CrossProduct (Spline|GetDirectionatDistanceAlongSpline _spline_3 (CallFunction|BoosterLengthtoSplineLength Booster (* _index_2 _distanceperboosterarrow_2)) "World") (Math|Vector|CrossProduct (Spline|GetDirectionatDistanceAlongSpline _spline_3 (CallFunction|BoosterLengthtoSplineLength Booster (* _index_2 _distanceperboosterarrow_2)) "World") _returnvalue_2)) 2e+01) (Math|Vector|CrossProduct (Spline|GetDirectionatDistanceAlongSpline _spline_3 (CallFunction|BoosterLengthtoSplineLength Booster (* _index_2 _distanceperboosterarrow_2)) "World") _returnvalue_2)) (Math|Rotator|CombineRotators (Math|Rotator|MakeRotationfromAxes (Spline|GetDirectionatDistanceAlongSpline _spline_3 (CallFunction|BoosterLengthtoSplineLength Booster (* _index_2 _distanceperboosterarrow_2)) "World") (Math|Vector|CrossProduct (Spline|GetDirectionatDistanceAlongSpline _spline_3 (CallFunction|BoosterLengthtoSplineLength Booster (* _index_2 _distanceperboosterarrow_2)) "World") _returnvalue_2) (Math|Vector|Normalize (Math|Vector|RotateVectorAroundAxis (Math|Vector|CrossProduct (Spline|GetDirectionatDistanceAlongSpline _spline_3 (CallFunction|BoosterLengthtoSplineLength Booster (* _index_2 _distanceperboosterarrow_2)) "World") (Math|Vector|CrossProduct (Spline|GetDirectionatDistanceAlongSpline _spline_3 (CallFunction|BoosterLengthtoSplineLength Booster (* _index_2 _distanceperboosterarrow_2)) "World") _returnvalue_2)) 0.0 (Spline|GetDirectionatDistanceAlongSpline _spline_3 (CallFunction|BoosterLengthtoSplineLength Booster (* _index_2 _distanceperboosterarrow_2)) "World")))))))
        (Utilities|Array|Add (Variables|Booster|GetBoosterMeshes) (Variables|Default|Getbooster_InstancedStaticMesh))
        (bind _returnvalue_13 (Components|InstancedStaticMesh|AddInstance (Variables|Default|Getbooster_InstancedStaticMesh) (Math|Transform|MakeTransform (Utilities|Operators|Add (Spline|GetLocationatDistanceAlongSpline _spline_2 (CallFunction|BoosterLengthtoSplineLength Booster (* _index_2 _distanceperboosterarrow_2)) "World") (* (Math|Vector|CrossProduct (Spline|GetDirectionatDistanceAlongSpline _spline_3 (CallFunction|BoosterLengthtoSplineLength Booster (* _index_2 _distanceperboosterarrow_2)) "World") (Math|Vector|CrossProduct (Spline|GetDirectionatDistanceAlongSpline _spline_3 (CallFunction|BoosterLengthtoSplineLength Booster (* _index_2 _distanceperboosterarrow_2)) "World") _returnvalue_2)) 2e+01) (Math|Vector|CrossProduct (Spline|GetDirectionatDistanceAlongSpline _spline_3 (CallFunction|BoosterLengthtoSplineLength Booster (* _index_2 _distanceperboosterarrow_2)) "World") _returnvalue_2)) (Math|Rotator|CombineRotators (Math|Rotator|MakeRotationfromAxes (Spline|GetDirectionatDistanceAlongSpline _spline_3 (CallFunction|BoosterLengthtoSplineLength Booster (* _index_2 _distanceperboosterarrow_2)) "World") (Math|Vector|CrossProduct (Spline|GetDirectionatDistanceAlongSpline _spline_3 (CallFunction|BoosterLengthtoSplineLength Booster (* _index_2 _distanceperboosterarrow_2)) "World") _returnvalue_2) (Math|Vector|Normalize (Math|Vector|RotateVectorAroundAxis (Math|Vector|CrossProduct (Spline|GetDirectionatDistanceAlongSpline _spline_3 (CallFunction|BoosterLengthtoSplineLength Booster (* _index_2 _distanceperboosterarrow_2)) "World") (Math|Vector|CrossProduct (Spline|GetDirectionatDistanceAlongSpline _spline_3 (CallFunction|BoosterLengthtoSplineLength Booster (* _index_2 _distanceperboosterarrow_2)) "World") _returnvalue_2)) 0.0 (Spline|GetDirectionatDistanceAlongSpline _spline_3 (CallFunction|BoosterLengthtoSplineLength Booster (* _index_2 _distanceperboosterarrow_2)) "World")))))) true))
        (Variables|Booster|Setinstance%20meshIndex _returnvalue_13)))))

(fn AddInteractionRailMeshComponent
  (bind _returnvalue (AddComponent|NODE_AddSplineMeshComponent-5 :bManualAttachment true))
  (Utilities|IsValid _returnvalue
    (:Is%20Valid
      (Components|StaticMesh|SetStaticMesh _returnvalue (Variables|Default|GetMesh_Interaction)))
    (:Is%20Not%20Valid
      _))
  (return _returnvalue))

(fn AddSolidRailMeshComponent
  (params (SegmentLength :type Float%20%28double-precision%29) (FirstPoint :type Integer) (SecondPoint :type Integer))
  (bind _returnvalue (AddComponent|NODE_AddSplineMeshComponent-3))
  (Utilities|IsValid _returnvalue
    (:Is%20Valid
      (bind _meshtouse (CallFunction|GetMeshforSplineLength SegmentLength))
      (Components|StaticMesh|SetStaticMesh _returnvalue _meshtouse))
    (:Is%20Not%20Valid
      _))
  (return _returnvalue))

(fn AddTip
  (params (EndTip :type Boolean) (OnlyUpdateSpline :type Boolean) (TipMesh :type Static%20Mesh%20Object%20Reference))
  (bind _spline (Variables|Default|GetSpline))
  (bind _returnvalue_location (Math|Transform|BreakTransform (Spline|GetTransformatDistanceAlongSpline _spline (LocalVariables|GetTipDistanceAlongSpline) "World" false)))
  (bind (_tip_caps_capstarttip_14_be02d3114e8d77de637964b9e3b9bf78 _tip_caps_capendtip_15_1938e7644d71206872041c932b0df129) (Utilities|Struct|BreakGrindRailTipOptions (Variables|GetTip%20Caps_CapStartTip_14_BE02D3114E8D77DE637964B9E3B9BF78)))
  (bind (_tip_caps_capstarttip_14_be02d3114e8d77de637964b9e3b9bf78_2 _tip_caps_capendtip_15_1938e7644d71206872041c932b0df129_2) (Utilities|Struct|BreakGrindRailTipOptions (Variables|GetTip%20Caps_CapStartTip_14_BE02D3114E8D77DE637964B9E3B9BF78)))
  (bind _newcomp (LocalVariables|GetNewComp))
  (if (not (select EndTip (CallFunction|HasTailConnection) (CallFunction|HasHeadConnection)))
    (LocalVariables|SetTipDistanceAlongSpline (Math|Float|SelectFloat (Spline|GetSplineLength _spline) 0.0 EndTip))
    (LocalVariables|SetTipLocation _returnvalue_location)
    (LocalVariables|SetSplineTipRotation (Default|GetRotationforTip EndTip))
    (if (Math|Boolean|ORBoolean (and (Variables|Default|GetEndTip) (not _tip_caps_capendtip_15_1938e7644d71206872041c932b0df129_2)) (and (not (Variables|Default|GetEndTip)) (not _tip_caps_capstarttip_14_be02d3114e8d77de637964b9e3b9bf78_2)) (Editor|InEditor))
      (bind _tipcomponent (Default|SpawnTipComponent))
      (bind _output_get (LocalVariables|SetNewComp (Utilities|Casting|CastToGrindRail_TipComponent _tipcomponent)))
      (Collision|SetCollisionEnabled _output_get "QueryOnly")
      (Transformation|SetWorldLocationAndRotation _output_get (Math|Vector|SelectVector (+ (+ (LocalVariables|GetTipLocation) (* (Transformation|GetUpVector (Utilities|Casting|CastToGrindRail_TipComponent _tipcomponent)) "0.000000,0.000000,-5.000000")) (Variables|Default|GetManualTipLocation)) (+ (LocalVariables|GetTipLocation) (* (Transformation|GetUpVector (Utilities|Casting|CastToGrindRail_TipComponent _tipcomponent)) "0.000000,0.000000,-5.000000")) (Variables|Default|GetAllow%20Manual%20Tip%20Location%20)) (Math|Rotator|SelectRotator (Variables|Default|GetManualTipRotation) (LocalVariables|GetSplineTipRotation) (Variables|Default|GetAllow%20Manual%20Tip%20Rotation)) false false)
      (if (Variables|Default|GetEndTip)
        (Variables|Default|SetEndTip (LocalVariables|GetNewComp))
        (if (and (Editor|InEditor) _tip_caps_capendtip_15_1938e7644d71206872041c932b0df129)
          _)
        (else
          (Variables|Default|SetStartTip (LocalVariables|GetNewComp))
          (if (and _tip_caps_capstarttip_14_be02d3114e8d77de637964b9e3b9bf78 (Editor|InEditor))
            _)))
      (Components|StaticMesh|SetStaticMesh _newcomp (Variables|EditorAssets|GetTipMesh))
      (Development|SetHiddeninGame _newcomp true false)
      (Transformation|SetWorldScale3D _newcomp "1.000000,1.000000,1.000000")
      (Transformation|SetWorldRotation _newcomp (Math|Rotator|MakeRotfromZ (Math|Vector|NegateVector (Spline|GetDirectionatDistanceAlongSpline (Variables|Default|GetSpline) (LocalVariables|GetTipDistanceAlongSpline) "World"))) false false))))

(fn BlueprintGetInteractionTime
  (params (InteractingPawn :type Fort%20Pawn%20Object%20Reference) (InteractionBeingAttempted :type EInteractionBeingAttempted%20Enum :default "FirstInteraction"))
  (return 0.0 true))

(fn BlueprintOnLocalInteract
  (params (InteractingPawn :type Fort%20Player%20Pawn%20Object%20Reference) (InteractionBeingAttempted :type EInteractionBeingAttempted%20Enum :default "FirstInteraction"))
  (Class|FortPawnComponentGrindingProvider|RequestGrindingonRail (Actor|GetComponentbyClass InteractingPawn "/GrindRail/Blueprints/BP_FortPawnGrindingProvider.BP_FortPawnGrindingProvider_C") self (Variables|FortGrindRail|GetIdealInteractStartDistance) true)
  (return true))

(fn BoosterLengthToSplineLength
  (params (Booster :type Grind%20Rail%20Booster%20Directional%20Structure%20%28by%20ref%29) (Length%20Along%20Booster%20%28cm%29 :type Float%20%28double-precision%29))
  (bind (_locationonrailintiles_21_4c298d7e43d6d531e9294d9ebf6d35af _lengthintiles_22_72b92e3b4647946bc77e16b7c38e3157 _fliplocation_24_a14d038b4ce3b99f148ce080581b9b64) (Utilities|Struct|BreakGrindRailBoosterDirectional Booster))
  (return (+ (Math|Float|SelectFloat (- (Spline|GetSplineLength (Variables|Default|GetSpline)) (* _locationonrailintiles_21_4c298d7e43d6d531e9294d9ebf6d35af 512.0)) (* _locationonrailintiles_21_4c298d7e43d6d531e9294d9ebf6d35af 512.0) _fliplocation_24_a14d038b4ce3b99f148ce080581b9b64) (* Length%20Along%20Booster%20%28cm%29 (Math|Float|SelectFloat -1.0 1.0 _fliplocation_24_a14d038b4ce3b99f148ce080581b9b64)))))

(fn CreateSplineMesh
  (params (FirstPoint :type Integer) (SecondPoint :type Integer) (MeshEditorSprite :type Texture%202D%20Object%20Reference))
  (bind _spline (Variables|Default|GetSpline))
  (LocalVariables|SetFirstPointLocation (+ (Variables|/Script/Engine.SceneComponent|GetRelativeLocation (Variables|Default|GetSpline)) (Spline|GetLocationatSplinePoint _spline FirstPoint "Local")))
  (LocalVariables|SetFirstPointTangent (Spline|GetTangentatSplinePoint _spline FirstPoint "Local"))
  (LocalVariables|SetSecondPointLocation (+ (Variables|/Script/Engine.SceneComponent|GetRelativeLocation (Variables|Default|GetSpline)) (Spline|GetLocationatSplinePoint _spline SecondPoint "Local")))
  (LocalVariables|SetSecndPointTangent (Spline|GetTangentatSplinePoint _spline SecondPoint "Local"))
  (LocalVariables|SetDistanceToStartPoint (Spline|GetDistanceAlongSplineatSplinePoint (Variables|Default|GetSpline) (Variables|GetFirstPoint)))
  (LocalVariables|SetDistanceToEndPoint (Spline|GetDistanceAlongSplineatSplinePoint (Variables|Default|GetSpline) (Variables|GetSecondPoint)))
  (Utilities|FlowControl|Sequence
    (:then_0
      ;[1 1 1] Generate Solid Rail Mesh component
      (bind _splinemeshcomponent (CallFunction|AddSolidRailMeshComponent (Math|Float|Absolute(Float) (- (LocalVariables|GetDistanceToEndPoint) (LocalVariables|GetDistanceToStartPoint))) (Variables|GetFirstPoint) (Variables|GetSecondPoint)))
      (bind _output_get (LocalVariables|SetSolidRailMeshComp _splinemeshcomponent))
      (Utilities|IsValid _output_get
        (:Is%20Valid
          (CallFunction|UnifiedMaterialAndUV (Variables|GetFirstPoint) (Variables|GetSecondPoint) _output_get) ; material_unifiedmaterial and UV
          (SplineMesh|SetStartandEnd (LocalVariables|GetSolidRailMeshComp) (LocalVariables|GetFirstPointLocation) (LocalVariables|GetFirstPointTangent) (LocalVariables|GetSecondPointLocation) (LocalVariables|GetSecndPointTangent) true :!class /Script/Engine.SplineMeshComponent)
          (CallFunction|SetupMeshInfo (LocalVariables|GetSolidRailMeshComp))
          ;;[1 0.1264 0.697] Matrerial
          (Utilities|Array|Add (Variables|Default|GetSplineStaticMesh) (LocalVariables|GetSolidRailMeshComp)))))
    (:then_1
      ;[1 1 1] Generate Interaction mesh component
      (bind _splinemeshcomponent_2 (CallFunction|AddInteractionRailMeshComponent))
      (bind _output_get_2 (LocalVariables|SetInteractionRailMeshComp _splinemeshcomponent_2))
      (Transformation|SetWorldScale3D _output_get_2 (* (Transformation|GetWorldScale _output_get_2) (Variables|Default|GetInteractionScale)))
      (Utilities|IsValid (LocalVariables|GetInteractionRailMeshComp)
        (:Is%20Valid
          (Transformation|AttachComponentToComponent (LocalVariables|GetInteractionRailMeshComp) (LocalVariables|GetSolidRailMeshComp) "None" "KeepRelative" "KeepRelative" "KeepRelative" false)
          (SplineMesh|SetStartandEnd (LocalVariables|GetInteractionRailMeshComp) (LocalVariables|GetFirstPointLocation) (LocalVariables|GetFirstPointTangent) (LocalVariables|GetSecondPointLocation) (LocalVariables|GetSecndPointTangent) true :!class /Script/Engine.SplineMeshComponent)
          ;[1 1 1] Editor debug billboard
          (if (Math|Boolean|ANDBoolean (Editor|InEditor) (> (Variables|GetFirstPoint) 0) (Utilities|IsValid (Variables|GetMeshEditorSprite) :!class /Script/Engine.KismetSystemLibrary))
            (bind _returnvalue (AddComponent|NODE_AddBillboardComponent-0 :RelativeTransform_Location "0, 0, 0" :RelativeTransform_Rotation "0, 0, 0" :RelativeTransform_Scale "0.300000,0.300000,0.300000"))
            (Transformation|SetWorldLocation _returnvalue (+ (Spline|GetLocationatSplinePoint (Variables|Default|GetSpline) (Variables|GetFirstPoint) "World") "0.000000,0.000000,50.000000") false true)
            (Rendering|Components|Sprite|SetSprite _returnvalue (Variables|GetMeshEditorSprite))))))))

(fn DestroyAllTipMeshComponents
  (for _array_element (Actor|GetComponentsByClass (Variables|Default|GetGrindRail_TipComponent%20Class))
    (Components|DestroyComponent _array_element)))

(fn FillInBoosterInfo
  (for _array_element (Variables|Booster|GetBoosters)
    (Utilities|Array|Add (Variables|FortGrindRail|GetBoosterInfos) (Utilities|Struct|MakeGrindRailBoosterInfo (Math|Float|Min(Float) (CallFunction|BoosterLengthtoSplineLength _array_element) (CallFunction|BoosterLengthtoSplineLength _array_element (Utilities|TilestoCM (.LengthinTiles_22_72B92E3B4647946BC77E16B7C38E3157 (Utilities|Struct|BreakGrindRailBoosterDirectional _array_element))))) (Math|Float|Max(Float) (CallFunction|BoosterLengthtoSplineLength _array_element) (CallFunction|BoosterLengthtoSplineLength _array_element (Utilities|TilestoCM (.LengthinTiles_22_72B92E3B4647946BC77E16B7C38E3157 (Utilities|Struct|BreakGrindRailBoosterDirectional _array_element))))) (Utilities|Enum|BytetoEnumEGrindRailBoosterMode (.PositiveMode_18_F5DE4E09428A0EE407A054BBEDC68176 (Utilities|Struct|BreakGrindRailBoosterDirectional _array_element))) (Utilities|Enum|BytetoEnumEGrindRailBoosterMode (.NegativeMode_20_6D6ECC5845390875BEC167B11090B704 (Utilities|Struct|BreakGrindRailBoosterDirectional _array_element)))))))

(fn GenerateMeshesAlongSpline
  (Fortnite|FortShippingLog self "Going to generate meshes along spline")
  (Spline|UpdateSpline (Variables|Default|GetSpline))
  (Utilities|Array|Clear (Variables|FortGrindRail|GetGrindRailMeshes))
  (if (Editor|InEditor)
    (bind _returnvalue (Utilities|LoadAssetBlocking (Variables|EditorAssets|GetSplineMeshBillboard)))
    (bind _astexture_2d (Utilities|Casting|CastToTexture2D _returnvalue))
    (LocalVariables|SetSplineMeshBilboardEditorTexture _astexture_2d)
    (else
      _))
  (Utilities|Array|Clear (Variables|Default|GetSplineStaticMesh))
  (bind _output_get (Variables|Material|SetLast%20index (- (Spline|GetNumberOfSplinePoints (Variables|Default|GetSpline)) 2)))
  (for _index (range (+ _output_get 1))
    (CallFunction|CreateSplineMesh _index (+ _index 1) (LocalVariables|GetSplineMeshBilboardEditorTexture)))
  (if (Variables|Default|GetFull%20Circle)
    (Spline|SetClosedLoop (Variables|Default|GetSpline) true true)
    (CallFunction|CreateSplineMesh (- (Spline|GetNumberOfSplinePoints (Variables|Default|GetSpline)) 1) (Spline|GetNumberOfSplinePoints (Variables|Default|GetSpline)) (LocalVariables|GetSplineMeshBilboardEditorTexture))))

(fn GetCurvatureAtDistance
  (params (Distance :type Float%20%28double-precision%29))
  (return (- (Class|AbilityFunctions|GetAngleBetweenTwoVectors (Math|Vector|CrossProduct (Spline|GetDirectionatDistanceAlongSpline (Variables|Default|GetSpline) Distance "World") "0, 0,1.000000") (Spline|GetDirectionatDistanceAlongSpline (Variables|Default|GetSpline) (- Distance 1e+01) "World")) 9e+01)))

(fn GetDirectionAtDistance
  (params (Distance :type Float%20%28single-precision%29))
  (return (Spline|GetDirectionatDistanceAlongSpline (Variables|Default|GetSpline) Distance "World")))

(fn GetLength
  (return (Spline|GetSplineLength (Variables|Default|GetSpline))))

(fn GetMeshCount
  (return (- (Utilities|Array|Length (Actor|GetComponentsByClass "/Script/Engine.MeshComponent")) 1)))

(fn GetMeshForSplineLength
  (params (Segment%20Length :type Float%20%28double-precision%29))
  (return (Variables|Default|GetMesh_Rail_Default)))

(fn GetRailColorAndEmissive
  (return (Math|Color|Desaturate(LinearColor) (Utilities|Select :Index (Variables|Material|GetColor_Select_Enum) :NewEnumerator0 (Variables|Material|GetRedColor) :NewEnumerator1 (Variables|Material|GetGreenColor) :NewEnumerator2 (Variables|Material|GetBlueColor) :NewEnumerator3 (Variables|Material|GetYellowColor) :NewEnumerator4 (Variables|Material|GetSkyBlue) :NewEnumerator5 (Variables|Material|GetPurple) :NewEnumerator6 (Variables|Material|GetOrange)) (Variables|Material|GetColorDesaturationAmount)) (Variables|Material|GetColorEmissiveScale)))

(fn GetRotationForTip
  (params (bEndTip :type Boolean))
  (bind (_returnvalue_location _returnvalue_rotation) (Math|Transform|BreakTransform (Spline|GetTransformatDistanceAlongSpline (Variables|Default|GetSpline) (Math|Float|SelectFloat (Spline|GetSplineLength (Variables|Default|GetSpline)) 0.0 bEndTip) "World" false)))
  (return _returnvalue_rotation))

(fn GetSnapPosition
  (params (CurrentPosition :type Vector%20%28by%20ref%29))
  (return (Spline|FindLocationClosesttoWorldLocation (Variables|Default|GetSpline) CurrentPosition "World") (Spline|GetDistanceAlongSplineatSplineInputKey (Variables|Default|GetSpline) (Spline|FindInputKeyClosesttoWorldLocation (Variables|Default|GetSpline) (Spline|FindLocationClosesttoWorldLocation (Variables|Default|GetSpline) CurrentPosition "World")))))

(fn IsPlayerExitingThisRail
  (params (Player :type Fort%20Player%20Pawn%20Object%20Reference))
  (return (Math|Boolean|ANDBoolean (not (Class|FortPawnComponentGrindingProvider|IsGrinding (Actor|GetComponentbyClass Player "/GrindRail/Blueprints/BP_FortPawnGrindingProvider.BP_FortPawnGrindingProvider_C"))) (== (Class|FortPawnComponentGrindingProvider|GetCurrentGrindingRailActor (Actor|GetComponentbyClass Player "/GrindRail/Blueprints/BP_FortPawnGrindingProvider.BP_FortPawnGrindingProvider_C")) self) (<= (Math|Trig|Acos(Degrees) (Math|Vector|DotProduct (Math|Vector|Normalize (Transformation|GetVelocity Player)) (Spline|FindUpVectorClosesttoWorldLocation (Variables|Default|GetSpline) (Transformation|GetActorLocation Player) "World"))) (ScalableFloat|GetValueatLevel (Variables|Default|GetMaxAngleBetweenVelocityAndNormalForExitDetection))))))

(fn IsTipCapped
  (params (bStartTip :type Boolean))
  (bind (_tip_caps_capstarttip_14_be02d3114e8d77de637964b9e3b9bf78 _tip_caps_capendtip_15_1938e7644d71206872041c932b0df129) (Utilities|Struct|BreakGrindRailTipOptions (Variables|GetTip%20Caps_CapStartTip_14_BE02D3114E8D77DE637964B9E3B9BF78)))
  (return (or (and bStartTip _tip_caps_capstarttip_14_be02d3114e8d77de637964b9e3b9bf78) (and _tip_caps_capendtip_15_1938e7644d71206872041c932b0df129 (not bStartTip)))))

(fn JumpOffMaterial
  (CallFunction|ResetCPDGrindingMaterialIndeces (Variables|FortGrindRail|GetGrindRailMeshes) (Utilities|Array|MakeArray 8 15 16) (Utilities|Array|MakeArray 6 7)))

(fn RemoveAllMeshComponents
  (for _array_element (Actor|GetComponentsByClass "/Script/Engine.MeshComponent")
    (Components|DestroyComponent _array_element)
    (Development|PrintString (Utilities|GetDisplayName _array_element) false)))

(fn SetSplineMeshMaterials
  (params (SplineMeshComp :type Spline%20Mesh%20Component%20Object%20Reference))
  (bind _materialofemisssive (Variables|Material|GetMaterialOfEmisssive))
  (Utilities|IsValid _materialofemisssive
    (:Is%20Valid
      (Rendering|Material|SetMaterial SplineMeshComp 0 _materialofemisssive)
      (Rendering|Material|SetMaterial SplineMeshComp 1 _materialofemisssive))))

(fn ShowErrorVisuals
  (params (error :type String))
  (bind _spline (Variables|Default|GetSpline))
  (bind _returnvalue (Spline|GetNumberOfSplinePoints _spline))
  (bind _error (Variables|Geterror))
  (if (Editor|InEditor)
    (Utilities|FlowControl|Sequence
      (:then_0
        (if (> _returnvalue 1)
          (bind _returnvalue_2 (AddComponent|NODE_AddInstancedStaticMeshComponent-2))
          (bind _returnvalue_3 (Utilities|LoadAssetBlocking (LocalVariables|Getl_materialToLoad)))
          (bind _asmaterial_interface (Utilities|Casting|CastToMaterialInterface _returnvalue_3))
          (Rendering|Material|SetMaterial _returnvalue_2 0 _asmaterial_interface)
          (for _index (range 1 _returnvalue)
            (Utilities|Array|Add (LocalVariables|Getl_transforms) (Math|Transform|MakeTransform (.location (Spline|GetTransformatSplinePoint (Variables|Default|GetSpline) _index "World" false)))))
          (Components|InstancedStaticMesh|AddInstances _returnvalue_2 (LocalVariables|Getl_transforms) false true false)))
      (:then_1
        (if (> (Utilities|String|Len _error) (LocalVariables|Getl_maxLength))
          (bind _returnvalue_4 (Utilities|String|GetCharacterArrayfromString _error))
          (for (_array_element _array_index) _returnvalue_4
            (LocalVariables|Setl_finalError (Utilities|String|Append (LocalVariables|Getl_finalError) _array_element (select (and (== (Math|Integer|%25%28Integer%29 _array_index (LocalVariables|Getl_maxLength)) 0) (> _array_index 1)) "
" ""))))
          (else
            (LocalVariables|Setl_finalError _error))))
      (:then_2
        (bind _returnvalue_5 (AddComponent|NODE_AddTextRenderComponent-0))
        (Rendering|Components|TextRender|SetText _returnvalue_5 (Utilities|Text|ToText(String) (LocalVariables|Getl_finalError)))))))

(fn SpawnTipComponent
  (bind _returnvalue (Game|AddComponentbyClass 0 (Variables|Default|GetGrindRail_TipComponent%20Class) false (struct :Location "0, 0, 0" :Rotation "0, 0, 0" :Scale "1.000000,1.000000,1.000000")))
  (return _returnvalue))

(fn SprintingStateMaterial
  (params (sprinting :type Boolean))
  (CallFunction|SetCPDSprintingMaterial sprinting (Variables|FortGrindRail|GetGrindRailMeshes) 8))

(fn StartedGrindingMaterial
  (CallFunction|ResetCPDGrindingMaterialIndeces (Variables|FortGrindRail|GetGrindRailMeshes) (Utilities|Array|MakeArray 6 7) (Utilities|Array|MakeArray 16 15)))

(fn TryCreateEndTips
  (if (not (Variables|Default|GetFull%20Circle))
    (if (Editor|InEditor)
      (bind _returnvalue (Utilities|LoadAssetBlocking (Variables|EditorAssets|GetTipMesh)))
      (bind _asstatic_mesh (Utilities|Casting|CastToStaticMesh _returnvalue))
      (else
        _))
    (CallFunction|AddTip true false _asstatic_mesh)
    (CallFunction|AddTip false false _asstatic_mesh)))

(fn UnifiedMaterialandUV
  (params (firstPoint :type Integer) (SecondPoint :type Integer) (Add%20Spline%20mesh :type Spline%20Mesh%20Component%20Object%20Reference))
  (bind (_railcolor _coloremissivescale) (Default|GetRailColorandEmissive))
  (bind _spline (Variables|Default|GetSpline))
  (bind _returnvalue (Spline|GetSplineLength _spline))
  (bind (_returnvalue_r _returnvalue_g _returnvalue_b _returnvalue_a) (Math|Color|BreakColor (* _railcolor _coloremissivescale)))
  (CallFunction|SetSplineMeshMaterials (Variables|GetAdd%20Spline%20mesh))
  (Rendering|Material|SetDefaultCustomPrimitiveDataFloat (Variables|GetAdd%20Spline%20mesh) 0 (Math|Float|SafeDivide (Spline|GetDistanceAlongSplineatSplinePoint _spline (Variables|GetfirstPoint)) _returnvalue))
  (Rendering|Material|SetDefaultCustomPrimitiveDataFloat (Variables|GetAdd%20Spline%20mesh) 1 (Math|Float|SafeDivide (Spline|GetDistanceAlongSplineatSplinePoint _spline (Variables|GetSecondPoint)) _returnvalue))
  (Rendering|Material|SetDefaultCustomPrimitiveDataVector4 (Variables|GetAdd%20Spline%20mesh) 2 (Math|Vector4|MakeVector4 _returnvalue_r _returnvalue_g _returnvalue_b _returnvalue_a))
  (Rendering|Material|SetDefaultCustomPrimitiveDataFloat (Variables|GetAdd%20Spline%20mesh) 13 (Math|Conversions|ToFloat(Boolean) (Variables|Material|Get%20Glow%20start%20and%20End)))
  (Rendering|Material|SetDefaultCustomPrimitiveDataFloat (Variables|GetAdd%20Spline%20mesh) 6 1.0)
  (Rendering|Material|SetDefaultCustomPrimitiveDataFloat (Variables|GetAdd%20Spline%20mesh) 9 (Variables|Material|GetGlow%20start%20Radius))
  (Rendering|Material|SetDefaultCustomPrimitiveDataFloat (Variables|GetAdd%20Spline%20mesh) 10 (Variables|Material|GetGlow%20End%20Radius))
  (Rendering|Material|SetDefaultCustomPrimitiveDataFloat (Variables|GetAdd%20Spline%20mesh) 11 (Variables|Material|GetGlow%20start%20%20Intensity))
  (Rendering|Material|SetDefaultCustomPrimitiveDataFloat (Variables|GetAdd%20Spline%20mesh) 12 (Variables|Material|GetGlow%20%20end%20Intensity))
  (Rendering|Material|SetDefaultCustomPrimitiveDataFloat (Variables|GetAdd%20Spline%20mesh) 14 (Math|Conversions|ToFloat(Integer) (Variables|Material|GetLast%20index)))
  (Rendering|Material|SetDefaultCustomPrimitiveDataFloat (Variables|GetAdd%20Spline%20mesh) 17 (Math|Conversions|ToFloat(Boolean) (Variables|Material|GetIgnoreEmissiveDistanceFade)))
  (if (Variables|Default|GetIsChronoGrindRail)
    (Rendering|Material|SetDefaultCustomPrimitiveDataFloat (Variables|GetAdd%20Spline%20mesh) 18 (/ (- (Spline|GetDistanceAlongSplineatSplinePoint _spline (Variables|GetSecondPoint)) (Spline|GetDistanceAlongSplineatSplinePoint _spline (Variables|GetfirstPoint))) (Variables|Default|GetSplineMeshLength))))
  (return))

(fn ConstructionScript
  (CallFunction|GenerateRailComponents))
