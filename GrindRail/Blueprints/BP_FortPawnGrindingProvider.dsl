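; Pawn grinding provider Blueprint for grind rails that decides whether its owning pawn may begin grinding.
; The begin-grind check refuses when the owner pawn is in a vehicle, is DBNO, or has any gameplay tag from the begin-grind block tag list.
; The holster check always returns false, whatever holster ID it receives.
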(fn BP_CanBeginGrinding
  (if (Math|Boolean|ORBoolean (Vehicle|IsinVehicle (Variables|FortPawnComponent_GrindingProvider|GetOwnerPawn)) (Health|IsDBNO (Variables|FortPawnComponent_GrindingProvider|GetOwnerPawn)) (GameplayTags|HasAnyMatchingGameplayTags (Variables|FortPawnComponent_GrindingProvider|GetOwnerPawn) (Variables|Default|GetBeginGrindBlockTags)))
    (return false)
    (else
      (return true))))

(fn CanApplyHolster
  (params (HolsterId :type Name%20%28by%20ref%29 :default "None"))
  (return false))
