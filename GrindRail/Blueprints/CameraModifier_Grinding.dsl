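; Camera modifier blueprint for grinding that switches itself on and off and removes itself from its camera manager.
; TurnOn enables the modifier and clears the TurningOff flag.
; TurnOff starts a non-immediate disable, sets TurningOff and starts a timer of AlphaOutTime that calls RemoveSelf, so removal follows the blend-out.
; TurnOn does not clear that timer, and nothing here reads TurningOff.
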(event Custom|TurnOff
  (CameraModifier|DisableModifier false)
  (Variables|Default|SetTurningOff true)
  (Utilities|Time|SetTimerbyEvent (delegate Custom|RemoveSelf) (Variables|CameraModifier|GetAlphaOutTime)))

(event Custom|TurnOn
  (CameraModifier|EnableModifier)
  (Variables|Default|SetTurningOff false))

(event Custom|RemoveSelf
  (Game|Player|RemoveCameraModifier (Variables|CameraModifier|GetCameraOwner) self))
