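; Grind rail movement-mode extension that applies gameplay effects and cues as a player grinds.
; Event handlers pull cues from BGA_GrindRail for landing, sprinting, boosting and impacts, and apply effects for grinding, fall immunity, jump air control and vehicle or device damage.
; Grinding start skips the effect on simulated proxies.
; A ScalableFloat-gated timer delays IsGrinding removal so GA_GrindRail_Smash is not cancelled immediately, which could trap players on obstacles.
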(event EventHandleGrindingEnded
  ;[1 1 1] FORT-622490 - Remove grinding tag slightly after grinding stops. This is to make sure we don't end GA_GrindRail_Smash immediately after applying, which can cause players to get stuck on objects in a loop where the the smash never happens, and they can't exit the rail.
  ; 0.2s was the minimum delay to not be able to repro the issue
  (if (ScalableFloat|GetValueasBool (Variables|ScalableFloat|GetSF_RemoveGASmashAfterDelay))
    (bind _returnvalue (Utilities|Time|SetTimerbyEvent (delegate Custom|RemoveIsGrindingAfterDelay) 0.2))
    (Variables|Default|SetRemoveIsGrindingTimer _returnvalue)
    (else
      (GameplayEffects|RemoveActiveEffectswithGrantedTags (Variables|FortMovementMode_BaseExtLogic|GetOwnerAbilitySystemComponent) "(GameplayTags=((TagName=\"Granted.Athena.GrindRails.IsGrinding\")))"))))

(event EventHandleGrindingBegun
  (params (GrindRail :type Fort%20Grind%20Rail%20Object%20Reference))
  ;[0.3848 0.7098 1] Begin Grinding
  (bind _asbga_grind_rail (Utilities|Casting|CastToBGA_GrindRail GrindRail))
  (bind _asplayer_pawn_athena (Utilities|Casting|CastToPlayerPawn_Athena (Variables|FortMovementMode_BaseExtLogic|GetOwnerPawn)))
  (Class|BGAGrindRail|HandlePlayerGrindStart _asbga_grind_rail _asplayer_pawn_athena)
  (if (Utilities|Enum|NotEqual(Enum) (Networking|GetLocalRole (Variables|FortMovementMode_BaseExtLogic|GetOwnerPawn)) "ROLE_SimulatedProxy")
    (GameplayEffects|ApplyGameplayEffectToSelf :self (Variables|FortMovementMode_BaseExtLogic|GetOwnerAbilitySystemComponent) :GameplayEffectClass "/GrindRail/Blueprints/GE_GrindRail_IsGrinding.GE_GrindRail_IsGrinding_C" :Level 0.0)))

(event Custom|RemoveIsGrindingAfterDelay
  ;[1 1 1] FORT-622490 - Remove grinding tag slightly after grinding stops. This is to make sure we don't end GA_GrindRail_Smash immediately after applying, which can cause players to get stuck on objects in a loop where the the smash never happens, and they can't exit the rail.
  ; 0.2s was the minimum delay to not be able to repro the issue
  (GameplayEffects|RemoveActiveEffectswithGrantedTags (Variables|FortMovementMode_BaseExtLogic|GetOwnerAbilitySystemComponent) "(GameplayTags=((TagName=\"Granted.Athena.GrindRails.IsGrinding\")))"))

(event EventHandleGameplayCuesOnFeetLanding
  (params (GrindRail :type Fort%20Grind%20Rail%20Object%20Reference))
  (bind _asbga_grind_rail (Utilities|Casting|CastToBGA_GrindRail GrindRail))
  (CallFunction|ExecuteGameplayCueNotReplicated (Variables|/GrindRail/Blueprints/BGA_GrindRail.BGA_GrindRail_C|GetGC_LandOnRail _asbga_grind_rail)))

(event EventHandleGameplayCuesOnSprinting
  (params (GrindRail :type Fort%20Grind%20Rail%20Object%20Reference) (bWantsToGrindSprint :type Boolean))
  (bind _asbga_grind_rail (Utilities|Casting|CastToBGA_GrindRail GrindRail))
  (if bWantsToGrindSprint
    (CallFunction|AddGameplayCueReplicated (Variables|/GrindRail/Blueprints/BGA_GrindRail.BGA_GrindRail_C|GetGC_Sprinting _asbga_grind_rail))
    (else
      (CallFunction|RemoveGameplayCueReplicated (Variables|/GrindRail/Blueprints/BGA_GrindRail.BGA_GrindRail_C|GetGC_Sprinting _asbga_grind_rail)))))

(event EventRemoveGrindRailInitiatedEffects
  (params (GrindRail :type Fort%20Grind%20Rail%20Object%20Reference))
  (bind _asbga_grind_rail (Utilities|Casting|CastToBGA_GrindRail GrindRail))
  (CallFunction|RemoveGameplayCueReplicated (Variables|/GrindRail/Blueprints/BGA_GrindRail.BGA_GrindRail_C|GetGC_Sprinting _asbga_grind_rail))
  (CallFunction|RemoveGameplayCueReplicated (Variables|/GrindRail/Blueprints/BGA_GrindRail.BGA_GrindRail_C|GetGC_SlowBoosting _asbga_grind_rail))
  (CallFunction|RemoveGameplayCueReplicated (Variables|/GrindRail/Blueprints/BGA_GrindRail.BGA_GrindRail_C|GetGC_FastBoosting _asbga_grind_rail)))

(event EventHandleFallDamageImmunityOnGrindingEnd
  (params (bShouldApplyFallDamageImmunityOnEnd :type Boolean))
  (if bShouldApplyFallDamageImmunityOnEnd
    (GameplayEffects|ApplyGameplayEffectToSelf :self (Variables|FortMovementMode_BaseExtLogic|GetOwnerAbilitySystemComponent) :GameplayEffectClass "/Game/Athena/Items/Parents/FallImmunity/GE_Athena_TillLandFallDamageImmunity_Grant_ApplyInAirOnly.GE_Athena_TillLandFallDamageImmunity_Grant_ApplyInAirOnly_C" :Level 0.0)))

(event EventHandleGameplayCuesOnBoosterModeChange
  (params (GrindRail :type Fort%20Grind%20Rail%20Object%20Reference) (OldBoosterMode :type EGrindRailBoosterMode%20Enum :default "SpeedUp") (NewBoosterMode :type EGrindRailBoosterMode%20Enum :default "SpeedUp"))
  (bind _asbga_grind_rail (Utilities|Casting|CastToBGA_GrindRail GrindRail))
  (Utilities|FlowControl|Sequence
    (:then_0
      (switch Utilities|FlowControl|Switch|SwitchonEGrindRailBoosterMode NewBoosterMode
        (:SpeedUp
          (CallFunction|AddGameplayCueReplicated (Variables|/GrindRail/Blueprints/BGA_GrindRail.BGA_GrindRail_C|GetGC_FastBoosting _asbga_grind_rail)))
        (:SlowDown
          (CallFunction|AddGameplayCueReplicated (Variables|/GrindRail/Blueprints/BGA_GrindRail.BGA_GrindRail_C|GetGC_SlowBoosting _asbga_grind_rail)))))
    (:then_1
      (switch Utilities|FlowControl|Switch|SwitchonEGrindRailBoosterMode OldBoosterMode
        (:SpeedUp
          (CallFunction|RemoveGameplayCueReplicated (Variables|/GrindRail/Blueprints/BGA_GrindRail.BGA_GrindRail_C|GetGC_FastBoosting _asbga_grind_rail)))
        (:SlowDown
          (CallFunction|RemoveGameplayCueReplicated (Variables|/GrindRail/Blueprints/BGA_GrindRail.BGA_GrindRail_C|GetGC_SlowBoosting _asbga_grind_rail)))))))

(event EventHandleGameplayCuesOnRailJump
  (params (bIsAnchorJumpActive :type Boolean))
  (if bIsAnchorJumpActive
    (GameplayEffects|ApplyGameplayEffectToSelf :self (Variables|FortMovementMode_BaseExtLogic|GetOwnerAbilitySystemComponent) :GameplayEffectClass "/GrindRail/Blueprints/GE_DisableAirControl_GrindRail.GE_DisableAirControl_GrindRail_C" :Level 0.0)))

(event EventHandleGameplayEffectsOnVehicleHit
  (params (AthenaVehicle :type Fort%20Athena%20Vehicle%20Object%20Reference))
  (GameplayEffects|ApplyGameplayEffectToTarget (Variables|FortMovementMode_BaseExtLogic|GetOwnerAbilitySystemComponent) "/GrindRail/Blueprints/GE_GrindRail_VehicleDamage.GE_GrindRail_VehicleDamage_C" (Ability|GetAbilitySystemComponent AthenaVehicle) 0.0 (GameplayEffects|MakeEffectContext (Variables|FortMovementMode_BaseExtLogic|GetOwnerAbilitySystemComponent))))

(event EventHandleGameplayCuesOnHit
  (params (GrindRail :type Fort%20Grind%20Rail%20Object%20Reference))
  (bind _asbga_grind_rail (Utilities|Casting|CastToBGA_GrindRail GrindRail))
  (Utilities|FlowControl|DoOnce false
    (:Completed
      (CallFunction|ExecuteGameplayCueReplicated (Variables|/GrindRail/Blueprints/BGA_GrindRail.BGA_GrindRail_C|GetGC_ImpactPlayer _asbga_grind_rail))
      (Utilities|Time|SetTimerbyEvent (delegate Custom|AllowAnotherImpactEffect) 0.5))))

(event Custom|AllowAnotherImpactEffect
  (Utilities|FlowControl|DoOnce false
    (:Completed
      (CallFunction|ExecuteGameplayCueReplicated (Variables|/GrindRail/Blueprints/BGA_GrindRail.BGA_GrindRail_C|GetGC_ImpactPlayer (stale EventHandleGameplayCuesOnHit _asbga_grind_rail)))
      (Utilities|Time|SetTimerbyEvent (delegate Custom|AllowAnotherImpactEffect) 0.5))))

(event EventApplyDamageToDevice
  (params (BuildingActor :type Building%20Actor%20Object%20Reference))
  (Utilities|IsValid BuildingActor
    (:Is%20Valid
      (if (Variables|/Script/Engine.Actor|GetbCanBeDamaged BuildingActor)
        (GameplayEffects|ApplyGameplayEffectSpecToTarget (Variables|FortMovementMode_BaseExtLogic|GetOwnerAbilitySystemComponent) (GameplayEffects|MakeOutgoingSpec (Variables|FortMovementMode_BaseExtLogic|GetOwnerAbilitySystemComponent) "/GrindRail/Blueprints/GE_GrindRail_DeviceDamage.GE_GrindRail_DeviceDamage_C" 1.0 (GameplayEffects|MakeEffectContext (Variables|FortMovementMode_BaseExtLogic|GetOwnerAbilitySystemComponent))) (Ability|GetAbilitySystemComponent BuildingActor))
        (CallFunction|AddTemporaryMoveIgnoreActor BuildingActor (Variables|Default|GetTemporaryMoveIgnoreDuration))))))

(fn GetGrindInAirGameplayEffectClass
  (params (GrindRail :type Fort%20Grind%20Rail%20Object%20Reference))
  (bind _asbga_grind_rail (Utilities|Casting|CastToBGA_GrindRail GrindRail)
    (:then
      (return (Variables|/GrindRail/Blueprints/BGA_GrindRail.BGA_GrindRail_C|GetGE_GrindingIncludingInAir _asbga_grind_rail)))
    (:CastFailed
      (return "/GrindRail/Blueprints/GE_GrindingIncludingInAir.GE_GrindingIncludingInAir_C"))))

(fn IsDamageableDevice
  (params (BuildingActor :type Building%20Actor%20Object%20Reference))
  (return (GameplayTags|HasMatchingGameplayTag BuildingActor "(TagName=\"Asset.Creative.Device\")")))

(fn IsThrowableItem
  (params (BuildingActor :type Building%20Actor%20Object%20Reference))
  (Utilities|Casting|CastToBGA_HeldObject_Parent BuildingActor
    (:then
      _)
    (:CastFailed
      _))
  (Utilities|Casting|CastToBGA_Athena_WithGravity_Parent BuildingActor
    (:then
      _)
    (:CastFailed
      _))
  (Utilities|Casting|CastToBGA_LinkedActorWithGravity_Parent BuildingActor
    (:then
      _)
    (:CastFailed
      _))
  (return (Math|Boolean|ORBoolean false false false)))
