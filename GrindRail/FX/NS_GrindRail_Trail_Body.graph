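# Niagara system for the body trail on a character grinding a rail: air meshes, dust spikes, foot flares with lights, glow drops and hand ribbons.
# Most foot effects attach to User.SkeletalMesh bones or sockets, with drop offsets following lean direction.
# Speed drives spike and drop spawn rates, ColorIndex selects rail colors, time of day scales brightness and Jumped hides most effects.
# Drops use CPU and GPU emitters, one inheriting source movement.
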NiagaraSystem /GrindRail/FX/NS_GrindRail_Trail_Body
dependencies {
 usedDataInterfaces [
  &NiagaraDataInterfaceCurve
  &NiagaraDataInterfaceArrayFloat
  &NiagaraDataInterfaceArrayInt32
  &NiagaraDataInterfaceArrayFloat3
  &NiagaraDataInterfaceColorCurve
  &NiagaraDataInterfaceVector2DCurve
  &NiagaraDataInterfaceMeshRendererInfo
  &NiagaraDataInterfaceArrayColor
  &NiagaraDataInterfaceArrayFloat2
 ]
 usedDynamicInputs [
  &self:GetEulerRot
  &ClampFloat
  &RemapRange
  &VectorLength
  &Boolean_OrOperation
  &Boolean_NotOperation
  &MakeCustomIntFromBool
  &self:Grinding_Color
  /Niagara/DynamicInputs/Lerp/SmoothLerpOverTime_Color
 ]
 usedModules [
  &self:SystemSpawnScript.SystemScriptSource.SystemScriptGraph.NiagaraNodeAssignment_9.SetVariables_#1
  &SystemState
  &SampleSkeletalMesh
  &self:SystemSpawnScript.SystemScriptSource.SystemScriptGraph.NiagaraNodeAssignment_12.SetVariables_#2
  &self:SystemSpawnScript.SystemScriptSource.SystemScriptGraph.NiagaraNodeAssignment_6.SetVariables_#3
  &InterpolateOverTime
  &self:SystemSpawnScript.SystemScriptSource.SystemScriptGraph.NiagaraNodeAssignment_0.SetVariables_#4
  &self:SystemSpawnScript.SystemScriptSource.SystemScriptGraph.NiagaraNodeAssignment_1.SetVariables_#5
  &FN_Timer
  &self:SystemSpawnScript.SystemScriptSource.SystemScriptGraph.NiagaraNodeAssignment_2.SetVariables_#6
  &self:SystemSpawnScript.SystemScriptSource.SystemScriptGraph.NiagaraNodeAssignment_3.SetVariables_#7
  &self:SystemSpawnScript.SystemScriptSource.SystemScriptGraph.NiagaraNodeAssignment_15.SetVariables_#8
  &self:SystemSpawnScript.SystemScriptSource.SystemScriptGraph.NiagaraNodeAssignment_4.SetVariables_#9
 ]
}
emitters {
 Air {
  emitterUpdateScript [
   &EmitterState()
   &SpawnBurst_Instantaneous(Spawn Probability=$System.FN_Timer.Value, Use Spawn Probability=true, UseLoopCountLimit=false)
   &SpawnRate(
    SpawnRate=&Multiply_Float(
     A=2
     B=&Multiply_Float(
      A=&Multiply_Float(A=&Add_Float(A=$User.Speed_Alpha, B=0.4), B=1)
      B=$System.FN_Timer.Value
     )
    )
    Use Spawn Probability=false
   )
  ]
  particleSpawnScript [
   &InitializeParticle(Color=<rgba 1 1 1 0.05>, Mesh Scale=<1.5 3 3>, Mesh Scale Mode=::NewEnumerator4, UsePositionOffset=false, Write Material Random=true)
   &InitialMeshOrientation(Mesh Orientation Mode=::NewEnumerator7, Rotation=<0.5 0 0>, Use Rotation Vector=true)
  ]
  particleUpdateScript [
   &ParticleState()
   &ScaleColor(
    Scale Alpha=&FloatFromCurve(
     FloatCurve=DataInterface() {propertyValues {Curve {keys [
         {arriveTangent=5.8909, interpMode=RCIM_Cubic, leaveTangent=5.8909, tangentMode=RCTM_User, time=0, value=0}
         <at 0.3 1 RCIM_Cubic>
         {arriveTangent=-1.4286, interpMode=RCIM_Cubic, leaveTangent=-1.4286, time=1, value=0}
        ]}}}
     Scale Curve=&MakeCustomFloatFromBool(Bool=$User.Jumped, True Float=0, False Float=1)
    )
    ScaleA=true
    ScaleRGB=false
   )
   &ScaleColor(
    Scale RGB=&VectorFromFloat(Value=&FNTimeOfDayScalar(Night Scale=0.35, Morning Scale=0.5, Day Scale=0.8, Evening Scale=0.4))
    ScaleA=true
    ScaleRGB=true
   ) {module=ScaleColor002}
  ]
  rendererClasses [&NiagaraMeshRendererProperties]
  renderers [{rendererClass=&NiagaraMeshRendererProperties, rendererIndex=0}]
 }
 DustSpikes001 {
  emitterUpdateScript [
   &EmitterState(
    Life Cycle Mode=::NewEnumerator1
    Loop Behavior=::NewEnumerator1
    Loop Delay=$User.Delay
    Loop Duration Mode=::NewEnumerator1
    UseLoopDelay=true
   )
   &SpawnRate(
    SpawnRate=&FloatFromCurve(
     FloatCurve=DataInterface() {propertyValues {Curve {keys [
         <at 0 0 RCIM_Cubic>
         {arriveTangent=2.5694, interpMode=RCIM_Cubic, leaveTangent=2.5694, tangentMode=RCTM_User, time=0.75, value=0}
         {arriveTangent=463.4894, interpMode=RCIM_Cubic, leaveTangent=463.4894, tangentMode=RCTM_User, time=1, value=40}
        ]}}}
     CurveIndex=&MakeCustomFloatFromBool(
      Bool=$User.TurnON_Niagara_Velocity
      True Float=$System.NormalizedVelocityRange
      False Float=&NormalizeFloat(Value=&Add_Float(A=$User.Speed_Alpha, B=0.4), Min=0, Max=1)
     )
     Scale Curve=&Multiply_Float(
      A=&MakeCustomFloatFromBool(
       Bool=&SetBoolByFloatComparison(A=$System.NormalizedVelocityRange, B=0.75, Comparison Type=::NewEnumerator6)
       True Float=0
       False Float=&FloatFromCurve(
        FloatCurve=DataInterface() {propertyValues {Curve {keys [
            {interpMode=RCIM_Cubic, tangentMode=RCTM_User, time=0, value=0}
            {arriveTangent=2.6713, interpMode=RCIM_Cubic, leaveTangent=2.6713, tangentMode=RCTM_User, time=1, value=1}
           ]}}}
        CurveIndex=$System.NormalizedVelocityRange
        Scale Curve=2
       )
      )
      B=&Add_Float(A=$User.Speed_Alpha, B=0.4)
     )
    )
    Use Spawn Probability=false
   )
  ]
  particleSpawnScript [
   &FN_SetSpawnGroup(Number of Spawn Groups=2)
   &InitializeParticle(
    Color=&SelectLinearColorFromArray(Array Sampling Mode=::NewEnumerator1, Direct Array Index=$User.ColorIndex, Color Selection Array=$System.GrindRail_ColorArray, Random Seed=?)
    Lifetime=&RandomRangeFloat(Minimum=0.1, Maximum=0.2)
    Mesh Scale Mode=::NewEnumerator1
    Mesh Uniform Scale Max=0.6
    Mesh Uniform Scale Min=0.4
    UsePositionOffset=false
    Write Material Random=true
   )
   &CameraOffset(Camera Offset Amount=10)
   &ShapeLocation(
    Box Size=<0 2 0>
    MaskSpawnGroup=false
    Non Uniform Scale=<0 0 0>
    Override Seed=false
    Shape Primitive=::NewEnumerator2
    Surface Only=false
    UseOwnerScale=true
   )
   &AddVelocity(
    Cone Angle=10
    Cone Axis=<-1 0 0.05>
    Module.UseConeFalloff=false
    Module.UseVelDistribution=false
    Override Seed=false
    Velocity Mode=::NewEnumerator2
    Velocity Speed=&Multiply_Float(
     A=&Multiply_Float(A=&RandomRangeFloat(Minimum=150, Maximum=300), B=1)
     B=1
    )
   )
   &InitialMeshOrientation(
    Mesh Orientation Mode=::NewEnumerator2
    Rotation=&MakeVector(X=&RandomRangeFloat(Minimum=0, Maximum=1), Y=0, Z=0)
    Use Rotation Vector=true
   ) {module=InitialMeshOrientation001}
  ]
  particleUpdateScript [
   &ParticleState()
   &SkeletalMeshLocation(
    Bone / Socket Index=&SelectIntFromArray(Direct Array Index=$Particles.SpawnGroup, Int Selection Array=DataInterface() {propertyValues {IntData [8,9]}})
    Bone Sampling Mode=::NewEnumerator4
    MaskSpawnGroup=false
    Orientation Sampling=::NewEnumerator1
    Sampled Position Offset=&AddVectorToPosition(
     Position=&SelectVectorFromArray(Direct Array Index=$Particles.SpawnGroup, Vector Selection Array=DataInterface() {propertyValues {FloatData [<2.5 0 -10>, <5 0 -10>]}})
     Vector=$Particles.ShapeLocation.ShapePosition
    )
    Sampled Position Offset Coordinate Space:ENiagaraSkelMeshTransforms=::NewEnumerator2
    Skeletal Mesh=$User.SkeletalMesh
   )
   &DynamicMaterialParameters(
    Index 0 Param 1=$Particles.NormalizedAge
    Index 0 Param 2=&RandomRangeFloat(Minimum=1, Maximum=1.5)
    Index 0 Param 3=&RandomRangeFloat(Minimum=0, Maximum=0)
    Index 0 Param 4=&FloatFromCurve(
     FloatCurve=DataInterface() {propertyValues {Curve {keys [
         {interpMode=RCIM_Cubic, tangentMode=RCTM_User, time=0, value=0}
         {arriveTangent=3.2801, interpMode=RCIM_Cubic, leaveTangent=3.2801, tangentMode=RCTM_User, time=1, value=1}
        ]}}}
     CurveIndex=$System.NormalizedVelocityRange
     Scale Curve=&RandomRangeFloat(Minimum=0.3, Maximum=0.2)
    )
    Module.Param0WriteEnabled=true
    Module.Param1WriteEnabled=true
    Module.Param2WriteEnabled=true
    Module.Param3WriteEnabled=true
   )
   &Drag(
    Drag=&MaskFloatBySpawnGroup(Masked Float=4, Passthrough Float=1)
    Use Linear Drag=true
    Use Rotational Drag=false
   )
   &ScaleMeshSize(Scale Factor=&VectorFromFloat(Value=&FloatFromCurve(
      FloatCurve=DataInterface() {propertyValues {Curve {keys [
          {arriveTangent=9.0725, interpMode=RCIM_Cubic, leaveTangent=9.0725, tangentMode=RCTM_User, time=0, value=0}
          {arriveTangent=-2.5249, interpMode=RCIM_Cubic, leaveTangent=-2.5249, tangentMode=RCTM_User, time=0.1408, value=0.9847}
          {arriveTangent=-0.0405, interpMode=RCIM_Cubic, leaveTangent=-0.0405, tangentMode=RCTM_User, time=1, value=0}
         ]}}}
      Scale Curve=&SelectFloatFromArray(Direct Array Index=$System.Color_Index, Float Selection Array=DataInterface() {propertyValues {FloatData [1,1.5,0.8,1.5]}})
     )))
   &ScaleColor(
    Linear Color Curve=DataInterface() {propertyValues {
      AlphaCurve {keys [<at 0 5>,<at 1 0>]}
      BlueCurve {keys [<at 0 50 RCIM_Cubic>]}
      GreenCurve {keys [<at 0 50 RCIM_Cubic>]}
      RedCurve {keys [<at 0 50 RCIM_Cubic>]}
     }}
    Scale Mode=::NewEnumerator2
   )
   &SolveForcesAndVelocity(Clamp Velocity=false, Limit Acceleration=false)
   &ScaleColor(
    Scale Alpha=&Multiply_Float(
     A=&MakeCustomFloatFromBool(Bool=$User.Jumped, True Float=0, False Float=5)
     B=&Lerp_Float(A=1, B=1, Alpha=$User.Is180Turning)
    )
    Scale RGB=&Multiply_Vector(
     A=&VectorFromFloat(Value=&FNTimeOfDayScalar(Night Scale=5, Morning Scale=3, Evening Scale=2))
     B=&VectorFromFloat(Value=&SelectFloatFromArray(Direct Array Index=$User.ColorIndex, Float Selection Array=DataInterface() {propertyValues {FloatData [1,0.8,1,1,0.75,1,0.9]}}))
    )
    ScaleA=true
    ScaleRGB=true
   ) {module=ScaleColor002}
  ]
  rendererClasses [&NiagaraMeshRendererProperties]
  renderers [{rendererClass=&NiagaraMeshRendererProperties, rendererIndex=0}]
 }
 Foot_Flare001 {
  emitterUpdateScript [
   &EmitterState(
    Inactive Response=::NewEnumerator1
    Life Cycle Mode=::NewEnumerator1
    Loop Behavior=::NewEnumerator1
    Loop Delay=$User.Delay
    Loop Duration Mode=::NewEnumerator1
    UseLoopDelay=true
   )
   &SpawnBurst_Instantaneous(Spawn Count=3, Use Spawn Probability=false)
  ]
  particleSpawnScript [
   &PartitionParticles(Number Of Partitions=3, Particles Per Partition=1, Partition Mode:ENiagara_PartitionMode=::NewEnumerator1)
   /Niagara/Modules/Spawn/Initialization/InitializeParticle(
    Color=$System.Color_Smoothed
    Module.Write Color=true
    Module.Write Lifetime=true
    Module.Write Mass=true
    Module.Write Material Random=true
    Module.Write Position=false
    Module.Write Scale=true
    Module.Write SpriteRotation=true
    Module.Write SpriteSize=true
    Position=&SimulationPosition()
    Sprite Rotation=&Multiply_Float(
     A=&RandomRangeFloat(Minimum=-360, Maximum=360)
     B=&SelectFloatFromArray(Direct Array Index=$Particles.PartitionParticles.Partition, Float Selection Array=DataInterface() {propertyValues {FloatData [1,1,0]}})
    )
    Sprite Size=&SelectVector2DFromArray(
     Direct Array Index=$Particles.PartitionParticles.Partition
     Vector 2D Selection Array=DataInterface() {propertyValues {FloatData [<50 50>, <50 50>, <120 200>]}}
    )
   )
   /SetVariables(
    Particles.GlowFoot_RailColor:LinearColor=&MultiplyLinearColor(
     Linear Color=&SelectLinearColorFromArray(Array Sampling Mode=::NewEnumerator1, Direct Array Index=$User.ColorIndex, Color Selection Array=$System.GrindRail_ColorArray, Random Seed=?)
     Scale Factor=&MakeFloat4FromVectorAndFloat("Vector (RGB)"=&VectorFromFloat(Value=0.6), "Float (Alpha)"=0.2)
    )
    Particles.GlowFoot_Size:Vector2f=<100 100>
   ) {isSetParameters=true}
   /SetVariables(Particles.ColorGlow:LinearColor=&MultiplyLinearColor(
     Linear Color=$System.Color_Smoothed
     Scale Factor=&MakeFloat4FromVectorAndFloat("Vector (RGB)"=&VectorFromFloat(Value=0.6), "Float (Alpha)"=0.2)
    )) {isSetParameters=true}
   &SkeletalMeshLocation(
    Bone / Socket Index=&SelectIntFromArray(Direct Array Index=$Particles.PartitionParticles.Partition, Int Selection Array=DataInterface() {propertyValues {IntData [8,9,11]}})
    Bone Sampling Mode=::NewEnumerator4
    MaskSpawnGroup=false
    Sampled Position Offset Coordinate Space:ENiagaraSkelMeshTransforms=::NewEnumerator2
    Skeletal Mesh=$User.SkeletalMesh
   ) {module=SkeletalMeshLocation001}
   /SetVariables(Particles.VisibilityTag=&SelectIntFromArray(Direct Array Index=$Particles.PartitionParticles.Partition, Int Selection Array=DataInterface() {propertyValues {IntData [0,0,1]}})) {isSetParameters=true}
   /SetVariables(
    Particles.CameraOffsetGlow=30
    Particles.VisibilityLarge:NiagaraInt32=&SelectIntFromArray(Direct Array Index=$Particles.PartitionParticles.Partition, Int Selection Array=DataInterface() {propertyValues {IntData [1,1,0]}})
   ) {isSetParameters=true}
  ]
  particleUpdateScript [
   &ParticleState(
    Kill Particles When Lifetime Has Elapsed=false
    Let Infinitely Lived Particles Die When Emitter Deactivates=true
    Loop Particles Lifetime=true
   )
   &SpriteRotationRate(Rotation Rate=&SelectFloatFromArray(Direct Array Index=$Particles.PartitionParticles.Partition, Float Selection Array=DataInterface() {propertyValues {FloatData [24,24,0]}}))
   &SkeletalMeshLocation(
    Bone / Socket Index=&SelectIntFromArray(Direct Array Index=$Particles.PartitionParticles.Partition, Int Selection Array=DataInterface() {propertyValues {IntData [8,9,11]}})
    Bone Sampling Mode=::NewEnumerator4
    MaskSpawnGroup=false
    Sampled Position Offset=&SelectVectorFromArray(
     Direct Array Index=$Particles.PartitionParticles.Partition
     Vector Selection Array=DataInterface() {propertyValues {FloatData [<0 0 0>, <0 0 0>, <0 0 -15>]}}
    )
    Sampled Position Offset Coordinate Space:ENiagaraSkelMeshTransforms=::NewEnumerator2
    Skeletal Mesh=$User.SkeletalMesh
   )
   &DynamicMaterialParameters(
    Index 0 Param 1=&FloatFromCurve(FloatCurve=DataInterface() {propertyValues {Curve {keys [<at 0 1 RCIM_Cubic>,<at 0.8 0 RCIM_Cubic>]}}})
    Index 0 Param 2=2
    Index 0 Param 3=1
    Index 0 Param 4=1
    Index 1 Param 1=&FloatFromCurve(FloatCurve=DataInterface() {propertyValues {Curve {keys [
         {arriveTangent=-2.3163, interpMode=RCIM_Cubic, time=0, value=2}
         {arriveTangent=-1.9019, interpMode=RCIM_Cubic, leaveTangent=-1.9019, tangentMode=RCTM_User, time=1, value=0.4}
        ]}}})
    Index 1 Param 2=0.4
    Index 1 Param 3=0.6
    Index 1 Param 4=1
    Index 2 Param 1=&FloatFromCurve(FloatCurve=DataInterface() {propertyValues {Curve {keys [<at 0.46 0>,<at 1 1>]}}})
    Index 2 Param 2=45
    Module.Param0WriteEnabled=true
    Module.Param0WriteEnabled1=false
    Module.Param0WriteEnabled2=true
    Module.Param1WriteEnabled=true
    Module.Param1WriteEnabled1=true
    Module.Param1WriteEnabled2=true
    Module.Param2WriteEnabled=true
    Module.Param2WriteEnabled1=true
    Module.Param2WriteEnabled2=true
    Module.Param3WriteEnabled=true
    Module.Param3WriteEnabled1=true
    Module.Param3WriteEnabled2=true
    Write Parameter Index 1=true
    Write Parameter Index 2=true
   )
   /SetVariables(Particles.DynMatGlow:Vector4f=<3 1 0 0>) {isSetParameters=true}
   &ScaleColor(
    Scale Mode=::NewEnumerator1
    Scale RGBA=&MakeVector4_FromLinearColor(LinearColor=$System.Color_Smoothed)
    ScaleRGBA=true
   )
   &ScaleColor(
    Scale RGB=&VectorFromFloat(Value=&FloatFromCurve(
      FloatCurve=DataInterface() {propertyValues {Curve {keys [
          {interpMode=RCIM_Cubic, tangentMode=RCTM_User, time=0, value=1}
          {interpMode=RCIM_Cubic, tangentMode=RCTM_User, time=0.5, value=1.5}
          {tangentMode=RCTM_User, time=1, value=1}
         ]}}}
      Scale Curve=&SelectFloatFromArray(Direct Array Index=$User.ColorIndex, Float Selection Array=DataInterface() {propertyValues {FloatData [1,1,1,1,1,1,1]}})
     ))
    ScaleA=false
    ScaleRGB=true
   ) {module=ScaleColor002}
   /SetVariables(Particles.GlowFoot_RailColor:LinearColor=&MultiplyLinearColor(
     Linear Color=$Particles.Initial.GlowFoot_RailColor
     Scale Factor=&MakeFloat4FromVectorAndFloat(
      "Vector (RGB)"=&Multiply_VectorByFloat(
       Vector=&Multiply_Vector(
        A=&VectorFromFloat(Value=&FNTimeOfDayScalar(Night Scale=0.35, Morning Scale=0.2, Day Scale=0.4, Evening Scale=0.3))
        B=&VectorFromFloat(Value=&Add_Float(A=&Multiply_Float(A=$System.FN_Timer.Value, B=1.5), B=0.75))
       )
       Float=&SelectFloatFromArray(Direct Array Index=$User.ColorIndex, Float Selection Array=DataInterface() {propertyValues {FloatData [0.025,0.15,0.2,0.1,0.1,0.3,0.1]}})
      )
      "Float (Alpha)"=&MakeCustomFloatFromBool(Bool=$User.Jumped, True Float=0, False Float=1)
     )
    )) {isSetParameters=true}
   &ScaleColor(
    Scale RGB=&Multiply_Vector(
     A=&VectorFromFloat(Value=&FNTimeOfDayScalar(Night Scale=0.2, Morning Scale=0.3, Day Scale=0.4, Evening Scale=0.4))
     B=&VectorFromFloat(Value=&Add_Float(A=&Multiply_Float(A=$System.FN_Timer.Value, B=2.5), B=0.75))
    )
    ScaleA=true
    ScaleRGB=true
   ) {module=ScaleColor001}
   &FN_Light_Attributes(
    Color=&SelectLinearColorFromArray(Array Sampling Mode=::NewEnumerator1, Direct Array Index=$User.ColorIndex, Color Selection Array=$System.GrindRail_ColorArray, Random Seed=?)
    Create Color Attribute=true
    Enable Particle Light Enable=true
    Scale Color=&MakeCustomFloatFromBool(Bool=$User.Jumped, True Float=0, False Float=20)
   )
   /SetVariables(Particles.ColorGlow:LinearColor=&MultiplyLinearColor(
     Linear Color=$System.Color_Smoothed
     Scale Factor=&MakeFloat4FromVectorAndFloat(
      "Vector (RGB)"=&VectorFromFloat(Value=&FloatFromCurve(
        FloatCurve=DataInterface() {propertyValues {Curve {keys [
            {interpMode=RCIM_Cubic, tangentMode=RCTM_User, time=0, value=1}
            {interpMode=RCIM_Cubic, tangentMode=RCTM_User, time=0.5, value=2.5}
            {tangentMode=RCTM_User, time=1, value=1}
           ]}}}
        Scale Curve=&FNTimeOfDayScalar(Night Scale=0.5, Morning Scale=0.4, Day Scale=0.65, Evening Scale=0.6)
       ))
      "Float (Alpha)"=&Multiply_Float(A=1, B=&Add_Float(A=0.2, B=&Multiply_Float(A=$System.FN_Timer.Value, B=0.2)))
     )
    )) {isSetParameters=true}
  ]
  rendererClasses [&NiagaraSpriteRendererProperties, &NiagaraLightRendererProperties]
  renderers [
   {rendererClass=&NiagaraSpriteRendererProperties}
   {rendererClass=&NiagaraSpriteRendererProperties, rendererIndex=1}
   {rendererClass=&NiagaraLightRendererProperties, rendererIndex=2}
   {rendererClass=&NiagaraSpriteRendererProperties, rendererIndex=3}
  ]
 }
 GlowDrops_CPU_Local {
  emitterUpdateScript [
   &EmitterState(
    Enable Distance Culling=true
    Life Cycle Mode=::NewEnumerator1
    Loop Behavior=::NewEnumerator1
    Loop Delay=$User.Delay
    Loop Duration Mode=::NewEnumerator1
    Max Distance Response=::KillImmediately
    Module.bApplySpawnCountScaleByDistance=true
    Module.bUseMaxDistance=true
    Module.bUseMinDistance=false
    Scalability Mode=::NewEnumerator1
    Scale Spawn Count=true
    Spawn Count Scale=3
    Spawn Count Scale By Distance=DataInterface() {propertyValues {Curve {keys [<at 0.3676 0.9657>,<at 0.9962 0.4982>]}}}
    UseLoopDelay=true
   )
   &SpawnRate(
    SpawnRate=&FloatFromCurve(
     FloatCurve=DataInterface() {propertyValues {Curve {keys [
         <at 0 0 RCIM_Cubic>
         {arriveTangent=5.0918, interpMode=RCIM_Cubic, leaveTangent=5.0918, tangentMode=RCTM_User, time=0.1385, value=1}
         {arriveTangent=81.1635, interpMode=RCIM_Cubic, leaveTangent=81.1635, tangentMode=RCTM_User, time=0.2253, value=13.5091}
         <at 1 40>
        ]}}}
     CurveIndex=&MakeCustomFloatFromBool(
      Bool=$User.TurnON_Niagara_Velocity
      True Float=$System.NormalizedVelocityRange
      False Float=&NormalizeFloat(Value=&Add_Float(A=$User.Speed_Alpha, B=0.4), Min=0, Max=1)
     )
     Scale Curve=&Multiply_Float(
      A=&MakeCustomFloatFromBool(
       Bool=&SetBoolByFloatComparison(A=$System.NormalizedVelocityRange, B=0.65, Comparison Type=::NewEnumerator6)
       True Float=0
       False Float=&FloatFromCurve(
        FloatCurve=DataInterface() {propertyValues {Curve {keys [
            {interpMode=RCIM_Cubic, tangentMode=RCTM_User, time=0, value=0}
            {arriveTangent=2.6713, interpMode=RCIM_Cubic, leaveTangent=2.6713, tangentMode=RCTM_User, time=1, value=1}
           ]}}}
        CurveIndex=$System.NormalizedVelocityRange
       )
      )
      B=&Add_Float(A=$User.Speed_Alpha, B=0.4)
     )
    )
    Use Spawn Probability=false
   )
   &SpawnBurst_Instantaneous(
    Spawn Count=&RandomRangeInt(Minimum=5, Maximum=25)
    Use Spawn Probability=false
    UseLoopCountLimit=false
   )
  ]
  particleSpawnScript [
   /Niagara/Modules/Spawn/Initialization/V2/InitializeParticle(
    Color Mode=::NewEnumerator0
    Lifetime Max=0.5
    Lifetime Min=0.25
    Lifetime Mode=::NewEnumerator1
    Sprite Size Mode=::NewEnumerator1
    Uniform Sprite Size Max=4
    Uniform Sprite Size Min=2
    UsePositionOffset=false
    Write Material Random=true
   )
   &FN_SetSpawnGroup(Number of Spawn Groups=2)
   &SkeletalMeshLocation(
    Bone / Socket Index=&SelectIntFromArray(Direct Array Index=$Particles.SpawnGroup, Int Selection Array=DataInterface() {propertyValues {IntData [8,9]}})
    Bone Sampling Mode=::NewEnumerator4
    MaskSpawnGroup=false
    Sampled Position Offset=&SelectVectorFromArray(
     Direct Array Index=&Add_Integer(A=1, B=&MakeIntFromFloat(Float=$User.Leaning Direction))
     Vector Selection Array=DataInterface() {propertyValues {FloatData [<0 0 -10>, <0 5 -10>, <0 15 -5>]}}
    )
    Sampled Position Offset Coordinate Space:ENiagaraSkelMeshTransforms=::NewEnumerator2
    Skeletal Mesh=$User.SkeletalMesh
   ) {module=SkeletalMeshLocation001}
   &ShapeLocation(Box Size=<2 2.5 2>, MaskSpawnGroup=false, Override Seed=false, Shape Primitive=::NewEnumerator2, Surface Only=false, UseOwnerScale=true)
   &AddVelocity(
    Cone Axis=<-1 0 0.5>
    Module.UseConeFalloff=false
    Module.UseVelDistribution=false
    Override Seed=false
    Velocity Mode=::NewEnumerator2
    Velocity Speed=&Multiply_Float(
     A=&Multiply_Float(A=&RandomRangeFloat(Minimum=250, Maximum=500), B=1)
     B=1
    )
   )
  ]
  particleUpdateScript [
   &ParticleState()
   &GravityForce(Gravity=&RandomRangeVector(Minimum=<0 0 -920>, Maximum=<0 0 -480>))
   &GravityForce(Coordinate Space=::Local, Gravity=&RandomRangeVector(Minimum=<-250 0 0>, Maximum=<-500 0 0>)) {module=GravityForce001}
   &Drag(
    Drag=&Multiply_Float(
     A=&UniformRangedFloat(Minimum=2, Maximum=4)
     B=&Lerp_Float(A=1, B=2, Alpha=$System.FN_Timer.Value)
    )
    Use Linear Drag=true
    Use Rotational Drag=false
   )
   &SolveForcesAndVelocity(Clamp Velocity=false, Limit Acceleration=false)
   &ScaleSpriteSize(Scale Factor=&Vector2DFromCurve(
     Vector2Curve=DataInterface() {propertyValues {
       XCurve {keys [<at 0 1 RCIM_Cubic>, {arriveTangent=-2.8959, interpMode=RCIM_Cubic, leaveTangent=-2.8959, tangentMode=RCTM_User, time=1, value=0}]}
       YCurve {keys [<at 0 1 RCIM_Cubic>, {arriveTangent=-2.8959, interpMode=RCIM_Cubic, leaveTangent=-2.8959, tangentMode=RCTM_User, time=1, value=0}]}
      }}
     ScaleCurve=<1 1>
    ))
   &ScaleSpriteSizeBySpeed(Max Scale Factor=<0.5 5>, Min Scale Factor=<0.75 1.5>)
   /SetVariables(Particles.Color=&MultiplyLinearColor(
     Linear Color=&self:LerpColor3(
      A:LinearColor=&MultiplyLinearColor(
       Linear Color=&MultiplyLinearColor(
        Linear Color=&SelectLinearColorFromArray(Array Sampling Mode=::NewEnumerator1, Direct Array Index=$User.ColorIndex, Color Selection Array=$System.GrindRail_ColorArray, Random Seed=?)
        Scale Factor=<50 50 50 1>
       )
       Scale Factor=&MakeFloat4FromVectorAndFloat(
        "Vector (RGB)"=&Multiply_Vector(
         A=&VectorFromFloat(Value=&SelectFloatFromArray(Direct Array Index=$User.ColorIndex, Float Selection Array=DataInterface() {propertyValues {FloatData [1,0.8,1,1,0.75,1,0.9]}}))
         B=&VectorFromFloat(Value=&FNTimeOfDayScalar(Night Scale=0.8, Morning Scale=1, Evening Scale=1))
        )
        "Float (Alpha)"=1
       )
      )
      Alpha:NiagaraFloat=$Particles.NormalizedAge
      B:LinearColor=&SelectLinearColorFromArray(
       Array Sampling Mode=::NewEnumerator1
       Direct Array Index=$User.ColorIndex
       Color Selection Array=DataInterface() {propertyValues {ColorData [
          <rgba 50 0.1 0 1>
          <rgba 30 23.8449 0 1>
          <rgba 30 14.3475 0 1>
          <rgba 30 1.6669 0 1>
          <rgba 30 11.5633 0 1>
          <rgba 30 0 6.9486 1>
          <rgba 50 0.1 0 1>
         ]}}
       Random Seed=?
      )
      C:LinearColor=&SelectLinearColorFromArray(
       Array Sampling Mode=::NewEnumerator1
       Direct Array Index=$User.ColorIndex
       Color Selection Array=DataInterface() {propertyValues {ColorData [<rgba 30 0 0 1>, <rgba 30 0.2 0 1>, <rgba 30 0.2 0 1>, <rgba 30 0 0 1>, <rgba 30 0 0 1>, <rgba 30 0 0 1>, <rgba 30 0 0 1>]}}
       Random Seed=?
      )
     )
     Scale Factor=&MakeVector4_FromFloat(Float=&MakeCustomFloatFromBool(Bool=$User.Jumped, True Float=0, False Float=1))
    )) {isSetParameters=true, module=SetVariables_#10}
  ]
  rendererClasses [&NiagaraSpriteRendererProperties]
  renderers [{rendererClass=&NiagaraSpriteRendererProperties}]
 }
 GlowDrops_GPU_Local_ {
  simTarget=GPUComputeSim
  emitterUpdateScript [
   &EmitterState(
    Enable Distance Culling=true
    Life Cycle Mode=::NewEnumerator1
    Loop Behavior=::NewEnumerator1
    Loop Delay=$User.Delay
    Loop Duration Mode=::NewEnumerator1
    Max Distance Response=::KillImmediately
    Module.bApplySpawnCountScaleByDistance=true
    Module.bUseMaxDistance=true
    Module.bUseMinDistance=false
    Scalability Mode=::NewEnumerator1
    Scale Spawn Count=true
    Spawn Count Scale=3
    Spawn Count Scale By Distance=DataInterface() {propertyValues {Curve {keys [<at 0.3676 0.9657>,<at 0.9962 0.4982>]}}}
    UseLoopDelay=true
   )
   &SpawnRate(
    SpawnRate=&FloatFromCurve(
     FloatCurve=DataInterface() {propertyValues {Curve {keys [
         <at 0 0 RCIM_Cubic>
         {arriveTangent=5.0918, interpMode=RCIM_Cubic, leaveTangent=5.0918, tangentMode=RCTM_User, time=0.1385, value=1}
         {arriveTangent=81.1635, interpMode=RCIM_Cubic, leaveTangent=81.1635, tangentMode=RCTM_User, time=0.2253, value=13.5091}
         <at 1 40>
        ]}}}
     CurveIndex=&MakeCustomFloatFromBool(
      Bool=$User.TurnON_Niagara_Velocity
      True Float=$System.NormalizedVelocityRange
      False Float=&NormalizeFloat(Value=&Add_Float(A=$User.Speed_Alpha, B=0.4), Min=0, Max=1)
     )
     Scale Curve=&Multiply_Float(
      A=&MakeCustomFloatFromBool(
       Bool=&SetBoolByFloatComparison(A=$System.NormalizedVelocityRange, B=0.65, Comparison Type=::NewEnumerator6)
       True Float=0
       False Float=&FloatFromCurve(
        FloatCurve=DataInterface() {propertyValues {Curve {keys [
            {interpMode=RCIM_Cubic, tangentMode=RCTM_User, time=0, value=0}
            {arriveTangent=2.6713, interpMode=RCIM_Cubic, leaveTangent=2.6713, tangentMode=RCTM_User, time=1, value=1}
           ]}}}
        CurveIndex=$System.NormalizedVelocityRange
        Scale Curve=2
       )
      )
      B=&Add_Float(A=$User.Speed_Alpha, B=0.4)
     )
    )
    Use Spawn Probability=false
   )
   &SpawnBurst_Instantaneous(
    Spawn Count=&RandomRangeInt(Minimum=5, Maximum=25)
    Use Spawn Probability=false
    UseLoopCountLimit=false
   )
  ]
  particleSpawnScript [
   &InitializeParticle(
    Color Mode=::NewEnumerator0
    Lifetime Max=0.5
    Lifetime Min=0.25
    Lifetime Mode=::NewEnumerator1
    Sprite Size Mode=::NewEnumerator1
    Uniform Sprite Size Max=4
    Uniform Sprite Size Min=2
    UsePositionOffset=false
    Write Material Random=true
   )
   &FN_SetSpawnGroup(Number of Spawn Groups=2)
   &SkeletalMeshLocation(
    Bone / Socket Index=&SelectIntFromArray(Direct Array Index=$Particles.SpawnGroup, Int Selection Array=DataInterface() {propertyValues {IntData [8,9]}})
    Bone Sampling Mode=::NewEnumerator4
    MaskSpawnGroup=false
    Orientation Sampling=::NewEnumerator1
    Sampled Position Offset=&SelectVectorFromArray(
     Direct Array Index=&Add_Integer(A=1, B=&MakeIntFromFloat(Float=$User.Leaning Direction))
     Vector Selection Array=DataInterface() {propertyValues {FloatData [<0 0 -10>, <0 5 -10>, <0 15 -5>]}}
    )
    Sampled Position Offset Coordinate Space:ENiagaraSkelMeshTransforms=::NewEnumerator2
    Skeletal Mesh=$User.SkeletalMesh
   ) {module=SkeletalMeshLocation001}
   &ShapeLocation(Box Size=<2 2.5 2>, MaskSpawnGroup=false, Override Seed=false, Shape Primitive=::NewEnumerator2, Surface Only=false, UseOwnerScale=true)
   &AddVelocity(
    Cone Axis=<-1 0 0.5>
    Module.UseConeFalloff=false
    Module.UseVelDistribution=false
    Override Seed=false
    Velocity Mode=::NewEnumerator2
    Velocity Speed=&Multiply_Float(
     A=&Multiply_Float(A=&RandomRangeFloat(Minimum=250, Maximum=500), B=1)
     B=1
    )
   )
  ]
  particleUpdateScript [
   &ParticleState()
   &GravityForce(Gravity=&RandomRangeVector(Minimum=<0 0 -920>, Maximum=<0 0 -480>))
   &GravityForce(Coordinate Space=::Local, Gravity=&RandomRangeVector(Minimum=<-250 0 0>, Maximum=<-500 0 0>)) {module=GravityForce001}
   &Drag(
    Drag=&Multiply_Float(
     A=&UniformRangedFloat(Minimum=2, Maximum=4)
     B=&Lerp_Float(A=1, B=2, Alpha=$System.FN_Timer.Value)
    )
    Use Linear Drag=true
    Use Rotational Drag=false
   )
   &CurlNoiseForce(Noise Frequency=5, Noise Strength=2.5, Pan Noise=false)
   &SolveForcesAndVelocity(Clamp Velocity=false, Limit Acceleration=false)
   &ScaleSpriteSize(Scale Factor=&Vector2DFromCurve(
     Vector2Curve=DataInterface() {propertyValues {
       XCurve {keys [<at 0 1 RCIM_Cubic>, {arriveTangent=-2.8959, interpMode=RCIM_Cubic, leaveTangent=-2.8959, tangentMode=RCTM_User, time=1, value=0}]}
       YCurve {keys [<at 0 1 RCIM_Cubic>, {arriveTangent=-2.8959, interpMode=RCIM_Cubic, leaveTangent=-2.8959, tangentMode=RCTM_User, time=1, value=0}]}
      }}
     ScaleCurve=<1 1>
    ))
   &ScaleSpriteSizeBySpeed(Max Scale Factor=<0.5 5>, Min Scale Factor=<0.75 1.5>)
   /SetVariables(Particles.Color=&MultiplyLinearColor(
     Linear Color=&self:LerpColor3(
      A:LinearColor=&MultiplyLinearColor(
       Linear Color=&MultiplyLinearColor(
        Linear Color=&SelectLinearColorFromArray(Array Sampling Mode=::NewEnumerator1, Direct Array Index=$User.ColorIndex, Color Selection Array=$System.GrindRail_ColorArray, Random Seed=?)
        Scale Factor=<50 50 50 1>
       )
       Scale Factor=&MakeFloat4FromVectorAndFloat(
        "Vector (RGB)"=&Multiply_Vector(
         A=&VectorFromFloat(Value=&SelectFloatFromArray(Direct Array Index=$User.ColorIndex, Float Selection Array=DataInterface() {propertyValues {FloatData [1,0.8,1,1,0.75,1,0.9]}}))
         B=&VectorFromFloat(Value=&FNTimeOfDayScalar(Night Scale=0.8, Morning Scale=1, Evening Scale=1))
        )
        "Float (Alpha)"=1
       )
      )
      Alpha:NiagaraFloat=$Particles.NormalizedAge
      B:LinearColor=&SelectLinearColorFromArray(
       Array Sampling Mode=::NewEnumerator1
       Direct Array Index=$User.ColorIndex
       Color Selection Array=DataInterface() {propertyValues {ColorData [
          <rgba 50 0.1 0 1>
          <rgba 30 23.8449 0 1>
          <rgba 30 14.3475 0 1>
          <rgba 30 1.6669 0 1>
          <rgba 30 11.5633 0 1>
          <rgba 30 0 6.9486 1>
          <rgba 50 0.1 0 1>
         ]}}
       Random Seed=?
      )
      C:LinearColor=&SelectLinearColorFromArray(
       Array Sampling Mode=::NewEnumerator1
       Direct Array Index=$User.ColorIndex
       Color Selection Array=DataInterface() {propertyValues {ColorData [<rgba 30 0 0 1>, <rgba 30 0.2 0 1>, <rgba 30 0.2 0 1>, <rgba 30 0 0 1>, <rgba 30 0 0 1>, <rgba 30 0 0 1>, <rgba 30 0 0 1>]}}
       Random Seed=?
      )
     )
     Scale Factor=&MakeVector4_FromFloat(Float=&MakeCustomFloatFromBool(Bool=$User.Jumped, True Float=0, False Float=1))
    )) {isSetParameters=true, module=SetVariables_#10}
  ]
  rendererClasses [&NiagaraSpriteRendererProperties]
  renderers [{rendererClass=&NiagaraSpriteRendererProperties}]
 }
 GlowDrops_GPU_World_ {
  simTarget=GPUComputeSim
  emitterUpdateScript [
   &EmitterState(
    Enable Distance Culling=true
    Life Cycle Mode=::NewEnumerator1
    Loop Behavior=::NewEnumerator1
    Loop Delay=$User.Delay
    Loop Duration Mode=::NewEnumerator1
    Max Distance Response=::KillImmediately
    Module.bApplySpawnCountScaleByDistance=true
    Module.bUseMaxDistance=true
    Module.bUseMinDistance=false
    Scalability Mode=::NewEnumerator1
    Scale Spawn Count=true
    Spawn Count Scale=3
    Spawn Count Scale By Distance=DataInterface() {propertyValues {Curve {keys [<at 0.3676 0.9657>,<at 0.9962 0.4982>]}}}
    UseLoopDelay=true
   )
   &SpawnRate(
    SpawnRate=&FloatFromCurve(
     FloatCurve=DataInterface() {propertyValues {Curve {keys [
         <at 0 0 RCIM_Cubic>
         {arriveTangent=5.0918, interpMode=RCIM_Cubic, leaveTangent=5.0918, tangentMode=RCTM_User, time=0.1385, value=1}
         {arriveTangent=81.1635, interpMode=RCIM_Cubic, leaveTangent=81.1635, tangentMode=RCTM_User, time=0.2253, value=13.5091}
         <at 1 40>
        ]}}}
     CurveIndex=&MakeCustomFloatFromBool(
      Bool=$User.TurnON_Niagara_Velocity
      True Float=$System.NormalizedVelocityRange
      False Float=&NormalizeFloat(Value=&Add_Float(A=$User.Speed_Alpha, B=0.4), Min=0, Max=1)
     )
     Scale Curve=&Multiply_Float(
      A=&MakeCustomFloatFromBool(
       Bool=&SetBoolByFloatComparison(
        A=$System.NormalizedVelocityRange
        B=&MakeCustomFloatFromBool(Bool=$User.SlowDown, True Float=0.3, False Float=0.65)
        Comparison Type=::NewEnumerator6
       )
       True Float=0
       False Float=&FloatFromCurve(
        FloatCurve=DataInterface() {propertyValues {Curve {keys [
            {interpMode=RCIM_Cubic, tangentMode=RCTM_User, time=0, value=0}
            {arriveTangent=2.6713, interpMode=RCIM_Cubic, leaveTangent=2.6713, tangentMode=RCTM_User, time=1, value=1}
           ]}}}
        CurveIndex=$System.NormalizedVelocityRange
        Scale Curve=8
       )
      )
      B=&Add_Float(A=$User.Speed_Alpha, B=0.4)
     )
    )
    Use Spawn Probability=false
   )
  ]
  particleSpawnScript [
   &InitializeParticle(
    Color Mode=::NewEnumerator0
    Lifetime=&Multiply_Float(
     A=&RandomRangeFloat(Minimum=0.05, Maximum=0.15)
     B=&MakeCustomFloatFromBool(Bool=$User.SlowDown, True Float=2, False Float=1)
    )
    Sprite Size Mode=::NewEnumerator1
    Uniform Sprite Size Max=4
    Uniform Sprite Size Min=2
    UsePositionOffset=false
    Write Material Random=true
   )
   &FN_SetSpawnGroup(Number of Spawn Groups=2)
   &SkeletalMeshLocation(
    Bone / Socket Index=&SelectIntFromArray(Direct Array Index=$Particles.SpawnGroup, Int Selection Array=DataInterface() {propertyValues {IntData [8,9]}})
    Bone Sampling Mode=::NewEnumerator4
    MaskSpawnGroup=false
    Orientation Sampling=::NewEnumerator1
    Sampled Position Offset=&SelectVectorFromArray(
     Direct Array Index=&Add_Integer(A=1, B=&MakeIntFromFloat(Float=$User.Leaning Direction))
     Vector Selection Array=DataInterface() {propertyValues {FloatData [<0 0 -10>, <0 5 -10>, <0 15 -5>]}}
    )
    Sampled Position Offset Coordinate Space:ENiagaraSkelMeshTransforms=::NewEnumerator2
    Skeletal Mesh=$User.SkeletalMesh
   ) {module=SkeletalMeshLocation001}
   &ShapeLocation(Box Size=<2 2.5 2>, MaskSpawnGroup=false, Override Seed=false, Shape Primitive=::NewEnumerator2, Surface Only=false, UseOwnerScale=true)
   &AddVelocity(
    Cone Axis=<-1 0 0.5>
    Module.UseConeFalloff=false
    Module.UseVelDistribution=false
    Override Seed=false
    Velocity Mode=::NewEnumerator2
    Velocity Speed=&Multiply_Float(
     A=&Multiply_Float(A=&RandomRangeFloat(Minimum=250, Maximum=500), B=1)
     B=1
    )
   )
  ]
  particleUpdateScript [
   &ParticleState()
   &GravityForce(Gravity=&RandomRangeVector(Minimum=<0 0 -512>, Maximum=<0 0 -256>))
   &GravityForce(Coordinate Space=::Local, Gravity=&RandomRangeVector(Minimum=<-250 0 0>, Maximum=<-500 0 0>)) {module=GravityForce001}
   &Drag(
    Drag=&Multiply_Float(
     A=&UniformRangedFloat(Minimum=2, Maximum=4)
     B=&Lerp_Float(A=1, B=2, Alpha=$System.FN_Timer.Value)
    )
    Use Linear Drag=true
    Use Rotational Drag=false
   )
   &InheritSourceMovement(
    Applied Position Scale=&Multiply_VectorByFloat(
     Vector=<1 1 1>
     Float=&FloatFromCurve(
      FloatCurve=DataInterface() {propertyValues {Curve {keys [
          {arriveTangent=3.4383, interpMode=RCIM_Cubic, leaveTangent=3.4383, tangentMode=RCTM_User, time=0, value=0.25}
          {arriveTangent=1.6145, interpMode=RCIM_Cubic, leaveTangent=1.6145, tangentMode=RCTM_User, time=0.2, value=0.7}
          {arriveTangent=-0.0394, interpMode=RCIM_Cubic, leaveTangent=-0.0394, tangentMode=RCTM_User, time=1, value=1}
         ]}}}
      CurveIndex=$System.NormalizedVelocityRange
      Scale Curve=0.75
     )
    )
    ApplyForce=false
    ApplyPosition=true
    ApplyPositionFalloff=false
   )
   &CurlNoiseForce(Noise Frequency=5, Noise Strength=2.5, Pan Noise=false)
   &SolveForcesAndVelocity(Clamp Velocity=false, Limit Acceleration=false)
   &ScaleSpriteSize(Scale Factor=&Vector2DFromCurve(
     Vector2Curve=DataInterface() {propertyValues {
       XCurve {keys [<at 0 1 RCIM_Cubic>, {arriveTangent=-2.8959, interpMode=RCIM_Cubic, leaveTangent=-2.8959, tangentMode=RCTM_User, time=1, value=0}]}
       YCurve {keys [<at 0 1 RCIM_Cubic>, {arriveTangent=-2.8959, interpMode=RCIM_Cubic, leaveTangent=-2.8959, tangentMode=RCTM_User, time=1, value=0}]}
      }}
     ScaleCurve=<1 1>
    ))
   &ScaleSpriteSizeBySpeed(Max Scale Factor=<0.5 5>, Min Scale Factor=<0.75 1.5>)
   /SetVariables(Particles.Color=&MultiplyLinearColor(
     Linear Color=&self:LerpColor3(
      A:LinearColor=&MultiplyLinearColor(
       Linear Color=&MultiplyLinearColor(
        Linear Color=&SelectLinearColorFromArray(Array Sampling Mode=::NewEnumerator1, Direct Array Index=$User.ColorIndex, Color Selection Array=$System.GrindRail_ColorArray, Random Seed=?)
        Scale Factor=<50 50 50 1>
       )
       Scale Factor=&MakeFloat4FromVectorAndFloat(
        "Vector (RGB)"=&Multiply_Vector(
         A=&VectorFromFloat(Value=&SelectFloatFromArray(Direct Array Index=$User.ColorIndex, Float Selection Array=DataInterface() {propertyValues {FloatData [1,0.8,1,1,0.75,1,0.9]}}))
         B=&VectorFromFloat(Value=&FNTimeOfDayScalar(Night Scale=0.8, Morning Scale=1, Evening Scale=1))
        )
        "Float (Alpha)"=1
       )
      )
      Alpha:NiagaraFloat=$Particles.NormalizedAge
      B:LinearColor=&SelectLinearColorFromArray(
       Array Sampling Mode=::NewEnumerator1
       Direct Array Index=$User.ColorIndex
       Color Selection Array=DataInterface() {propertyValues {ColorData [
          <rgba 50 0.1 0 1>
          <rgba 30 23.8449 0 1>
          <rgba 30 14.3475 0 1>
          <rgba 30 1.6669 0 1>
          <rgba 30 11.5633 0 1>
          <rgba 30 0 6.9486 1>
          <rgba 50 0.1 0 1>
         ]}}
       Random Seed=?
      )
      C:LinearColor=&SelectLinearColorFromArray(
       Array Sampling Mode=::NewEnumerator1
       Direct Array Index=$User.ColorIndex
       Color Selection Array=DataInterface() {propertyValues {ColorData [<rgba 30 0 0 1>, <rgba 30 0.2 0 1>, <rgba 30 0.2 0 1>, <rgba 30 0 0 1>, <rgba 30 0 0 1>, <rgba 30 0 0 1>, <rgba 30 0 0 1>]}}
       Random Seed=?
      )
     )
     Scale Factor=&MakeVector4_FromFloat(Float=&MakeCustomFloatFromBool(Bool=$User.Jumped, True Float=0, False Float=1))
    )) {isSetParameters=true}
  ]
  rendererClasses [&NiagaraSpriteRendererProperties]
  renderers [{rendererClass=&NiagaraSpriteRendererProperties}]
 }
 GrindRail_Hands {
  emitterUpdateScript [
   /SetVariables(Emitter.EmitterRandom:NiagaraFloat=&RandomRangeFloat(Minimum=0, Maximum=1)) {isSetParameters=true}
   &EmitterState(Enable Distance Culling=true, MaxDistance=9000, Module.bUseMaxDistance=true, Module.bUseMinDistance=false, Scalability Mode=::NewEnumerator1)
   &FN_SpawnRibbon_v2(
    Number of Ribbons=2
    Spawn Mode=::NewEnumerator0
    SpawnRate:NiagaraFloat=&FloatFromCurve(
     FloatCurve=DataInterface() {propertyValues {Curve {keys [{arriveTangent=1459.5123, interpMode=RCIM_Cubic, leaveTangent=1459.5123, tangentMode=RCTM_User, time=0, value=0}, <at 0.01 25>, <at 1 40>]}}}
     CurveIndex=&MakeCustomFloatFromBool(
      Bool=$User.TurnON_Niagara_Velocity
      True Float=$System.NormalizedVelocityRange
      False Float=&Add_Float(A=$User.Speed_Alpha, B=0.75)
     )
     Scale Curve=2
    )
   )
  ]
  particleSpawnScript [
   &FN_InitializeRibbon_v3(
    Color:LinearColor=$System.Color_Smoothed
    Enable Skeletal Position=true
    Lifetime:NiagaraFloat=&Lerp_Float(A=&Lerp_Float(A=0.15, B=0.2, Alpha=$System.FN_Timer001.Value), B=0.2, Alpha=$System.FN_Timer.Value)
    Ribbon Width=20
    Sampled Position Coordinate Space:ENiagaraCoordinateSpace=::Local
    Sampled Rotation Coordinate Space:ENiagaraCoordinateSpace=::Local
    Sampled Velocity Coordinate Space:ENiagaraCoordinateSpace=::Local
   ) {module=FN_InitializeRibbon_v3001}
   &SkeletalMeshLocation(
    Bone / Socket Index=&SelectIntFromArray(Direct Array Index=&SpawnGroup(), Int Selection Array=DataInterface() {propertyValues {IntData [3,4]}})
    Bone Sampling Mode=::NewEnumerator4
    Interpolate Mesh Sampling=false
    MaskSpawnGroup=false
    Orientation Sampling=::NewEnumerator1
    Sampled Position Offset=<0 -2 -3>
    Sampled Position Offset Coordinate Space:ENiagaraSkelMeshTransforms=::NewEnumerator2
    Skeletal Mesh=$User.SkeletalMesh
   )
  ]
  particleUpdateScript [
   &SolveForcesAndVelocity(Clamp Velocity=false, Limit Acceleration=false)
   &DynamicMaterialParameters(
    Index 0 Param 1=&FloatFromCurve(FloatCurve=DataInterface() {propertyValues {Curve {keys [
         {arriveTangent=15.9564, interpMode=RCIM_Cubic, leaveTangent=15.9564, tangentMode=RCTM_User, time=0, value=0}
         {interpMode=RCIM_Cubic, tangentMode=RCTM_User, time=0.1, value=1}
         {interpMode=RCIM_Cubic, tangentMode=RCTM_User, time=1, value=0}
        ]}}})
    Index 0 Param 2=&FloatFromCurve(FloatCurve=DataInterface() {propertyValues {Curve {keys [
         {arriveTangent=24.5129, interpMode=RCIM_Cubic, leaveTangent=24.5129, tangentMode=RCTM_User, time=0, value=0}
         {arriveTangent=-5.3543, interpMode=RCIM_Cubic, leaveTangent=-5.3543, tangentMode=RCTM_User, time=0.1, value=1}
         {interpMode=RCIM_Cubic, tangentMode=RCTM_User, time=0.3, value=0}
        ]}}})
    Index 0 Param 3=-2
    Index 0 Param 4=1
    Module.Param0WriteEnabled=true
    Module.Param1WriteEnabled=true
    Module.Param2WriteEnabled=true
    Module.Param3WriteEnabled=true
   )
   &FN_UpdateRibbon(
    Enable Skeletal Position=true
    Initial Color:LinearColor=$Particles.Initial.Color
    Initial RibbonWidth:NiagaraFloat=$Particles.Initial.RibbonWidth
   )
   &ScaleColor(
    Scale Alpha=&FloatFromCurve(
     FloatCurve=DataInterface() {propertyValues {Curve {keys [
         {arriveTangent=15.9564, interpMode=RCIM_Cubic, leaveTangent=15.9564, tangentMode=RCTM_User, time=0, value=0}
         {arriveTangent=-3.4931, interpMode=RCIM_Cubic, leaveTangent=-3.4931, tangentMode=RCTM_User, time=0.0583, value=1.0864}
         {interpMode=RCIM_Cubic, tangentMode=RCTM_User, time=1, value=0}
        ]}}}
     Scale Curve=&SelectFloatFromArray(Direct Array Index=$System.Color_Index, Float Selection Array=DataInterface() {propertyValues {FloatData [2.5,1.5,5,2.5]}})
    )
    Scale RGB=&Add_Vector(
     A=&VectorFromFloat(Value=&FNTimeOfDayScalar(Night Scale=0.4, Morning Scale=0.4, Day Scale=0.7, Evening Scale=0.6))
     B=&SelectVectorFromArray(
      Direct Array Index=$System.Color_Index
      Vector Selection Array=DataInterface() {propertyValues {FloatData [<0 0 0>, <-0.1 1 0>, <0 0 0>, <0 0 0>]}}
     )
    )
    ScaleA=true
    ScaleRGB=true
   ) {module=ScaleColor001}
   &ScaleColor(
    Linear Color Curve=DataInterface() {propertyValues {
      AlphaCurve {keys [<at 0 0>,<at 0.0195 1>,<at 1 1>]}
      BlueCurve {keys [<at 0 2>,<at 0.2785 1>,<at 0.5448 1>]}
      GreenCurve {keys [<at 0 2>,<at 0.2785 0.383>,<at 0.5448 0.0113>]}
      RedCurve {keys [<at 0 2>,<at 0.2785 0.5713>,<at 0.5448 0>]}
     }}
    Scale Mode=::NewEnumerator2
   )
  ]
  rendererClasses [&NiagaraRibbonRendererProperties]
  renderers [{rendererClass=&NiagaraRibbonRendererProperties}]
 }
 GrindRail_Main {
  emitterUpdateScript [
   &EmitterState()
   &FN_SpawnRibbon_v2(
    Spawn Mode=::NewEnumerator0
    SpawnRate:NiagaraFloat=&FloatFromCurve(
     FloatCurve=DataInterface() {propertyValues {Curve {keys [{arriveTangent=1459.5123, interpMode=RCIM_Cubic, leaveTangent=1459.5123, tangentMode=RCTM_User, time=0, value=0}, <at 0.01 25>, <at 1 40>]}}}
     CurveIndex=&MakeCustomFloatFromBool(
      Bool=$User.TurnON_Niagara_Velocity
      True Float=$System.NormalizedVelocityRange
      False Float=&Add_Float(A=$User.Speed_Alpha, B=0.75)
     )
     Scale Curve=6
    )
   )
  ]
  particleSpawnScript [
   &FN_SetSpawnGroup(Number of Spawn Groups=5)
   &InitializeRibbon(
    Color=&MultiplyLinearColor(
     Linear Color=$System.Color_Smoothed
     Scale Factor=&MakeVector4_FromLinearColor(LinearColor=&RandomRangeLinearColor(Minimum=<rgba 44.844 46.133 50 0.075>, Maximum=<rgba 50 41.5857 44.6431 0.075>))
    )
    Lifetime=&SelectFloatFromArray(Direct Array Index=$Particles.SpawnGroup, Float Selection Array=DataInterface() {propertyValues {FloatData [0.15,0.2,0.25,0.15,0.15]}})
    Module.Write Color=true
    Module.Write Lifetime=true
    Module.Write Mass=true
    Module.Write Material Random=true
    Module.Write Position=false
    Module.Write RibbonTwist=true
    Module.Write RibbonWidth=true
    Ribbon Twist:NiagaraFloat=&SelectFloatFromArray(Direct Array Index=$Particles.SpawnGroup, Float Selection Array=DataInterface() {propertyValues {FloatData [0.5,0.25,1,1,-0.75]}})
    Ribbon Width=&SelectFloatFromArray(Direct Array Index=$Particles.SpawnGroup, Float Selection Array=DataInterface() {propertyValues {FloatData [80,30,50,35,50]}})
   )
   /SetVariables(Particles.RibbonDistanceFromStart:NiagaraFloat=$Emitter.DistanceTraveled) {isSetParameters=true}
   &SkeletalMeshLocation(
    Bone / Socket Index=&SelectIntFromArray(Direct Array Index=$Particles.SpawnGroup, Int Selection Array=DataInterface() {propertyValues {IntData [7,8,9,15,16]}})
    Bone Sampling Mode=::NewEnumerator4
    Interpolate Mesh Sampling=false
    MaskSpawnGroup=false
    Orientation Sampling=::NewEnumerator1
    Sampled Position Offset=&SelectVectorFromArray(
     Direct Array Index=$Particles.SpawnGroup
     Vector Selection Array=DataInterface() {propertyValues {FloatData [<5 -5 -25>, <0 0 10>, <15 10 15>, <0 0 0>, <0 0 0>]}}
    )
    Sampled Position Offset Coordinate Space:ENiagaraSkelMeshTransforms=::NewEnumerator2
    Skeletal Mesh=$User.SkeletalMesh
   )
  ]
  particleUpdateScript [
   &ParticleState()
   &SolveForcesAndVelocity(Clamp Velocity=false, Limit Acceleration=false)
   &ScaleColor(
    Scale Alpha=&FloatFromCurve(
     FloatCurve=DataInterface() {propertyValues {Curve {keys [
         {interpMode=RCIM_Cubic, tangentMode=RCTM_User, time=0, value=0}
         {interpMode=RCIM_Cubic, tangentMode=RCTM_User, time=0.4, value=0.2}
         {tangentMode=RCTM_User, time=1, value=0}
        ]}}}
     Scale Curve=&SelectFloatFromArray(Direct Array Index=$System.Color_Index, Float Selection Array=DataInterface() {propertyValues {FloatData [3,2,7,10]}})
    )
    Scale RGB=&Add_Vector(
     A=&VectorFromFloat(Value=&FNTimeOfDayScalar(Night Scale=0.4, Morning Scale=0.4, Day Scale=0.7, Evening Scale=0.6))
     B=&SelectVectorFromArray(
      Direct Array Index=$System.Color_Index
      Vector Selection Array=DataInterface() {propertyValues {FloatData [<0 0 0>, <-0.1 1 0>, <0 0 0>, <0 0 0>]}}
     )
    )
    ScaleA=true
    ScaleRGB=true
   ) {module=ScaleColor001}
   &ScaleColor(
    Linear Color Curve=DataInterface() {propertyValues {
      AlphaCurve {keys [<at 0 1>,<at 1 1>]}
      BlueCurve {keys [<at 0 1>,<at 0.2785 1>,<at 0.5448 1>]}
      GreenCurve {keys [<at 0 1>,<at 0.2785 0.383>,<at 0.5448 0.0113>]}
      RedCurve {keys [<at 0 1>,<at 0.2785 0.5713>,<at 0.5448 0>]}
     }}
    Scale Mode=::NewEnumerator2
   )
  ]
  rendererClasses [&NiagaraRibbonRendererProperties]
  renderers [{rendererClass=&NiagaraRibbonRendererProperties}]
 }
 Ring {
  emitterSpawnScript [/SetVariables(Emitter.GlowDepthFade=25) {isSetParameters=true}]
  emitterUpdateScript [
   &EmitterState(
    Enable Distance Culling=true
    Life Cycle Mode=::NewEnumerator1
    Loop Behavior=::NewEnumerator1
    Loop Delay=$User.Delay
    Loop Duration=2
    Max Distance Response=::KillImmediately
    MaxDistance=3500
    Module.bUseMaxDistance=true
    Module.bUseMinDistance=false
    Scalability Mode=::NewEnumerator1
    UseLoopDelay=true
   )
   &SpawnBurst_Instantaneous(Spawn Count=3, Use Spawn Probability=false)
  ]
  particleSpawnScript [
   &FN_SetSpawnGroup(Number of Spawn Groups=2)
   &PartitionParticles(Number Of Partitions=3)
   &InitializeParticle(
    Color=$User.Color_Default
    Lifetime=&SelectFloatFromArray(
     Direct Array Index=$Particles.PartitionParticles.Partition
     Float Selection Array=DataInterface() {propertyValues {FloatData [0.4,0.4,0.25]}}
    )
    Position Offset Coordinate Space=::Local
    Sprite Rotation Mode=::NewEnumerator2
    Sprite Size=&SelectVector2DFromArray(
     Direct Array Index=$Particles.PartitionParticles.Partition
     Vector 2D Selection Array=DataInterface() {propertyValues {FloatData [<300 300>, <300 300>, <100 150>]}}
    )
    Sprite Size Mode=::NewEnumerator4
    UsePositionOffset=true
    Write Material Random=true
   )
   &CameraOffset(Camera Offset Amount=-20)
   /SetVariables(Particles.CameraOffsetWrist=15) {isSetParameters=true}
   /SetVariables(
    Particles.VisibilityTag=&SelectIntFromArray(Direct Array Index=$Particles.PartitionParticles.Partition, Int Selection Array=DataInterface() {propertyValues {IntData [0,0,1]}})
    Particles.VisibilityWrist:NiagaraInt32=&SelectIntFromArray(Direct Array Index=$Particles.PartitionParticles.Partition, Int Selection Array=DataInterface() {propertyValues {IntData [1,1,0]}})
   ) {isSetParameters=true}
  ]
  particleUpdateScript [
   &ParticleState()
   &SkeletalMeshLocation(
    Bone / Socket Index=&SelectIntFromArray(Direct Array Index=$Particles.PartitionParticles.Partition, Int Selection Array=DataInterface() {propertyValues {IntData [8,9,2]}})
    Bone Sampling Mode=::NewEnumerator4
    MaskSpawnGroup=false
    Sampled Position Offset=<0 7.5 -7.5>
    Sampled Position Offset Coordinate Space:ENiagaraSkelMeshTransforms=::NewEnumerator2
    Skeletal Mesh=$User.SkeletalMesh
   )
   &ScaleColor(
    Linear Color Curve=DataInterface() {propertyValues {
      AlphaCurve {keys [<at 0 1>,<at 1 0>]}
      BlueCurve {keys [<at 0 1 RCIM_Cubic>,<at 0.9283 2 RCIM_Cubic>]}
      GreenCurve {keys [<at 0 1 RCIM_Cubic>,<at 0.9283 0.0853 RCIM_Cubic>]}
      RedCurve {keys [<at 0 1 RCIM_Cubic>,<at 0.9283 0 RCIM_Cubic>]}
     }}
    Scale Mode=::NewEnumerator2
   )
   &ScaleColor(
    Scale RGB=&Multiply_Vector(
     A=&VectorFromFloat(Value=&FNTimeOfDayScalar(Night Scale=0.35, Morning Scale=0.3, Day Scale=0.4, Evening Scale=0.4))
     B=<2 2 2>
    )
    ScaleA=true
    ScaleRGB=true
   ) {module=ScaleColor001}
   &ScaleSpriteSize(Scale Factor=&Vector2DFromFloat(Value=&FloatFromCurve(
      FloatCurve=DataInterface() {propertyValues {Curve {keys [{arriveTangent=2.7513, interpMode=RCIM_Cubic, leaveTangent=2.7513, tangentMode=RCTM_User, time=0, value=0}, <at 1 1.2>]}}}
      Scale Curve=0.5
     )))
   /SetVariables(Particles.WristSize:Vector2f=$Particles.Initial.SpriteSize) {isSetParameters=true}
   /SetVariables(
    Particles.FlareColor=&MultiplyLinearColor(Linear Color=$User.Color_Default, Scale Factor=&MakeVector4_FromFloat(Float=1))
    Particles.FlareSize:Vector2f=&Multiply_Vector2D_ByFloat(
     Float=&FloatFromCurve(FloatCurve=DataInterface() {propertyValues {Curve {keys [
          {interpMode=RCIM_Cubic, tangentMode=RCTM_User, time=0, value=0}
          {interpMode=RCIM_Cubic, tangentMode=RCTM_User, time=0.3, value=0.95}
          {tangentMode=RCTM_User, time=1, value=0}
         ]}}})
     Vector2D=$Particles.Initial.SpriteSize
    )
    Particles.FlashCameraOffset=20
   ) {isSetParameters=true}
  ]
  rendererClasses [&NiagaraSpriteRendererProperties]
  renderers [
   {rendererClass=&NiagaraSpriteRendererProperties}
   {rendererClass=&NiagaraSpriteRendererProperties, rendererIndex=1}
  ]
 }
 Sparks001 {
  emitterUpdateScript [
   &EmitterState(
    Enable Distance Culling=true
    Life Cycle Mode=::NewEnumerator1
    Loop Behavior=::NewEnumerator1
    Loop Delay=$User.Delay
    Loop Duration Mode=::NewEnumerator1
    MaxDistance=10000
    Module.bUseMaxDistance=true
    Module.bUseMinDistance=false
    Scalability Mode=::NewEnumerator1
    UseLoopDelay=true
   )
   &SpawnBurst_Instantaneous(
    Spawn Count=&RandomRangeInt(Minimum=2, Maximum=3)
    Spawn Probability=&OneMinusFloat(Float=$System.FN_Timer001.Value)
    Use Spawn Probability=true
   )
   &SpawnBurst_Instantaneous(
    Spawn Count=&RandomRangeInt(Minimum=2, Maximum=3)
    Spawn Probability=&OneMinusFloat(Float=$System.FN_Timer001.Value)
    Spawn Time=0.1
    Use Spawn Probability=true
   ) {module=SpawnBurst_Instantaneous002}
   &SpawnBurst_Instantaneous(
    Spawn Count=&RandomRangeInt(Minimum=5, Maximum=7)
    Spawn Probability=&OneMinusFloat(Float=$System.FN_Timer001.Value)
    Spawn Time=&RandomRangeFloat(Minimum=0.15, Maximum=0.25)
    Use Spawn Probability=true
   ) {module=SpawnBurst_Instantaneous001}
   &SpawnRate(
    SpawnRate=&Multiply_Float(
     A=&RandomRangeFloat(Minimum=25, Maximum=50)
     B=&Multiply_Float(
      A=$System.FN_Timer.Value
      B=&MakeCustomFloatFromBool(
       Bool=$User.Jumped
       True Float=0
       False Float=&Multiply_Float(A=&Add_Float(A=$User.Speed_Alpha, B=0.4), B=1)
      )
     )
    )
    Use Spawn Probability=false
   )
  ]
  particleSpawnScript [
   /SetVariables(Particles.Random=&RandomRangeFloat(Minimum=0, Maximum=1)) {isSetParameters=true}
   &FN_SetSpawnGroup(Number of Spawn Groups=2)
   /Niagara/Modules/Spawn/Initialization/InitializeParticle(
    Color=&MultiplyLinearColor(
     Linear Color=&SelectLinearColorFromArray(Array Sampling Mode=::NewEnumerator1, Direct Array Index=$User.ColorIndex, Color Selection Array=$System.GrindRail_ColorArray, Random Seed=?)
     Scale Factor=<1 1 1 1>
    )
    Lifetime=&Multiply_Float(A=&Lerp_Float(A=0.25, B=0.5, Alpha=$Particles.Random), B=0.7)
    Mesh Scale=&Multiply_Vector(
     A=&Multiply_Vector(
      A=&Lerp_Vector(A=<0.2 0.2 0.2>, B=<0.6 0.6 0.6>, Alpha=$Particles.Random)
      B=&SelectVectorFromArray(
       Array Sampling Mode=::NewEnumerator0
       Direct Array Index=$Particles.SpawnGroup
       Vector Selection Array=DataInterface() {propertyValues {FloatData [<1 1 0.4>, <1 1 -0.4>]}}
       Random Seed=0
      )
     )
     B=&VectorFromFloat(Value=0.5)
    )
    Module.Write Color=true
    Module.Write Lifetime=true
    Module.Write Mass=true
    Module.Write Material Random=true
    Module.Write Position=false
    Module.Write Scale=true
    Module.Write SpriteRotation=false
    Module.Write SpriteSize=false
    Position=&SimulationPosition()
   )
   &CameraOffset(Camera Offset Amount=5)
   &ShapeLocation(
    Box Size=<0 2 0>
    MaskSpawnGroup=false
    Non Uniform Scale=<0 0 0>
    Override Seed=false
    Shape Primitive=::NewEnumerator2
    Surface Only=false
    UseOwnerScale=true
   )
   &AddVelocity(
    Cone Angle=10
    Cone Axis=<-1 0 0.05>
    Module.UseConeFalloff=false
    Module.UseVelDistribution=false
    Override Seed=false
    Velocity Mode=::NewEnumerator2
    Velocity Speed=&Multiply_Float(
     A=&Multiply_Float(A=&RandomRangeFloat(Minimum=150, Maximum=300), B=1)
     B=1
    )
   )
   &DynamicMaterialParameters(
    Index 0 Param 2=&FloatFromCurve(
     FloatCurve=DataInterface() {propertyValues {Curve {keys [<at 0 0 RCIM_Cubic>, {arriveTangent=1, interpMode=RCIM_Cubic, leaveTangent=1, time=0.7035, value=0.3698}, <at 1 1>]}}}
     CurveIndex=&RandomRangeFloat(Minimum=0, Maximum=1)
    )
    Module.Param0WriteEnabled=true
    Module.Param1WriteEnabled=true
    Module.Param2WriteEnabled=true
    Module.Param3WriteEnabled=true
   ) {module=DynamicMaterialParameters001}
   &InitialMeshOrientation(
    Mesh Axis To Orient=<-1 0 0>
    Mesh Orientation Mode=::NewEnumerator2
    Rotation=&MakeVector(X=&RandomRangeFloat(Minimum=0, Maximum=1), Y=0, Z=0)
    Use Rotation Vector=true
   ) {module=InitialMeshOrientation001}
  ]
  particleUpdateScript [
   &ParticleState()
   &SkeletalMeshLocation(
    Bone / Socket Index=&SelectIntFromArray(Direct Array Index=$Particles.SpawnGroup, Int Selection Array=DataInterface() {propertyValues {IntData [8,9]}})
    Bone Sampling Mode=::NewEnumerator4
    MaskSpawnGroup=false
    Orientation Sampling=::NewEnumerator1
    Sampled Position Offset=&AddVectorToPosition(
     Position=&SelectVectorFromArray(Direct Array Index=$Particles.SpawnGroup, Vector Selection Array=DataInterface() {propertyValues {FloatData [<0 5 -10>, <-5 5 -10>]}})
     Vector=$Particles.ShapeLocation.ShapePosition
    )
    Sampled Position Offset Coordinate Space:ENiagaraSkelMeshTransforms=::NewEnumerator2
    Skeletal Mesh=$User.SkeletalMesh
   )
   &SolveForcesAndVelocity(Clamp Velocity=false, Limit Acceleration=false)
   &DynamicMaterialParameters(
    Index 0 Param 1=&FloatFromCurve(FloatCurve=DataInterface() {propertyValues {Curve {keys [<at 0 0>,<at 1 1>]}}})
    Module.Param0WriteEnabled=true
    Module.Param1WriteEnabled=false
    Module.Param2WriteEnabled=true
    Module.Param3WriteEnabled=true
   )
   &ScaleColor(
    Scale Alpha=&Multiply_Float(
     A=&MakeCustomFloatFromBool(Bool=$User.Jumped, True Float=0, False Float=1)
     B=&Lerp_Float(A=1, B=1, Alpha=$User.Is180Turning)
    )
    Scale RGB=&Multiply_Vector(
     A=&Multiply_VectorByFloat(
      Vector=&VectorFromFloat(Value=&FNTimeOfDayScalar(Night Scale=1.2, Morning Scale=0.9, Evening Scale=0.8))
      Float=&SelectFloatFromArray(Direct Array Index=$System.Color_Index, Float Selection Array=DataInterface() {propertyValues {FloatData [1.5,1,3,1]}})
     )
     B=&VectorFromFloat(Value=&SelectFloatFromArray(Direct Array Index=$User.ColorIndex, Float Selection Array=DataInterface() {propertyValues {FloatData [1.2,1.5,2.5,0.5,1,1,0.5]}}))
    )
    ScaleA=true
    ScaleRGB=true
   )
  ]
  rendererClasses [&NiagaraMeshRendererProperties]
  renderers [{rendererClass=&NiagaraMeshRendererProperties, rendererIndex=0}]
 }
 SpeedLine_Particles {
  emitterUpdateScript [
   &EmitterState(
    Delay First Loop Only=true
    Inactive Response=::NewEnumerator1
    Life Cycle Mode=::NewEnumerator1
    Loop Delay=0.3
    Loop Duration=1
    Recalculate Duration Each Loop=true
    UseLoopDelay=true
   )
   &SpawnRate(
    SpawnRate=&Multiply_Float(
     A=50
     B=&Multiply_Float(
      A=&Multiply_Float(A=&Add_Float(A=$User.Speed_Alpha, B=0.4), B=1)
      B=&FloatFromCurve(
       FloatCurve=DataInterface() {propertyValues {Curve {keys [<at 0 0 RCIM_Cubic>, {arriveTangent=2.9733, interpMode=RCIM_Cubic, leaveTangent=2.9733, tangentMode=RCTM_User, time=1, value=1}]}}}
       CurveIndex=$Emitter.Age
       Scale Curve=$System.FN_Timer.Value
      )
     )
    )
    Use Spawn Probability=false
   )
  ]
  particleSpawnScript [
   /Niagara/Modules/Spawn/Initialization/V2/InitializeParticle(
    Color=$System.Color_Smoothed
    Lifetime Max=0.2
    Lifetime Min=0.1
    Lifetime Mode=::NewEnumerator1
    Sprite Size Max=<10 70>
    Sprite Size Min=<1 42>
    Sprite Size Mode=::NewEnumerator2
    UsePositionOffset=false
    Write Material Random=true
   )
   &SkeletalMeshLocation(Bone / Socket Index=3, Bone Sampling Mode=::NewEnumerator4, MaskSpawnGroup=false, Skeletal Mesh=$User.SkeletalMesh)
   &ShapeLocation(
    MaskSpawnGroup=false
    Offset=<0 15 0>
    Offset Mode=::NewEnumerator0
    Override Seed=false
    Shape Origin:Vector3f=$StackContext.SkeletalMeshLocation.SampledPosition
    Sphere Radius=17
    UseOwnerScale=true
   )
   &AddVelocity(
    Rotation Coordinate Space=::World
    Velocity=$Engine.Owner.SystemXAxis
    Velocity Speed Scale=&RandomRangeFloat(Minimum=-1600, Maximum=-2200)
   )
  ]
  particleUpdateScript [
   &SolveForcesAndVelocity(Clamp Velocity=false, Limit Acceleration=false)
   &ParticleState()
   &ScaleColor(
    Scale Alpha=&FloatFromCurve(
     FloatCurve=DataInterface() {propertyValues {Curve {keys [
         {interpMode=RCIM_Cubic, leaveTangent=12.945, tangentMode=RCTM_Break, time=0, value=0}
         {interpMode=RCIM_Cubic, tangentMode=RCTM_Break, time=0.2, value=1}
         {arriveTangent=-1.8151, interpMode=RCIM_Cubic, tangentMode=RCTM_Break, time=1, value=0}
        ]}}}
     Scale Curve=&MakeCustomFloatFromBool(Bool=$User.Jumped, True Float=0, False Float=1)
    )
    Scale RGB=&VectorFromFloat(Value=0.25)
    ScaleA=true
    ScaleRGB=true
   )
   &ScaleSpriteSize(
    Non-Uniform Curve Sprite Scale=DataInterface() {propertyValues {
      XCurve {keys [{tangentMode=RCTM_User, time=0.55, value=1}, {tangentMode=RCTM_User, time=1, value=0}]}
      YCurve {keys [
        {interpMode=RCIM_Cubic, leaveTangent=0.9226, tangentMode=RCTM_Break, time=0, value=0.55}
        {arriveTangent=0.0265, interpMode=RCIM_Cubic, tangentMode=RCTM_Break, time=1, value=1}
       ]}
     }}
    Scale Sprite Size Mode=::NewEnumerator3
   )
   &ScaleSpriteSizeBySpeed(Max Scale Factor=<1 2>)
   &ScaleColor(
    Scale RGB=&VectorFromFloat(Value=&FNTimeOfDayScalar(Night Scale=0.25, Morning Scale=0.3, Day Scale=0.75, Evening Scale=0.5))
    ScaleA=true
    ScaleRGB=true
   ) {module=ScaleColor002}
  ]
  rendererClasses [&NiagaraSpriteRendererProperties]
  renderers [{rendererClass=&NiagaraSpriteRendererProperties}]
 }
 Speedlines {
  simTarget=GPUComputeSim
  emitterUpdateScript [
   &EmitterState()
   &SpawnRate(
    SpawnRate=&Multiply_Float(
     A=400
     B=&Multiply_Float(
      A=&Multiply_Float(A=&Add_Float(A=$User.Speed_Alpha, B=0.4), B=1)
      B=$System.FN_Timer.Value
     )
    )
    Use Spawn Probability=true
   )
  ]
  particleSpawnScript [
   /Niagara/Modules/Spawn/Initialization/InitializeParticle(
    Color=&LerpLinearColors(
     StartColor=<rgba 1 1 1 0.2>
     EndColor=<rgba 0 0 0 0.4>
     LerpFactor=&UniformAOrBFloat(A=0, Distribution Weight=0.15)
    )
    Lifetime=&UniformRangedFloat(Minimum=0.1, Maximum=0.15)
    Module.Write Color=true
    Module.Write Lifetime=true
    Module.Write Mass=true
    Module.Write Material Random=true
    Module.Write Position=false
    Module.Write Scale=false
    Module.Write SpriteRotation=false
    Module.Write SpriteSize=true
    Position=&SimulationPosition()
    Sprite Size=&UniformRangedVector2D(Minimum=<3 400>, Maximum=<1 500>)
   )
   &ShapeLocation(
    Handle Radius=0
    Large Radius=&RandomRangeFloat(Minimum=17, Maximum=23)
    MaskSpawnGroup=false
    Non Uniform Scale=<1 1.7 1>
    Override Seed=false
    Rotation Mode=::NewEnumerator1
    Shape Primitive=::NewEnumerator3
    UseOwnerScale=true
   )
   &AddVelocity(
    Module.Use Falloff Distance=false
    Override Seed=false
    Velocity Mode=::NewEnumerator1
    Velocity Speed=&UniformRangedFloat(Minimum=1000, Maximum=300)
   )
   &AddVelocity(Velocity=<1 0 0>, Velocity Speed Scale=-1500) {module=AddVelocity001}
  ]
  particleUpdateScript [
   &ParticleState()
   &SolveForcesAndVelocity(Clamp Velocity=false, Limit Acceleration=false)
   &ScaleSpriteSize(Scale Factor=&Vector2DFromCurve(
     Vector2Curve=DataInterface() {propertyValues {XCurve {keys [<at 0 0.5 RCIM_Cubic>,<at 0.5 1 RCIM_Cubic>]}, YCurve {keys [<at 0 1 RCIM_Cubic>,<at 0.5 1 RCIM_Cubic>]}}}
     ScaleCurve=&Vector2DFromFloat(Value=1)
    ))
   &ScaleColor(
    Scale Alpha=&Multiply_Float(
     A=&FloatFromCurve(
      FloatCurve=DataInterface() {propertyValues {Curve {keys [
          {arriveTangent=5.2641, interpMode=RCIM_Cubic, leaveTangent=5.2641, tangentMode=RCTM_User, time=0, value=0}
          {arriveTangent=0.7132, interpMode=RCIM_Cubic, leaveTangent=0.7132, tangentMode=RCTM_User, time=0.3, value=1.0207}
          {arriveTangent=-2.9069, interpMode=RCIM_Cubic, leaveTangent=-2.9069, tangentMode=RCTM_User, time=1, value=0}
         ]}}}
      Scale Curve=&MakeCustomFloatFromBool(Bool=$User.Jumped, True Float=0, False Float=1)
     )
     B=&FNTimeOfDayScalar(Night Scale=0.1, Morning Scale=0.1, Day Scale=0.1, Evening Scale=0.1)
    )
    ScaleA=true
    ScaleRGB=true
   )
  ]
  rendererClasses [&NiagaraSpriteRendererProperties]
  renderers [{rendererClass=&NiagaraSpriteRendererProperties}]
 }
}
summary {
 systemName=NS_GrindRail_Trail_Body
 emitters [
  {emitterName=GrindRail_Main, rendererClasses [&NiagaraRibbonRendererProperties]}
  {emitterName=GrindRail_Hands, rendererClasses [&NiagaraRibbonRendererProperties]}
  {emitterName=Sparks001, rendererClasses [&NiagaraMeshRendererProperties]}
  {emitterName=SpeedLine_Particles, rendererClasses [&NiagaraSpriteRendererProperties]}
  {emitterName=Speedlines, simTarget=GPUComputeSim, rendererClasses [&NiagaraSpriteRendererProperties]}
  {emitterName=Air, rendererClasses [&NiagaraMeshRendererProperties]}
  {emitterName=GlowDrops_GPU_Local_, simTarget=GPUComputeSim, rendererClasses [&NiagaraSpriteRendererProperties]}
  {emitterName=DustSpikes001, rendererClasses [&NiagaraMeshRendererProperties]}
  {emitterName=GlowDrops_GPU_World_, simTarget=GPUComputeSim, rendererClasses [&NiagaraSpriteRendererProperties]}
  {emitterName=GlowDrops_CPU_Local, rendererClasses [&NiagaraSpriteRendererProperties]}
  {emitterName=Foot_Flare001, rendererClasses [&NiagaraSpriteRendererProperties, &NiagaraLightRendererProperties]}
  {emitterName=Ring, rendererClasses [&NiagaraSpriteRendererProperties]}
 ]
 userVariables [
  User.Boosting {type=NiagaraBool}
  User.Color_Boosting {type=LinearColor, description="", defaultValue {struct=LinearColor, value=<rgba 0 1 0 1>}}
  User.Color_Default {type=LinearColor, description="", defaultValue {struct=LinearColor, value=<rgba 0.7511 0 1 1>}}
  User.Color_SlowDown {type=LinearColor, description="", defaultValue {struct=LinearColor, value=<rgba 1 0 0 1>}}
  User.Color_Sprinting {type=LinearColor, description="", defaultValue {struct=LinearColor, value=<rgba 0 0.7703 1 1>}}
  User.ColorIndex {type=NiagaraInt32}
  User.Delay {type=NiagaraFloat, defaultValue=0, description=""}
  User.Is180Turning {type=NiagaraFloat, defaultValue=0, description=""}
  User.Jumped {type=NiagaraBool}
  User.Leaning Direction {type=NiagaraFloat, defaultValue=0, description=""}
  User.RibbonIDOffset {type=NiagaraInt32}
  User.SkeletalMesh {
   type=NiagaraDataInterfaceSkeletalMesh
   defaultValue {
    struct=NiagaraExt_VariableValue_DataInterface
    value {dataInterface=&self:NiagaraDataInterfaceSkeletalMesh_0}
   }
  }
  User.SlowDown {type=NiagaraBool}
  User.Speed_Alpha {type=NiagaraFloat, defaultValue=0.6, description=""}
  User.Sprinting {type=NiagaraBool}
  User.TurnON_Niagara_Velocity {type=NiagaraBool, defaultValue=true}
 ]
}
