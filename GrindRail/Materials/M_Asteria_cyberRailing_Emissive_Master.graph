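# Unlit emissive master material for cyber grind-rail spline meshes, lighting rails for idle, grinding, sprinting and jump-off states.
# Custom primitive data drives color, gameplay states, unified UV range and start/end glows.
# A sphere mask around the collection's player location, stepped noise speedlines, UV line masks and pulsing tail lights shape the glow.
# Intensity blends by pixel depth, then becomes HDR color divided by eye adaptation.
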Material /GrindRail/Materials/M_Asteria_cyberRailing_Emissive_Master
graph {
 properties {bUsedWithSplineMeshes=true, pixelDepthOffsetMode=PDOM_Legacy, shadingModel=MSM_Unlit, usedShadingModels=Unlit}
 nodes [
  n1 = VectorParameter() {bUseCustomPrimitiveData=true, group=EmissiveColor, parameterName=EmissiveColorLight, primitiveDataIndex=2, defaultValue=<rgba 1 0 0 1>}
  ComponentMask(None=@n1.RGB) {b=true, g=true, r=true}
  n2 = NamedRerouteUsage()
  n3 = Constant() {r=0.3}
  n4 = Panner(Coordinate=ComponentMask(None=@n2) {g=true}, Speed=AppendVector(A=Constant(), B=@n3)) {speedY=0.2}
  n5 = ComponentMask(None=OneMinus(None=Frac(None=@n4))) {g=true}
  n6 = ComponentMask(None=@n2) {g=true}
  n7 = AppendVector(A=Constant(), B=@n3)
  n8 = StaticComponentMaskParameter(None=Frac(None=Panner(Coordinate=@n6, Speed=@n7) {speedY=0.2})) {defaultG=true}
  n9 = Power(Base=OneMinus(None=Multiply(A=@n5, B=@n8))) {constExponent=25}
  MaterialFunctionCall(In=@n9, Contrast=Constant() {r=0.1}) {materialFunction=&CheapContrast}
  n10 = NamedRerouteUsage()
  n11 = MaterialFunctionCall(In=NamedRerouteUsage(), Contrast=Constant() {r=0.5}) {materialFunction=&CheapContrast}
  Max(A=@n10, B=@n11.Result)
  n12 = ScalarParameter() {defaultValue=0.5, group=EmissiveColor, parameterName=" lighting intensity Close up"}
  n13 = ScalarParameter() {defaultValue=15, group=EmissiveColor, parameterName=Idel lighting intensity far}
  n14 = Divide(A=PixelDepth().R, B=ScalarParameter() {defaultValue=20000, group=EmissiveColor, parameterName=BlendDistance}) {constB=5000}
  n15 = Saturate(None=Power(Base=Clamp(None=@n14), Exp=Constant() {r=1.5}) {constExponent=1.2})
  n16 = Constant() {r=800}
  n17 = ScalarParameter() {defaultValue=80, group=EmissiveColor, parameterName=Grinding light intensity}
  n18 = LinearInterpolate(Alpha=Add(A=@n10, B=@n11.Result)) {constA=0.7, constB=2}
  n19 = ScalarParameter() {defaultValue=300, group=EmissiveColor, parameterName=sprinting light intensity}
  n20 = NamedRerouteUsage()
  n21 = TextureCoordinate()
  n22 = Abs(None=Subtract(A=ComponentMask(None=@n21) {r=true}, B=Constant() {r=0.125}) {constB=0.1})
  n23 = MaterialFunctionCall(In=Power(Base=OneMinus(None=@n22), Exp=Constant() {r=120}) {constExponent=60}, Contrast=Constant() {r=0.02}) {materialFunction=&CheapContrast}
  n24 = Add(A=@n23.Result, B=Multiply(A=@n23.Result, B=NamedRerouteUsage()))
  n25 = LinearInterpolate(
   A=LinearInterpolate(A=LinearInterpolate(A=@n12, B=@n13, Alpha=@n15) {constB=100}, B=LinearInterpolate(B=@n16, Alpha=@n15) {constA=10}, Alpha=NamedRerouteUsage()) {constA=200, constB=200}
   B=LinearInterpolate(
    A=Multiply(A=@n17, B=@n18)
    B=Multiply(A=@n19, B=@n18)
    Alpha=ScalarParameter() {bUseCustomPrimitiveData=true, group=Gameplay, parameterName=sprinting state, primitiveDataIndex=8}
   )
   Alpha=LinearInterpolate(
    A=Multiply(A=@n20, B=@n24)
    Alpha=ScalarParameter() {bUseCustomPrimitiveData=true, group=Gameplay, parameterName=Jumpoff, primitiveDataIndex=7}
   ) {constB=0}
  )
  n26 = TextureSample() {samplerType=SAMPLERTYPE_LinearColor, texture=/GrindRail/Textures/T_Railing_E}
  n27 = LinearInterpolate(Alpha=Power(Base=OneMinus(None=@n26.R)) {constExponent=1.2})
  Multiply(A=@n25, B=LinearInterpolate(A=Saturate(None=@n27), Alpha=NamedRerouteUsage()))
  n28 = SphereMask(
   A=WorldPosition().XYZ
   B=ComponentMask(None=CollectionParameter() {collection=&/MPC_GrindRail, parameterName=Player location}) {b=true, g=true, r=true}
   Radius=LinearInterpolate(
    A=ScalarParameter() {defaultValue=250, group=EmissiveColor, parameterName=RadiusDefaultEffect}
    B=ScalarParameter() {defaultValue=400, group=EmissiveColor, parameterName=RadiusSprintingEffectRadius}
    Alpha=ScalarParameter() {bUseCustomPrimitiveData=true, group=Gameplay, parameterName=sprinting state, primitiveDataIndex=8}
   )
   Hardness=ScalarParameter() {defaultValue=0.1, group=EmissiveColor, parameterName=HardnessRadius}
  )
  NamedRerouteDeclaration(None=Saturate(None=Power(Base=@n28))) {name=Player SphereMask, nodeColor=<rgba 0.5 0 1 1>}
  n29 = ScalarParameter() {bUseCustomPrimitiveData=true, defaultValue=20, parameterName="Glow start  Intensity", primitiveDataIndex=11}
  n30 = NamedRerouteUsage()
  n31 = ComponentMask(None=SphereMask(
    A=ComponentMask(None=@n30) {g=true}
    B=Constant()
    Radius=ScalarParameter() {bUseCustomPrimitiveData=true, defaultValue=0.05, parameterName=Glow start Radius, primitiveDataIndex=9}
   ) {attenuationRadius=0.1, hardnessPercent=0.5}) {g=true}
  n32 = ComponentMask(None=SphereMask(
    A=ComponentMask(None=@n30) {g=true}
    B=Constant() {r=1}
    Radius=ScalarParameter() {bUseCustomPrimitiveData=true, defaultValue=0.05, parameterName=Glow End Radius, primitiveDataIndex=10}
   ) {attenuationRadius=0.1, hardnessPercent=0.5}) {g=true}
  n33 = ScalarParameter() {bUseCustomPrimitiveData=true, defaultValue=20, parameterName="Glow  end Intensity", primitiveDataIndex=12}
  NamedRerouteDeclaration(None=Max(A=Multiply(A=@n29, B=@n31), B=Multiply(A=@n32, B=@n33) {constB=10})) {name=Glow start and end, nodeColor=<rgba 0.4297 1 0 1>}
  n34 = TextureCoordinate()
  MaterialExpressionRemap_0 = MaterialXRemap!(
   None=ComponentMask(None=@n34) {g=true}
   InputLow
   InputHigh
   TargetLow=ScalarParameter() {bUseCustomPrimitiveData=true, parameterName=Unified UV Start}
   TargetHigh=ScalarParameter() {bUseCustomPrimitiveData=true, parameterName=Unified UV End, primitiveDataIndex=1}
  )
  NamedRerouteDeclaration(None=AppendVector(A=ComponentMask(None=@n34) {r=true}, B=@MaterialExpressionRemap_0)) {name=Unified UV, nodeColor=<rgba 0.4531 1 0 1>}
  n35 = Time()
  n36 = Constant()
  n37 = Sine(None=Multiply(A=Constant() {r=0.2}, B=Add(A=@n35, B=@n36))) {period=2}
  n38 = Constant() {r=24}
  n39 = Divide(A=Floor(None=Multiply(A=@n37, B=@n38)), B=@n38)
  n40 = Multiply(A=Saturate(None=@n9), B=Saturate(None=Multiply(A=@n39) {constB=2}))
  n41 = ScalarParameter() {bUseCustomPrimitiveData=true, defaultValue=1, parameterName=IdleTail light switch, primitiveDataIndex=6}
  NamedRerouteDeclaration(None=Multiply(A=Saturate(None=@n40), B=LinearInterpolate(Alpha=@n41))) {name=light rail mask, nodeColor=<rgba 1 0 0.9609 1>}
  n42 = Multiply(A=TextureCoordinate(), B=ComponentMask(None=Constant3Vector() {constant=<rgba 40 1 0 1>}) {g=true, r=true})
  n43 = Constant() {r=12}
  n44 = Ceil(None=Multiply(A=Time(), B=@n43))
  n45 = Constant3Vector() {constant=<rgba 2 0.5 -2 -0.8>}
  n46 = AppendVector(A=@n45.R, B=@n45.R)
  n47 = TextureSample(UVs=Panner(Coordinate=@n42, Time=Divide(A=@n44, B=@n43), Speed=@n46) {bFractionalPart=true, speedY=0.05}) {samplerSource=SSM_Wrap_WorldGroupSettings, texture=/GrindRail/Textures/T_Railing_Noise}
  n48 = Constant() {r=12}
  n49 = Ceil(None=Multiply(A=Time(), B=@n48))
  n50 = TextureSample(UVs=Panner(Coordinate=@n42, Time=Divide(A=@n49, B=@n48), Speed=@n46) {bFractionalPart=true}) {samplerSource=SSM_Wrap_WorldGroupSettings, texture=&T_Railing_Noise}
  NamedRerouteDeclaration(None=Saturate(None=Multiply(A=@n47.R, B=@n50.G))) {name=light Speedline, nodeColor=<rgba 0 0.25 1 1>}
  n51 = TextureCoordinate()
  n52 = Abs(None=Subtract(A=ComponentMask(None=@n51) {r=true}, B=Constant() {r=0.125}) {constB=0.1})
  n53 = MaterialFunctionCall(In=Power(Base=OneMinus(None=@n52), Exp=Constant() {r=35}) {constExponent=60}, Contrast=Constant() {r=0.6}) {materialFunction=&CheapContrast}
  n54 = Constant() {r=1.5}
  n55 = Add(
   A=Multiply(A=NamedRerouteUsage(), B=ScalarParameter() {bUseCustomPrimitiveData=true, parameterName=Glow switch, primitiveDataIndex=13})
   B=Add(A=Multiply(A=@n53.Result, B=@n54) {constB=2}, B=@n25)
  )
  NamedRerouteDeclaration(None=Divide(
    A=MaterialFunctionCall(LDR Input=@n1.RGB, HDR Tint=@n1.RGB, HDR Intensity Multiplier=@n55) {materialFunction=&DeriveHDRfromLDR}.RemappedRange
    B=EyeAdaptation().EyeAdaptation
   )) {name=Emissive effect, nodeColor=<rgba 1 0.1875 0 1>}
  n56 = TextureCoordinate()
  n57 = Abs(None=Subtract(A=ComponentMask(None=@n56) {r=true}, B=Constant() {r=0.146}) {constB=0.1})
  n58 = NamedRerouteUsage()
  n59 = LinearInterpolate(
   A=Constant() {r=1200}
   B=LinearInterpolate(
    A=ScalarParameter() {defaultValue=250, group=EmissiveColor, parameterName=Width of line}
    B=ScalarParameter() {defaultValue=200, group=EmissiveColor, parameterName=Width of line Spriting}
    Alpha=ScalarParameter() {bUseCustomPrimitiveData=true, group=Gameplay, parameterName=sprinting state, primitiveDataIndex=8}
   ) {constB=400}
   Alpha=Saturate(None=Multiply(A=@n58) {constB=2})
  )
  n60 = Power(Base=OneMinus(None=Saturate(None=@n57)), Exp=@n59)
  n61 = TextureCoordinate()
  n62 = Abs(None=Subtract(A=ComponentMask(None=@n61) {r=true}, B=Constant() {r=0.105}) {constB=0.1})
  n63 = Power(Base=OneMinus(None=Saturate(None=@n62)), Exp=@n59)
  NamedRerouteDeclaration(None=Saturate(None=Add(A=@n60, B=@n63))) {name=Edge, nodeColor=<rgba 0 1 0.0625 1>}
  TextureCoordinate()
  TextureCoordinate()
 ]
}
