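# Clear-coat master material for cyber grind rails on Nanite spline meshes, with glowing emissive that reacts to the player.
# Per-rail custom primitive data drives color, sprint, jump-off, idle lighting, UV range and end glows.
# A player-position sphere mask, panning mask textures, time-of-day scalars and pixel-depth distance glow shape emissive, which is divided by eye adaptation off mobile.
# Quality switches reserve bump-offset UVs and scratch normals for Epic and tune emissive strength per tier.
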Material /GrindRail/Materials/M_Asteria_cyberRailing_Emissive_Master_v3
graph {
 properties {bUsedWithNanite=true, bUsedWithSplineMeshes=true, pixelDepthOffsetMode=PDOM_Legacy, shadingModel=MSM_ClearCoat, usedShadingModels=Clear Coat}
 nodes [
  n1 = VectorParameter() {bUseCustomPrimitiveData=true, group=EmissiveColor, parameterName=EmissiveColorLight, primitiveDataIndex=2, defaultValue=<rgba 1 0 0 1>}
  LinearInterpolate(
   A=TextureSample() {texture=/GrindRail/Textures/T_Railing_A_D}.RGB
   B=@n1.RGB
   Alpha=LinearInterpolate(
    A=NamedRerouteUsage()
    B=NamedRerouteUsage()
    Alpha=ScalarParameter() {bUseCustomPrimitiveData=true, parameterName=diastanc Glow, primitiveDataIndex=17}
   )
  ) {constB=0}
  n2 = Multiply(A=TextureCoordinate(), B=AppendVector(A=Constant() {r=4}, B=Constant() {r=15}))
  n3 = Constant() {r=0.5}
  NamedRerouteDeclaration(None=QualitySwitch(Default=@n2, Low=@n2, High=@n2, Medium=@n2, Epic=BumpOffset(Coordinate=@n2, Height=@n3))) {name=Depth UV, nodeColor=<rgba 0 1 0.3203 1>}
  n4 = NamedRerouteUsage()
  n5 = NamedRerouteUsage()
  NamedRerouteDeclaration(None=Saturate(None=Multiply(A=@n4, B=@n5))) {name="player Emissive  mask", nodeColor=<rgba 0.875 1 0 1>}
  n6 = ScalarParameter() {defaultValue=8000, group=EmissiveColor, parameterName=BlendDistance}
  n7 = Multiply(A=@n6) {constB=1.5}
  n8 = ScalarParameter() {bUseCustomPrimitiveData=true, defaultValue=1, parameterName=IdleTail light switch, primitiveDataIndex=6}
  n9 = Clamp(None=Divide(A=PixelDepth().R, B=LinearInterpolate(A=@n7, B=@n6, Alpha=@n8)) {constB=5000})
  n10 = Constant() {r=2}
  NamedRerouteDeclaration(None=Saturate(None=Power(Base=@n9, Exp=@n10) {constExponent=1.2})) {name=PixelDepthMask, nodeColor=<rgba 0 0.5781 1 1>}
  n11 = TextureSample() {samplerSource=SSM_Wrap_WorldGroupSettings, texture=/GrindRail/Textures/T_Railing_Mask}
  n12 = Constant() {r=0.2}
  n13 = Constant() {r=0.2}
  n14 = Constant() {r=0.6}
  n15 = LinearInterpolate(
   A=QualitySwitch(Default=@n13, Low=@n14, High=@n13, Medium=@n14, Epic=@n13)
   B=ScalarParameter() {defaultValue=2.5, group=EmissiveColor, parameterName=SprintEmissiveStrengthOnTop}
   Alpha=ScalarParameter() {bUseCustomPrimitiveData=true, group=Gameplay, parameterName=sprinting state, primitiveDataIndex=8}
  ) {constB=2.5}
  n16 = NamedRerouteUsage()
  n17 = Multiply(A=NamedRerouteUsage(), B=AppendVector(A=Constant() {r=1.8044}, B=Constant() {r=0.5727}))
  n18 = Max(
   A=Multiply(A=LinearInterpolate(Alpha=@n11.A) {constA=0.4}, B=LinearInterpolate(A=@n12, B=@n15, Alpha=@n16) {constA=0.1}) {constB=0.2}
   B=LinearInterpolate(
    A=ScalarParameter() {defaultValue=0.3, parameterName=PatternDetailLowValue}
    Alpha=TextureSample(UVs=@n17) {samplerSource=SSM_Wrap_WorldGroupSettings, texture=&T_Railing_Mask}.R
   ) {constA=0.5}
  )
  n19 = ScalarParameter() {defaultValue=0.04, group=EmissiveColor, parameterName=Global Emissive intensity}
  n20 = Multiply(A=@n19) {constB=3}
  n21 = ScalarParameter() {bUseCustomPrimitiveData=true, defaultValue=1, parameterName=IdleTail light switch, primitiveDataIndex=6}
  n22 = MaterialFunctionCall(
   Value_Morning=ScalarParameter() {defaultValue=0.55, parameterName=Global rail morning Emissive}
   Value_Day=ScalarParameter() {defaultValue=0.5, parameterName=Global rail DayEmissive}
   Value_Evening=ScalarParameter() {defaultValue=0.4, parameterName=Global rail Evening Emissive}
   Value_Night=ScalarParameter() {defaultValue=0.5, parameterName=Global rail Night Emissive}
  ) {materialFunction=&Scalar_Time_of_Day}
  n23 = Multiply(A=@n22.Result) {constB=0.4}
  n24 = Multiply(A=@n22.Result) {constB=0.8}
  n25 = Multiply(A=@n22.Result) {constB=4}
  n26 = Max(
   A=Multiply(A=NamedRerouteUsage(), B=ScalarParameter() {bUseCustomPrimitiveData=true, parameterName=Glow switch, primitiveDataIndex=13})
   B=LinearInterpolate(
    A=LinearInterpolate(A=@n19, B=@n20, Alpha=@n21)
    B=QualitySwitch(Default=@n22.Result, Low=@n23, High=@n24, Medium=@n25, Epic=@n22.Result)
    Alpha=NamedRerouteUsage()
   ) {constA=0.03, constB=10}
  )
  n27 = Multiply(A=NamedRerouteUsage(), B=AppendVector(A=Constant() {r=2.933}, B=Constant() {r=0.5}))
  n28 = Time()
  n29 = LinearInterpolate(
   A=TextureSample(UVs=Panner(Coordinate=@n27, Time=@n28) {speedX=0.005, speedY=0.5}) {samplerSource=SSM_Wrap_WorldGroupSettings, texture=&T_Railing_Mask}.B
   B=TextureSample(UVs=Panner(Coordinate=@n27, Time=@n28) {speedX=0.005, speedY=0.05}) {samplerSource=SSM_Wrap_WorldGroupSettings, texture=&T_Railing_Mask}.B
   Alpha=TextureSample(UVs=@n27) {texture=/GrindRail/Textures/T_Railing_SpeedMask}.R
  )
  n30 = Constant() {r=0.1}
  n31 = ScalarParameter() {defaultValue=0.6, group=EmissiveColor, parameterName=Emissive line indensity}
  n32 = NamedRerouteUsage()
  n33 = Constant() {r=0.7}
  n34 = LinearInterpolate(
   A=LinearInterpolate(
    A=LinearInterpolate(A=@n30, B=@n31, Alpha=@n32) {constA=0.2, constB=1.4}
    B=@n31
    Alpha=ScalarParameter() {bUseCustomPrimitiveData=true, defaultValue=1, parameterName=IdleTail light switch, primitiveDataIndex=6}
   )
   B=LinearInterpolate(
    A=@n33
    B=Multiply(A=@n33) {constB=5}
    Alpha=ScalarParameter() {bUseCustomPrimitiveData=true, group=Gameplay, parameterName=sprinting state, primitiveDataIndex=8}
   ) {constA=8, constB=25}
   Alpha=NamedRerouteUsage()
  ) {constA=0.2}
  n35 = Power(Base=NamedRerouteUsage()) {constExponent=1.5}
  n36 = ScalarParameter() {bUseCustomPrimitiveData=true, parameterName=Glow switch, primitiveDataIndex=13}
  n37 = NamedRerouteUsage()
  n38 = SphereMask(
   A=ComponentMask(None=TextureCoordinate()) {r=true}
   B=ScalarParameter() {defaultValue=0.45, parameterName=CenterGlowPosition}
   Radius=ScalarParameter() {defaultValue=0.015, parameterName=CenterGlowRadius}
   Hardness=Constant()
  )
  n39 = ScalarParameter() {defaultValue=2, parameterName=CenterGlowExponent}
  n40 = Add(
   A=Max(A=LinearInterpolate(Alpha=@n35) {constA=1, constB=3}, B=Multiply(A=@n36, B=@n37))
   B=Multiply(A=Power(Base=@n38, Exp=@n39), B=ScalarParameter() {defaultValue=2.5, parameterName=CenterGlowScale})
  )
  n41 = VectorParameter() {bUseCustomPrimitiveData=true, group=EmissiveColor, parameterName=EmissiveColorLight, primitiveDataIndex=2, defaultValue=<rgba 1 0 0 1>}
  n42 = Multiply(
   A=Max(A=Multiply(A=@n18, B=@n26) {constB=0.03}, B=Multiply(A=@n29, B=@n34))
   B=Multiply(A=Multiply(A=@n40, B=@n41.RGB), B=ScalarParameter() {defaultValue=5.5, parameterName=IdleSweepHigh})
  )
  n43 = ScalarParameter() {defaultValue=1.2, parameterName=Global Morning Emissive}
  n44 = ScalarParameter() {defaultValue=2, parameterName=Global DayEmissive}
  n45 = ScalarParameter() {defaultValue=2, parameterName=Global EvEning Emissive}
  n46 = ScalarParameter() {defaultValue=1.4, parameterName=Global Night Emissive}
  n47 = ScalarParameter() {defaultValue=3.5, parameterName=Global Idle Emissive}
  n48 = ScalarParameter() {defaultValue=1.5, parameterName=Global Idle night Emissive}
  n49 = Constant() {r=40}
  n50 = Constant() {r=10}
  n51 = LinearInterpolate(
   A=Multiply(A=MaterialFunctionCall(Value_Morning=@n43, Value_Day=@n44, Value_Evening=@n45, Value_Night=@n46) {materialFunction=&Scalar_Time_of_Day}.Result)
   B=LinearInterpolate(
    A=MaterialFunctionCall(Value_Morning=@n47, Value_Day=@n47, Value_Evening=@n47, Value_Night=@n48) {materialFunction=&Scalar_Time_of_Day}.Result
    B=MaterialFunctionCall(Value_Morning=@n49, Value_Day=@n49, Value_Evening=@n49, Value_Night=@n50) {materialFunction=&Scalar_Time_of_Day}.Result
    Alpha=NamedRerouteUsage()
   )
   Alpha=ScalarParameter() {bUseCustomPrimitiveData=true, defaultValue=1, parameterName=IdleTail light switch, primitiveDataIndex=6}
  )
  n52 = NamedRerouteUsage()
  n53 = ScalarParameter() {bUseCustomPrimitiveData=true, parameterName=diastanc Glow, primitiveDataIndex=17}
  n54 = LinearInterpolate(
   B=LinearInterpolate(A=Multiply(A=@n42, B=@n51) {constB=2}, B=LinearInterpolate(B=@n52, Alpha=@n53), Alpha=NamedRerouteUsage()) {constB=0}
   Alpha=NamedRerouteUsage()
  )
  n55 = EyeAdaptation()
  NamedRerouteDeclaration(None=ShadingPathSwitch(Default=Divide(A=@n54, B=@n55.EyeAdaptation), Mobile=@n54)) {name=Emissive Color, nodeColor=<rgba 0.6875 1 0 1>}
  n56 = VectorParameter() {bUseCustomPrimitiveData=true, group=EmissiveColor, parameterName=EmissiveColorLight, primitiveDataIndex=2, defaultValue=<rgba 1 0 0 1>}
  n57 = Saturate(None=InverseLinearInterpolate(
    A=ScalarParameter() {defaultValue=100000, group=EmissiveColor, parameterName=GlowAtDistance_EndDistance}
    B=ScalarParameter() {defaultValue=2500, group=EmissiveColor, parameterName=GlowAtDistance_StartDistance}
    Value=PixelDepth().R
   ))
  n58 = Multiply(A=@n56.RGB, B=@n57)
  n59 = Add(A=@n58, B=Multiply(A=@n58, B=@n57))
  n60 = ScalarParameter() {parameterName=GlowAtDistanceDesaturation}
  NamedRerouteDeclaration(None=Multiply(
    A=Desaturation(None=@n59, Fraction=@n60)
    B=ScalarParameter() {defaultValue=5, group=EmissiveColor, parameterName=Glow indensity at a distance}
   )) {name=Glow at a distance, nodeColor=<rgba 0 1 0.0391 1>}
  n61 = TextureCoordinate()
  MaterialExpressionRemap_0 = MaterialXRemap!(
   None=OneMinus(None=ComponentMask(None=@n61) {g=true})
   InputLow
   InputHigh
   TargetLow=ScalarParameter() {bUseCustomPrimitiveData=true, parameterName=Unified UV Start}
   TargetHigh=ScalarParameter() {bUseCustomPrimitiveData=true, parameterName=Unified UV End, primitiveDataIndex=1}
  )
  NamedRerouteDeclaration(None=AppendVector(A=ComponentMask(None=@n61) {r=true}, B=@MaterialExpressionRemap_0)) {name=Unified UV, nodeColor=<rgba 0.4531 1 0 1>}
  n62 = WorldPosition()
  n63 = CollectionParameter() {collection=&FortniteMaterialParameters, parameterName=PlayerPosition}
  n64 = ComponentMask(None=@n63) {b=true, g=true, r=true}
  n65 = LinearInterpolate(
   A=LinearInterpolate(
    A=ScalarParameter() {defaultValue=400, group=EmissiveColor, parameterName=RadiusDefaultEffect}
    B=ScalarParameter() {defaultValue=900, group=EmissiveColor, parameterName=RadiusSprintingEffectRadius}
    Alpha=ScalarParameter() {bUseCustomPrimitiveData=true, group=Gameplay, parameterName=sprinting state, primitiveDataIndex=8}
   )
   Alpha=ScalarParameter() {bUseCustomPrimitiveData=true, defaultValue=1, parameterName=IdleTail light switch, primitiveDataIndex=6}
  ) {constB=0}
  n66 = ScalarParameter() {group=EmissiveColor, parameterName=HardnessRadius}
  n67 = Multiply(A=Power(Base=SphereMask(A=@n62.XYZ, B=@n64, Radius=@n65, Hardness=@n66)), B=Constant() {r=1.02})
  n68 = ScalarParameter() {bUseCustomPrimitiveData=true, group=Gameplay, parameterName=Jumpoff, primitiveDataIndex=7}
  n69 = ScalarParameter() {bUseCustomPrimitiveData=true, group=Gameplay, parameterName=sprinting state, primitiveDataIndex=8}
  NamedRerouteDeclaration(None=Multiply(A=LinearInterpolate(A=@n67, Alpha=@n68) {constB=0}, B=LinearInterpolate(Alpha=@n69) {constA=0.7})) {name=Player SphereMask, nodeColor=<rgba 0.5 0 1 1>}
  n70 = Saturate(None=Multiply(A=TextureSample() {samplerSource=SSM_Wrap_WorldGroupSettings, texture=/GrindRail/Textures/T_Railing_A_E}.R) {constB=3})
  NamedRerouteDeclaration(None=@n70) {name=Emissive mask, nodeColor=<rgba 0.3125 1 0 1>}
  n71 = Power(
   Base=OneMinus(None=TextureSample() {samplerSource=SSM_Wrap_WorldGroupSettings, texture=&T_Railing_Mask}.A)
   Exp=Constant() {r=9}
  ) {constExponent=4}
  n72 = NamedRerouteUsage()
  NamedRerouteDeclaration(None=Saturate(None=Multiply(A=@n71, B=@n72))) {name=Side mask, nodeColor=<rgba 0.7109 0 1 1>}
  n73 = NamedRerouteUsage()
  n74 = SphereMask(
   A=ComponentMask(None=@n73) {g=true}
   B=Constant()
   Radius=ScalarParameter() {bUseCustomPrimitiveData=true, defaultValue=0.05, parameterName=Glow start Radius, primitiveDataIndex=9}
  ) {attenuationRadius=0.1, hardnessPercent=0}
  n75 = Multiply(
   A=ScalarParameter() {bUseCustomPrimitiveData=true, defaultValue=20, parameterName="Glow start  Intensity", primitiveDataIndex=11}
   B=Power(Base=ComponentMask(None=@n74) {g=true}) {constExponent=3}
  )
  n76 = SphereMask(
   A=ComponentMask(None=@n73) {g=true}
   B=Constant() {r=1}
   Radius=ScalarParameter() {bUseCustomPrimitiveData=true, defaultValue=0.05, parameterName=Glow End Radius, primitiveDataIndex=10}
  ) {attenuationRadius=0.1, hardnessPercent=0}
  n77 = Multiply(
   A=Power(Base=ComponentMask(None=@n76) {g=true}) {constExponent=3}
   B=ScalarParameter() {bUseCustomPrimitiveData=true, defaultValue=20, parameterName="Glow  end Intensity", primitiveDataIndex=12}
  ) {constB=10}
  NamedRerouteDeclaration(None=StaticSwitch(True=Max(A=@n75, B=@n77), False=Constant(), Value=StaticBoolParameter() {parameterName=Enable Glow start and end})) {name=Glow start and end, nodeColor=<rgba 0.4297 1 0 1>}
  n78 = PixelDepth()
  n79 = Constant() {r=30000}
  n80 = Power(Base=Clamp(None=Divide(A=@n78.R, B=@n79) {constB=5000}), Exp=Constant() {r=4}) {constExponent=1.2}
  NamedRerouteDeclaration(None=LinearInterpolate(A=@n70, Alpha=Saturate(None=@n80))) {name=PixelDepthEmissive mask, nodeColor=<rgba 1 0 0.9141 1>}
  NamedRerouteUsage()
  n81 = TextureSample() {samplerType=SAMPLERTYPE_Normal, texture=/GrindRail/Textures/T_Railing_A_N}
  n82 = TextureSample(UVs=MaterialFunctionCall(UVs=TextureCoordinate() {uTiling=5, vTiling=0.2}, "Rotation Angle (0-1)"=Constant() {r=0.25}) {materialFunction=&CustomRotator}.Rotated Values) {samplerType=SAMPLERTYPE_Normal, texture=/GrindRail/Textures/T_Scratches_Normal}
  n83 = Constant() {r=0.5}
  QualitySwitch(
   Default=@n81.RGB
   Low=@n81.RGB
   High=@n81.RGB
   Medium=@n81.RGB
   Epic=MaterialFunctionCall(
    BaseNormal=@n81.RGB
    AdditionalNormal=MaterialFunctionCall(Normal=@n82.RGB, Flatness=@n83) {materialFunction=&FlattenNormal}.Result
   ) {materialFunction=&BlendAngleCorrectedNormals}.Result
  )
  n84 = Constant() {r=0.1}
  n85 = TextureCoordinate() {uTiling=8, vTiling=0.2}
  n86 = Multiply(A=TextureCoordinate() {uTiling=4, vTiling=0.1}) {constB=20}
  n87 = Multiply(A=Max(
    A=TextureSample(UVs=@n85) {samplerSource=SSM_Wrap_WorldGroupSettings, texture=&T_Railing_Mask}.G
    B=TextureSample(UVs=@n86) {samplerSource=SSM_Wrap_WorldGroupSettings, texture=&T_Railing_Mask}.G
   )) {constB=10}
  QualitySwitch(Default=@n84, Low=@n84, High=@n84, Medium=@n84, Epic=LinearInterpolate(Alpha=Saturate(None=@n87)) {constA=0.2, constB=0.01})
  ShadingPathSwitch(Default=LinearInterpolate(B=Constant() {r=1}, Alpha=NamedRerouteUsage()), Mobile=Constant())
  TextureSample() {samplerType=SAMPLERTYPE_Masks, texture=/GrindRail/Textures/T_Railing_A_S}
 ]
}
