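# M_BaseMaterial is a placeholder Default Lit material in GrindRail, flagged for use with spline meshes.
# Its graph has no expressions, so it defines no textures, parameters or shading logic of its own.
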Material /GrindRail/Materials/M_BaseMaterial
graph {expressions {}, properties {bUsedWithSplineMeshes=true, pixelDepthOffsetMode=PDOM_Legacy, usedShadingModels=Default Lit}}
