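# M_booster_master is the grind rail booster master material for instanced meshes, drawing animated LED arrow screens on two sides.
# Per-instance custom data sets each side's green/red color, arrow channel and direction, plus period, trail length, instance index and count.
# A time-driven pulse sweeps across instance indices, brightening arrows and backgrounds to time-of-day intensities through a tiled LED mask.
# A pixel-depth mask blends surface textures toward constants with distance.
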Material /GrindRail/Materials/M_booster_master
graph {
 properties {bUsedWithInstancedStaticMeshes=true, pixelDepthOffsetMode=PDOM_Legacy, usedShadingModels=Default Lit}
 nodes [
  LinearInterpolate(A=TextureSample() {texture=/GrindRail/Textures/T_CyberRailing_Booster_A_D}.RGB, B=Constant() {r=0.05}, Alpha=NamedRerouteUsage())
  LinearInterpolate(
   A=TextureSample() {samplerType=SAMPLERTYPE_Normal, texture=/GrindRail/Textures/T_CyberRailing_Booster_A_N}.RGB
   B=Constant3Vector() {constant=<rgba 0 0 1 0>}
   Alpha=NamedRerouteUsage()
  )
  n1 = Divide(A=PixelDepth().R, B=ScalarParameter() {defaultValue=7500, group=EmissiveColor, parameterName=BlendDistance}) {constB=5000}
  LinearInterpolate(Alpha=NamedRerouteDeclaration(None=Clamp(None=@n1)) {name=PixelDepthMask, nodeColor=<rgba 0 0.5781 1 1>})
  LinearInterpolate(A=TextureSample() {samplerType=SAMPLERTYPE_Masks, texture=/GrindRail/Textures/T_CyberRailing_Booster_A_S}.B, Alpha=NamedRerouteUsage()) {constB=0.3}
  n2 = Constant()
  n3 = ScalarParameter() {defaultValue=1, parameterName=Animation Interval time}
  n4 = Add(A=Multiply(A=NamedRerouteUsage()) {constB=-1}, B=@n3)
  n5 = Add(A=Subtract(A=NamedRerouteUsage(), B=@n3), B=NamedRerouteUsage())
  n6 = NamedRerouteUsage()
  n7 = Distance(A=NamedRerouteUsage(), B=Round(None=LinearInterpolate(A=@n4, B=@n5, Alpha=@n6) {constB=20}))
  n8 = Subtract(A=NamedRerouteUsage(), B=@n7)
  n9 = NamedRerouteUsage()
  n10 = If(A=@n7, B=NamedRerouteUsage(), "A > B"=@n2, "A == B"=@n2, "A < B"=Power(Base=Divide(A=@n8, B=@n9)))
  Multiply(A=@n10)
  n11 = TextureSampleParameter2D() {parameterName=T_Arrow Emissive_E, samplerSource=SSM_Wrap_WorldGroupSettings, texture=/GrindRail/Textures/T_booster_SingleArrow}
  NamedRerouteDeclaration(None=Saturate(None=Multiply(A=@n11.G) {constB=4})) {name=Emissive Mask, nodeColor=<rgba 1 0.3047 0 1>}
  n12 = Multiply(A=TextureCoordinate(), B=ScalarParameter() {defaultValue=160, parameterName=LED_UV_TIling})
  NamedRerouteDeclaration(None=Multiply(
    A=NamedRerouteUsage()
    B=TextureSample(UVs=@n12) {samplerSource=SSM_Wrap_WorldGroupSettings, samplerType=SAMPLERTYPE_Masks, texture=/GrindRail/Textures/T_LED_Mask}.R
   )) {name=LED Mask, nodeColor=<rgba 0 0.0859 1 1>}
  n13 = PerInstanceCustomData!(DefaultValue) {dataIndex=2}
  n14 = Constant3Vector() {constant=<rgba 0.05 1 0.05 0>}
  n15 = Constant3Vector() {constant=<rgba 1 0.05 0.05 0>}
  NamedRerouteDeclaration(None=If(A=VertexInterpolator(VS=@n13).PS, "A > B"=@n14, "A == B"=Constant3Vector(), "A < B"=@n15)) {name=SideA Color, nodeColor=<rgba 0 0.2969 1 1>}
  n16 = PerInstanceCustomData() {dataIndex=3}
  NamedRerouteDeclaration(None=If(A=VertexInterpolator(VS=@n16).PS, "A > B"=@n14, "A == B"=Constant3Vector(), "A < B"=@n15)) {name=SideB_color, nodeColor=<rgba 0.2188 1 0 1>}
  n17 = PerInstanceCustomData() {dataIndex=2}
  n18 = TextureSampleParameter2D() {parameterName=T_Arrow Emissive_E, samplerSource=SSM_Wrap_WorldGroupSettings, texture=&T_booster_SingleArrow}
  n19 = TextureSampleParameter2D() {parameterName=T_Arrow Emissive_E, samplerSource=SSM_Wrap_WorldGroupSettings, texture=&T_booster_SingleArrow}
  NamedRerouteDeclaration(None=If(
    A=VertexInterpolator(VS=@n17).PS
    "A > B"=Power(Base=@n18.A) {constExponent=4}
    "A == B"=Constant3Vector()
    "A < B"=Power(Base=@n19.R) {constExponent=4}
   )) {name=SideA Arrow, nodeColor=<rgba 0 0.2969 1 1>}
  n20 = PerInstanceCustomData() {dataIndex=3}
  n21 = TextureSampleParameter2D() {parameterName=T_Arrow Emissive_E, samplerSource=SSM_Wrap_WorldGroupSettings, texture=&T_booster_SingleArrow}
  n22 = TextureSampleParameter2D() {parameterName=T_Arrow Emissive_E, samplerSource=SSM_Wrap_WorldGroupSettings, texture=&T_booster_SingleArrow}
  NamedRerouteDeclaration(None=If(
    A=VertexInterpolator(VS=@n20).PS
    "A > B"=Power(Base=@n21.A) {constExponent=4}
    "A == B"=Constant3Vector()
    "A < B"=Power(Base=@n22.R) {constExponent=4}
   )) {name=SideB Arrow, nodeColor=<rgba 0 0.2969 1 1>}
  n23 = Multiply(A=VertexInterpolator(VS=PerInstanceCustomData() {dataIndex=4}).PS) {constB=1.3}
  n24 = MaterialFunctionCall(Value=Time(), Period=@n23, -1 to 1=StaticBool()) {materialFunction=&LinearSine}
  n25 = MaterialFunctionCall(Value=Add(A=Time(), B=Multiply(A=@n23) {constB=0.5}), Period=@n23, -1 to 1=StaticBool()) {materialFunction=&LinearSine}
  n26 = Max(A=Multiply(A=@n24.Linear Sine, B=@n24.Linear Sine), B=Multiply(A=@n25.Linear Sine, B=@n25.Linear Sine))
  n27 = OneMinus(None=@n26)
  n28 = PerInstanceCustomData() {dataIndex=5}
  NamedRerouteDeclaration(None=Saturate(None=LinearInterpolate(A=@n26, B=@n27, Alpha=@n28))) {name=SideA_animation, nodeColor=<rgba 0 0.625 1 1>}
  n29 = MaterialFunctionCall(Value=Time(), Period=@n23, -1 to 1=StaticBool()) {materialFunction=&LinearSine}
  n30 = MaterialFunctionCall(Value=Add(A=Time(), B=Multiply(A=@n23) {constB=0.5}), Period=@n23, -1 to 1=StaticBool()) {materialFunction=&LinearSine}
  n31 = Max(A=Multiply(A=@n29.Linear Sine, B=@n29.Linear Sine), B=Multiply(A=@n30.Linear Sine, B=@n30.Linear Sine))
  n32 = OneMinus(None=@n31)
  n33 = PerInstanceCustomData() {dataIndex=6}
  NamedRerouteDeclaration(None=Saturate(None=LinearInterpolate(A=@n31, B=@n32, Alpha=@n33))) {name=SideB_animation, nodeColor=<rgba 0 0.625 1 1>}
  NamedRerouteDeclaration(None=VertexInterpolator(VS=PerInstanceCustomData() {constDefaultValue=3, dataIndex=7}).PS) {name=Animation trailing length, nodeColor=<rgba 0.7578 1 0 1>}
  NamedRerouteDeclaration(None=VertexInterpolator(VS=PerInstanceCustomData()).PS) {name=Total number of booster models, nodeColor=<rgba 0.5938 0 1 1>}
  NamedRerouteDeclaration(None=VertexInterpolator(VS=PerInstanceCustomData() {dataIndex=1}).PS) {name=Instance staic mesh index, nodeColor=<rgba 0 0.4609 1 1>}
  NamedRerouteDeclaration(None=MaterialFunctionCall(
    Value_Morning=ScalarParameter() {defaultValue=100, group=Emissive, parameterName=Arrow Emissive intensity Morning}
    Value_Day=ScalarParameter() {defaultValue=150, parameterName=Arrow Emissive intensity Day}
    Value_Evening=ScalarParameter() {defaultValue=150, parameterName=Arrow Emissive intensity Evening}
    Value_Night=ScalarParameter() {defaultValue=30, parameterName=Arrow Emissive intensity Night}
   ) {materialFunction=&Scalar_Time_of_Day}.Result) {name=The Emissive intensity of the arrow, nodeColor=<rgba 0 1 0.4609 1>}
  NamedRerouteDeclaration(None=MaterialFunctionCall(
    Value_Morning=ScalarParameter() {defaultValue=2, parameterName=intensity of the screen morning}
    Value_Day=ScalarParameter() {defaultValue=3, parameterName="intensity of the screen  day"}
    Value_Evening=ScalarParameter() {defaultValue=2, parameterName="intensity of the screen  evening"}
    Value_Night=ScalarParameter() {defaultValue=2, parameterName=intensity of the screen night}
   ) {materialFunction=&Scalar_Time_of_Day}.Result) {name=Emissive intensity of the screen, nodeColor=<rgba 1 0.1875 0 1>}
  NamedRerouteDeclaration(None=VertexInterpolator(VS=PerInstanceCustomData() {dataIndex=5}).PS) {name=SideA_Reverse animation direction, nodeColor=<rgba 1 0 0.7734 1>}
  NamedRerouteDeclaration(None=VertexInterpolator(VS=PerInstanceCustomData() {dataIndex=6}).PS) {name=SideB_Reverse animation direction, nodeColor=<rgba 1 0 0.8906 1>}
  n34 = Multiply(
   A=NamedRerouteUsage()
   B=LinearInterpolate(A=ScalarParameter() {defaultValue=1, parameterName=Arrow Default intensity}, B=NamedRerouteUsage(), Alpha=@n10)
  )
  n35 = LinearInterpolate(A=ScalarParameter() {defaultValue=0.1, parameterName=BG Screen intensity}, B=NamedRerouteUsage(), Alpha=@n10)
  n36 = Multiply(A=Multiply(A=NamedRerouteUsage(), B=Add(A=@n34, B=@n35)), B=NamedRerouteUsage())
  n37 = Saturate(None=Multiply(A=TextureSampleParameter2D() {parameterName=T_Arrow Emissive_E, samplerSource=SSM_Wrap_WorldGroupSettings, texture=&T_booster_SingleArrow}.B) {constB=20})
  n38 = Constant()
  n39 = ScalarParameter() {defaultValue=1, parameterName=Animation Interval time}
  n40 = Add(A=Multiply(A=NamedRerouteUsage()) {constB=-1}, B=@n39)
  n41 = Add(A=Subtract(A=NamedRerouteUsage(), B=@n39), B=NamedRerouteUsage())
  n42 = NamedRerouteUsage()
  n43 = Distance(A=NamedRerouteUsage(), B=Round(None=LinearInterpolate(A=@n40, B=@n41, Alpha=@n42) {constB=20}))
  n44 = Subtract(A=NamedRerouteUsage(), B=@n43)
  n45 = NamedRerouteUsage()
  n46 = If(A=@n43, B=NamedRerouteUsage(), "A > B"=@n38, "A == B"=@n38, "A < B"=Power(Base=Divide(A=@n44, B=@n45)))
  n47 = Multiply(
   A=NamedRerouteUsage()
   B=LinearInterpolate(A=ScalarParameter() {defaultValue=1, parameterName=Arrow Default intensity}, B=NamedRerouteUsage(), Alpha=@n46)
  )
  n48 = LinearInterpolate(A=ScalarParameter() {defaultValue=0.1, parameterName=BG Screen intensity}, B=NamedRerouteUsage(), Alpha=@n46)
  n49 = Multiply(A=Multiply(A=NamedRerouteUsage(), B=Add(A=@n47, B=@n48)), B=NamedRerouteUsage())
  n50 = OneMinus(None=TextureSampleParameter2D() {parameterName=T_Arrow Emissive_E, samplerSource=SSM_Wrap_WorldGroupSettings, texture=&T_booster_SingleArrow}.B)
  n51 = NamedRerouteUsage()
  n52 = Multiply(A=Saturate(None=Power(Base=@n50) {constExponent=20}), B=Saturate(None=@n51))
  NamedRerouteDeclaration(None=Add(A=Multiply(A=@n36, B=@n37), B=Multiply(A=@n49, B=@n52))) {name=Emissive, nodeColor=<rgba 1 0.7969 0 1>}
  Power(Base=@n51) {constExponent=20}
  Reroute(None=NamedRerouteUsage())
 ]
}
